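# Translucent, unlit, two-sided particle material for the TechnoAwaken Shake character's cone wave effect, usable with Niagara sprites.
# Particle color is hue-shifted along UV V and scaled by a V-based bottom boost gradient.
# A panning plasma texture, UV-distorted by panning normal-map noise, blends in along V to modulate that color for emissive.
# A separate alpha chain multiplies a triangle mask, squared depth fade, scalar weight and particle alpha.
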Material /Game/Characters/Player/Male/Medium/Bodies/M_MED_TechnoAwaken_Shake/FX/M_TechnoA_ConeWave
graph {
 properties {
  bUsedWithInstancedStaticMeshes=true
  bUsedWithNanite=true
  bUsedWithNiagaraSprites=true
  blendMode=BLEND_Translucent
  pixelDepthOffsetMode=PDOM_Legacy
  shadingModel=MSM_Unlit
  twoSided=true
  usedShadingModels=Unlit
 }
 nodes [
  n1 = TextureCoordinate()
  n2 = VectorParameter() {parameterName=BottomBoost, defaultValue=<rgba 1 15 3 1>}
  n3 = Power(
   Base=Multiply(
    A=MaterialFunctionCall(Float2=@n1) {materialFunction=&BreakOutFloat2Components}.G
    B=@n2.R
   )
   Exp=@n2.G
  )
  n4 = TextureCoordinate()
  n5 = VectorParameter() {parameterName=HueShift, defaultValue=<rgba 1 2 -0.2 1>}
  n6 = Power(
   Base=Multiply(
    A=MaterialFunctionCall(Float2=@n4) {materialFunction=&BreakOutFloat2Components}.G
    B=@n5.R
   )
   Exp=@n5.G
  )
  n7 = ParticleColor()
  n8 = Multiply(
   A=Add(A=Multiply(A=@n3, B=@n2.B))
   B=MaterialFunctionCall(Hue Shift Percentage=Multiply(A=@n6, B=@n5.B), Texture=Multiply(A=@n7.RGB)) {materialFunction=&HueShift}.Result
  )
  n9 = TextureObjectParameter() {
   group=SparkTexture
   parameterName=SparkTexture_Texture
   samplerType=SAMPLERTYPE_LinearGrayscale
   sortPriority=0
   texture=/Game/Effects/Fort_Effects/Emotes/UltraLight/Texture/T_Plasma_B_Greyscale
  }
  n10 = TextureObjectParameter() {
   group=Noise
   parameterName=Noise_Texture
   samplerType=SAMPLERTYPE_Normal
   sortPriority=0
   texture=&T_Weap_Distort_N_02
  }
  n11 = VectorParameter() {group=Noise, parameterName=Noise_UVControl, channelNames {a=VPan, b=UPan, g=VScale, r=UScale}, defaultValue=<rgba 1 1 0.1 0.3>}
  n12 = MaterialFunctionCall(Texture=@n10, UVControl_RGB=@n11, UVControl_A=@n11) {materialFunction=&FXCommon_TextureSample}
  n13 = VectorParameter() {
   group=Noise
   parameterName=Noise_NoiseControl
   sortPriority=60
   channelNames {a=VStrength, b=UStrength, g=VOffset, r=UOffset}
   defaultValue=<rgba 0 0 0 0.5>
  }
  n14 = VectorParameter() {
   group=SparkTexture
   parameterName=SparkTexture_UVControl
   channelNames {a=VPan, b=UPan, g=VScale, r=UScale}
   defaultValue=<rgba 0.5 2 0 0.12>
  }
  n15 = VectorParameter() {parameterName=aefaef, defaultValue=<rgba 5 1 8 1>}
  n16 = Power(
   Base=MaterialFunctionCall(
    Texture=@n9
    AdditionalUVOffset=MaterialFunctionCall!(NoiseValue=@n12.RG, Control_RGB=@n13, Control_A=@n13) {materialFunction=/Game/Effects/Niagara/MasterMaterials/Functions/Common/FXCommon_NoiseControl}.NoiseValue
    UVControl_RGB=@n14
    UVControl_A=@n14
   ) {materialFunction=&FXCommon_TextureSample}.R
   Exp=@n15.R
  )
  n17 = TextureCoordinate()
  n18 = LinearInterpolate(
   B=Add(A=Multiply(A=@n16, B=@n15.G))
   Alpha=Power(
    Base=MaterialFunctionCall(Float2=@n17) {materialFunction=&BreakOutFloat2Components}.G
    Exp=@n15.B
   ) {constExponent=5}
  ) {constA=1}
  MaterialFunctionCall(Emissive=Divide(A=Multiply(A=@n8, B=@n18)) {constB=1.977}) {materialFunction=&MF_FX_EmissiveAdjustments}
  n19 = TextureSampleParameter2D() {
   parameterName=Param
   samplerType=SAMPLERTYPE_Grayscale
   texture=/Game/Characters/Player/Male/Medium/Bodies/M_MED_TungStan/Textures/Tungstan_TriangleMask02
  }
  n20 = Power(Base=DepthFade(FadeDistance=ScalarParameter() {defaultValue=10, parameterName=DDF}) {fadeDistanceDefault=5})
  n21 = ScalarParameter() {defaultValue=0.3, parameterName=W}
  Multiply(A=Multiply(A=@n19.R, B=Multiply(A=@n20, B=@n21) {constB=0.3}), B=ParticleColor().A)
  ParticleColor()
 ]
}
