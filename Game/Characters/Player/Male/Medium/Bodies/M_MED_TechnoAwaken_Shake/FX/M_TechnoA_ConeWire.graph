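# Translucent, unlit, two-sided particle material for the TechnoAwaken Shake character FX, drawing a glowing triangle-wire pattern, by its name on a cone.
# A panning spark texture is refined, boosted and tinted by particle color, then hue-shifted along V for emissive.
# Panning noise, limited by a V gradient, distorts the wire mask's UVs.
# The refined mask is multiplied by a V edge mask, depth fade and particle alpha for transparency.
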Material /Game/Characters/Player/Male/Medium/Bodies/M_MED_TechnoAwaken_Shake/FX/M_TechnoA_ConeWire
graph {
 properties {
  bUsedWithInstancedStaticMeshes=true
  bUsedWithNanite=true
  bUsedWithNiagaraSprites=true
  blendMode=BLEND_Translucent
  pixelDepthOffsetMode=PDOM_Legacy
  shadingModel=MSM_Unlit
  twoSided=true
  usedShadingModels=Unlit
 }
 nodes [
  n1 = MaterialFunctionCall(Float2=TextureCoordinate()) {materialFunction=&BreakOutFloat2Components}
  n2 = VectorParameter() {parameterName=HueShift, defaultValue=<rgba 1 2 -0.2 1>}
  n3 = Multiply(A=Power(Base=Multiply(A=@n1.G, B=@n2.R), Exp=@n2.G), B=@n2.B)
  n4 = TextureObjectParameter() {
   group=SparkTexture
   parameterName=SparkTexture_Texture
   samplerType=SAMPLERTYPE_Normal
   sortPriority=0
   texture=/Game/Effects/Fort_Effects/Effects/Environment/Chrome/Materials/T_Chrome_Sand_Pickup2
  }
  n5 = VectorParameter() {group=SparkTexture, parameterName=SparkTexture_UVControl, channelNames {a=VPan, b=UPan, g=VScale, r=UScale}, defaultValue=<rgba 0.5 0.5 0 1>}
  n6 = MaterialFunctionCall(Texture=@n4, UVControl_RGB=@n5, UVControl_A=@n5) {materialFunction=&FXCommon_TextureSample}
  n7 = VectorParameter() {parameterName=SKRF, defaultValue=<rgba 40 2 0.02 0.5>}
  n8 = MaterialFunctionCall(Texture=Multiply(A=Add(A=@n6.R)) {constB=0.5}, RGB=@n7.RGB, A=@n7.RGB) {materialFunction=&MF_FX_Refine}
  n9 = ScalarParameter() {defaultValue=500, parameterName=t}
  n10 = Multiply(A=ParticleColor().RGB, B=Add(A=Multiply(A=@n8.Result, B=@n9)))
  MaterialFunctionCall(Emissive=Divide(A=MaterialFunctionCall(Hue Shift Percentage=@n3, Texture=@n10) {materialFunction=&HueShift}.Result) {constB=1.977}) {materialFunction=&MF_FX_EmissiveAdjustments}
  n11 = TextureObjectParameter() {
   group=MainTex
   parameterName=MainTex_Texture
   samplerType=SAMPLERTYPE_LinearGrayscale
   sortPriority=0
   texture=&/Tungstan_TriangleMask02_Wire
  }
  n12 = TextureObjectParameter() {
   group=Noise
   parameterName=Noise_Texture
   samplerType=SAMPLERTYPE_Normal
   sortPriority=0
   texture=&T_Weap_Distort_N_02
  }
  n13 = VectorParameter() {group=Noise, parameterName=Noise_UVControl, channelNames {a=VPan, b=UPan, g=VScale, r=UScale}, defaultValue=<rgba 1 1.5 0 1.5>}
  n14 = MaterialFunctionCall(Texture=@n12, UVControl_RGB=@n13, UVControl_A=@n13) {materialFunction=&FXCommon_TextureSample}
  n15 = VectorParameter() {
   group=Noise
   parameterName=Noise_NoiseControl
   sortPriority=60
   channelNames {a=VStrength, b=UStrength, g=VOffset, r=UOffset}
   defaultValue=<rgba 0 0 0.4 0>
  }
  n16 = Add(
   A=MaterialFunctionCall(Float2=TextureCoordinate()) {materialFunction=&BreakOutFloat2Components}.G
   B=ScalarParameter() {defaultValue=-0.5339, group=NoiseMask, parameterName=NoiseMask_GrayAdjust_Add}
  )
  n17 = ScalarParameter() {defaultValue=3, group=NoiseMask, parameterName=NoiseMask_GrayAdjust_Mul}
  n18 = Power(
   Base=Saturate(None=Multiply(A=@n16, B=@n17))
   Exp=ScalarParameter() {defaultValue=2, group=NoiseMask, parameterName=NoiseMask_GrayAdjust_Power}
  )
  n19 = VectorParameter() {group=MainTex, parameterName=MainTex_UVControl, channelNames {a=VPan, b=UPan, g=VScale, r=UScale}, defaultValue=<rgba 1 1 0 0>}
  n20 = MaterialFunctionCall(
   Texture=@n11
   AdditionalUVOffset=Multiply(
    A=MaterialFunctionCall!(NoiseValue=@n14.RG, Control_RGB=@n15, Control_A=@n15) {materialFunction=/Game/Effects/Niagara/MasterMaterials/Functions/Common/FXCommon_NoiseControl}.NoiseValue
    B=Saturate(None=@n18)
   )
   UVControl_RGB=@n19
   UVControl_A=@n19
  ) {materialFunction=&FXCommon_TextureSample}
  n21 = VectorParameter() {parameterName=RF, defaultValue=<rgba 25 1 0.7 0.5>}
  n22 = VectorParameter() {group=Mask, parameterName=Mask_EdgeMask_InvertControl, channelNames {a="", b=Power, g=Multiply, r=Offset}, defaultValue=<rgba 0 2 1.3 1>}
  n23 = VectorParameter() {group=Mask, parameterName=Mask_EdgeMask_Control, channelNames {a="", b=Power, g=Multiply, r=Offset}, defaultValue=<rgba 0 50 0 1>}
  n24 = MaterialFunctionCall!(
   UorV=MaterialFunctionCall(Float2=TextureCoordinate()) {materialFunction=&BreakOutFloat2Components}.G
   Invert_EdgeControl=@n22
   NoInvert_EdgeControl=@n23
  ) {materialFunction=/Game/Effects/Niagara/MasterMaterials/Functions/Common/FXCommon_EdgeMask_Ex}
  n25 = Power(Base=DepthFade(FadeDistance=ScalarParameter() {defaultValue=5, parameterName=ADF}) {fadeDistanceDefault=5})
  Multiply(
   A=Multiply(
    A=MaterialFunctionCall(Texture=@n20.R, RGB=@n21.RGB, A=@n21.RGB) {materialFunction=&MF_FX_Refine}.Result
    B=Multiply(A=@n24.Mask, B=@n25)
   )
   B=ParticleColor().A
  )
 ]
}
