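# Unlit, translucent, two-sided Niagara mesh-particle material for a spline circle effect, in the TechnoAwaken Shake player body FX folder.
# A refined edge mask across U multiplies particle color, hue-shifted by grayscale Perlin noise sampled in instance-space XY, feeding emissive adjustments.
# A second parameterized edge mask multiplies particle alpha.
# SplineThicken produces a world position offset, using one width parameter for base and tip.
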Material /Game/Characters/Player/Male/Medium/Bodies/M_MED_TechnoAwaken_Shake/FX/M_TechnoA_SplineCircle2
graph {
 properties {
  bUsedWithNiagaraMeshParticles=true
  blendMode=BLEND_Translucent
  pixelDepthOffsetMode=PDOM_Legacy
  shadingModel=MSM_Unlit
  twoSided=true
  usedShadingModels=Unlit
 }
 nodes [
  n1 = MaterialFunctionCall(Float2=TextureCoordinate()) {materialFunction=&BreakOutFloat2Components}
  n2 = Constant4Vector() {constant=<rgba 0 1 1 1>}
  n3 = VectorParameter() {parameterName=Refine, defaultValue=<rgba 3 30 0 0>}
  n4 = Add(A=MaterialFunctionCall(
    Texture=MaterialFunctionCall!(UorV=@n1.R, EdgeControl=@n2.RGBA) {materialFunction=/Game/Effects/Niagara/MasterMaterials/Functions/Common/FXCommon_EdgeMask_Simple}.Mask
    RGB=@n3.RGB
    A=@n3.RGB
   ) {materialFunction=&MF_FX_Refine}.Result)
  n5 = TextureObjectParameter() {
   group=Hue
   parameterName=Hue_Texture
   samplerType=SAMPLERTYPE_LinearGrayscale
   sortPriority=0
   texture=&T_SimplePerlin_Greyscale
  }
  n6 = VectorParameter() {group=Hue, parameterName=Hue_UVControl, channelNames {a=VPan, b=UPan, g=VScale, r=UScale}, defaultValue=<rgba 0.0005 0.0005 0 500>}
  n7 = WorldPosition()
  n8 = MaterialFunctionCall(
   Texture=@n5
   UVControl_RGB=@n6
   UVControl_A=@n6
   "[Override]UVInput"=ComponentMask(None=TransformPosition(None=@n7.XYZ) {transformSourceType=TRANSFORMPOSSOURCE_World, transformType=TRANSFORMPOSSOURCE_Instance}) {g=true, r=true}
  ) {materialFunction=&FXCommon_TextureSample}
  n9 = ScalarParameter() {defaultValue=-0.15, group=Hue, parameterName=Hue_GrayAdjust_Add}
  n10 = ParticleColor()
  n11 = MaterialFunctionCall(
   Hue Shift Percentage=Multiply(A=Add(A=@n8.R, B=@n9), B=ScalarParameter() {defaultValue=0.2, group=Hue, parameterName=Hue_GrayAdjust_Mul})
   Texture=@n10.RGB
  ) {materialFunction=&HueShift}
  MaterialFunctionCall(Emissive=Divide(A=Multiply(A=@n4, B=@n11.Result))) {materialFunction=&MF_FX_EmissiveAdjustments}
  n12 = VectorParameter() {group=EdgeMask, parameterName=EdgeMask_EdgeMask_Control, channelNames {a="", b=Power, g=Multiply, r=Offset}, defaultValue=<rgba 0 5 2 1>}
  Multiply(
   A=MaterialFunctionCall(UorV=@n1.R, EdgeControl=@n12) {materialFunction=&FXCommon_EdgeMask_Simple}.Mask
   B=@n10.A
  )
  n13 = ScalarParameter() {defaultValue=2, parameterName=CW}
  Reroute(None=MaterialFunctionCall(WidthBase=@n13, WidthTip=@n13) {materialFunction=&SplineThicken}.WorldPositionOffset)
 ]
}
