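# Translucent unlit Niagara mesh-particle material for the TechnoAwaken Shake character FX, drawing noise-shaken lines thickened with SplineThicken.
# Masked, panning noise, with an optional NoiseB layer, displaces vertices along normals, scaled by dynamic parameters.
# The same noise magnitude drives brightness spikes, multiplied by particle color hue-shifted through an instance-space noise texture.
# An edge mask and optional depth fade shape opacity, while offset RGB samples add optional chromatic splitting.
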Material /Game/Characters/Player/Male/Medium/Bodies/M_MED_TechnoAwaken_Shake/FX/M_TechnoA_SplineVersion
graph {
 expressions {
  MaterialExpressionAbs_0 {class=MaterialExpressionAbs, inputs [None {from=MaterialExpressionMultiply_9, output=""}], properties {}}
  MaterialExpressionAbs_1 {class=MaterialExpressionAbs, inputs [None {from=MaterialExpressionMultiply_6, output=""}], properties {}}
  MaterialExpressionAbs_2 {class=MaterialExpressionAbs, inputs [None {from=MaterialExpressionVectorParameter_2, output=UStrength}], properties {}}
  MaterialExpressionAdd_0 {
   class=MaterialExpressionAdd
   inputs [
    A {from=MaterialExpressionReroute_0}
    B {from=MaterialExpressionMaterialFunctionCall_0, output=WorldPositionOffset}
   ]
  }
  MaterialExpressionAdd_1 {
   class=MaterialExpressionAdd
   inputs [
    A {from=MaterialExpressionMultiply_0}
    B {from=MaterialExpressionWorldPosition_0, output=XYZ}
   ]
  }
  MaterialExpressionAdd_10 {
   class=MaterialExpressionAdd
   inputs [
    A {from=MaterialExpressionReroute_5}
    B {from=MaterialExpressionMultiply_20}
   ]
  }
  MaterialExpressionAdd_2 {
   class=MaterialExpressionAdd
   inputs [
    A {from=MaterialExpressionMaterialFunctionCall_12, output=Result}
    B {from=MaterialExpressionDynamicParameter_9, output=MaskOffset_NoiseB}
   ]
  }
  MaterialExpressionAdd_3 {
   class=MaterialExpressionAdd
   inputs [
    A {from=MaterialExpressionMultiply_12}
    B {from=MaterialExpressionAdd_7}
   ]
  }
  MaterialExpressionAdd_4 {
   class=MaterialExpressionAdd
   inputs [
    A {from=MaterialExpressionMaterialFunctionCall_15, output=Result}
    B {from=MaterialExpressionDynamicParameter_6, output=MaskOffset}
   ]
  }
  MaterialExpressionAdd_5 {
   class=MaterialExpressionAdd
   inputs [
    A {from=MaterialExpressionMaterialFunctionCall_16, output=R}
    B {from=MaterialExpressionScalarParameter_3}
   ]
  }
  MaterialExpressionAdd_6 {
   class=MaterialExpressionAdd
   inputs [
    A {from=MaterialExpressionMultiply_13}
    B {from=MaterialExpressionReroute_1}
   ]
  }
  MaterialExpressionAdd_7 {
   class=MaterialExpressionAdd
   inputs [
    A {from=MaterialExpressionMultiply_15}
    B {from=MaterialExpressionScalarParameter_11}
   ]
  }
  MaterialExpressionAdd_8 {
   class=MaterialExpressionAdd
   inputs [
    A {from=""}
    B {from=MaterialExpressionLinearInterpolate_2}
   ]
   properties {constA=1}
  }
  MaterialExpressionAppendVector_0 {
   class=MaterialExpressionAppendVector
   inputs [
    A {from=MaterialExpressionVectorParameter_8, output=R}
    B {from=MaterialExpressionVectorParameter_8, output=R}
   ]
  }
  MaterialExpressionAppendVector_1 {
   class=MaterialExpressionAppendVector
   inputs [
    A {from=MaterialExpressionVectorParameter_10, output=R}
    B {from=MaterialExpressionVectorParameter_10, output=R}
   ]
  }
  MaterialExpressionAppendVector_2 {
   class=MaterialExpressionAppendVector
   inputs [
    A {from=MaterialExpressionVectorParameter_11, output=UScale}
    B {from=MaterialExpressionVectorParameter_11, output=UScale}
   ]
  }
  MaterialExpressionAppendVector_3 {
   class=MaterialExpressionAppendVector
   inputs [
    A {from=MaterialExpressionVectorParameter_11, output=UPan}
    B {from=MaterialExpressionVectorParameter_11, output=UPan}
   ]
  }
  MaterialExpressionComponentMask_0 {class=MaterialExpressionComponentMask, inputs [None {from=MaterialExpressionTransformPosition_0}], properties {g=true, r=true}}
  MaterialExpressionConstant3Vector_0 {class=MaterialExpressionConstant3Vector, properties {constant=<rgba 0.7772 0 1 1>}}
  MaterialExpressionConvert_0 {
   class=MaterialExpressionConvert
   inputs [
    None {from=MaterialExpressionVectorParameter_0, output=UScale}
    None {from=MaterialExpressionLinearInterpolate_1}
    None {from=MaterialExpressionVectorParameter_0, output=UScale}
   ]
   properties {
    nodeName=Make Vector3
    convertInputs [
     {
      type=Scalar
      expressionInput {
       expression=&self:MaterialExpressionVectorParameter_0
       inputName=None
       mask=1
       maskA=0
       maskB=0
       maskG=0
       maskR=1
       outputIndex=1
      }
     }
     {
      type=Scalar
      expressionInput {
       expression=&self:MaterialExpressionLinearInterpolate_1
       inputName=None
       mask=0
       maskA=0
       maskB=0
       maskG=0
       maskR=0
       outputIndex=0
      }
     }
     {
      type=Scalar
      expressionInput {
       expression=&self:MaterialExpressionVectorParameter_0
       inputName=None
       mask=1
       maskA=0
       maskB=1
       maskG=0
       maskR=0
       outputIndex=3
      }
     }
    ]
    convertOutputs [{type=Vector3}]
   }
  }
  MaterialExpressionConvert_1 {
   class=MaterialExpressionConvert
   inputs [
    None {from=MaterialExpressionTextureSample_1, output=R}
    None {from=MaterialExpressionTextureSample_3, output=G}
    None {from=MaterialExpressionTextureSample_2, output=B}
   ]
   properties {
    nodeName=Make Vector3
    convertInputs [
     {
      type=Scalar
      expressionInput {
       expression=&self:MaterialExpressionTextureSample_1
       inputName=None
       mask=1
       maskA=0
       maskB=0
       maskG=0
       maskR=1
       outputIndex=1
      }
     }
     {
      type=Scalar
      expressionInput {
       expression=&self:MaterialExpressionTextureSample_3
       inputName=None
       mask=1
       maskA=0
       maskB=0
       maskG=1
       maskR=0
       outputIndex=2
      }
     }
     {
      type=Scalar
      expressionInput {
       expression=&self:MaterialExpressionTextureSample_2
       inputName=None
       mask=1
       maskA=0
       maskB=1
       maskG=0
       maskR=0
       outputIndex=3
      }
     }
    ]
    convertOutputs [{type=Vector3}]
   }
  }
  MaterialExpressionDepthFade_0 {
   class=MaterialExpressionDepthFade
   inputs [
    Opacity {from=MaterialExpressionMultiply_4, output=""}
    FadeDistance {from=MaterialExpressionScalarParameter_12, output=""}
   ]
   properties {}
  }
  MaterialExpressionDivide_0 {
   class=MaterialExpressionDivide
   inputs [
    A {from=MaterialExpressionAbs_0}
    B {from=MaterialExpressionDivide_1}
   ]
  }
  MaterialExpressionDivide_1 {
   class=MaterialExpressionDivide
   inputs [
    A {from=MaterialExpressionMultiply_10}
    B {from=""}
   ]
  }
  MaterialExpressionDivide_2 {
   class=MaterialExpressionDivide
   inputs [
    A {from=MaterialExpressionMultiply_11}
    B {from=""}
   ]
   properties {constB=1.977}
  }
  MaterialExpressionDivide_3 {
   class=MaterialExpressionDivide
   inputs [
    A {from=MaterialExpressionAbs_1}
    B {from=MaterialExpressionDivide_4}
   ]
  }
  MaterialExpressionDivide_4 {
   class=MaterialExpressionDivide
   inputs [
    A {from=MaterialExpressionMultiply_7}
    B {from=""}
   ]
  }
  MaterialExpressionDynamicParameter_0 {class=MaterialExpressionDynamicParameter, inputs [], properties {paramNames [MeshSize, SoftWaveStr, MaskOffset_NoiseB, SoftWaveStr_NoiseB]}}
  MaterialExpressionDynamicParameter_1 {class=MaterialExpressionDynamicParameter, inputs [], properties {paramNames [MeshSize, SoftWaveStr, MaskOffset_NoiseB, SoftWaveStr_NoiseB]}}
  MaterialExpressionDynamicParameter_2 {class=MaterialExpressionDynamicParameter, inputs [], properties {paramNames [MeshSize, SoftWaveStr, MaskOffset_NoiseB, SoftWaveStr_NoiseB]}}
  MaterialExpressionDynamicParameter_4 {class=MaterialExpressionDynamicParameter, inputs [], properties {parameterIndex=1, paramNames [Time_NoiseB,WPOStr,MaskOffset,Time]}}
  MaterialExpressionDynamicParameter_5 {class=MaterialExpressionDynamicParameter, inputs [], properties {parameterIndex=1, paramNames [Time_NoiseB,WPOStr,MaskOffset,Time]}}
  MaterialExpressionDynamicParameter_6 {class=MaterialExpressionDynamicParameter, inputs [], properties {parameterIndex=1, paramNames [Time_NoiseB,WPOStr,MaskOffset,Time]}}
  MaterialExpressionDynamicParameter_7 {class=MaterialExpressionDynamicParameter, inputs [], properties {parameterIndex=2, paramNames [SpikeB, SpikeB_NoiseB, WPOStr_NoiseB, ""]}}
  MaterialExpressionDynamicParameter_8 {class=MaterialExpressionDynamicParameter, inputs [], properties {parameterIndex=2, paramNames [SpikeB, SpikeB_NoiseB, WPOStr_NoiseB, ""]}}
  MaterialExpressionDynamicParameter_9 {class=MaterialExpressionDynamicParameter, inputs [], properties {paramNames [MeshSize, SoftWaveStr, MaskOffset_NoiseB, SoftWaveStr_NoiseB]}}
  MaterialExpressionLinearInterpolate_0 {
   class=MaterialExpressionLinearInterpolate
   inputs [
    A {from=MaterialExpressionMaterialFunctionCall_2, output=R}
    B {from=MaterialExpressionNamedRerouteUsage_0}
    Alpha {from=MaterialExpressionScalarParameter_15}
   ]
  }
  MaterialExpressionLinearInterpolate_1 {
   class=MaterialExpressionLinearInterpolate
   inputs [
    A {from=MaterialExpressionVectorParameter_0, output=VScale}
    B {from=MaterialExpressionSubtract_1}
    Alpha {from=MaterialExpressionScalarParameter_14}
   ]
  }
  MaterialExpressionLinearInterpolate_2 {
   class=MaterialExpressionLinearInterpolate
   inputs [
    A {from=MaterialExpressionReroute_4}
    B {from=MaterialExpressionMultiply_27}
    Alpha {from=MaterialExpressionScalarParameter_16}
   ]
  }
  MaterialExpressionLinearInterpolate_3 {
   class=MaterialExpressionLinearInterpolate
   inputs [
    A {from=""}
    B {from=MaterialExpressionParticleRandom_1}
    Alpha {from=MaterialExpressionScalarParameter_18}
   ]
  }
  MaterialExpressionMaterialFunctionCall_0 {
   class=MaterialExpressionMaterialFunctionCall
   inputs [
    WidthBase {from=MaterialExpressionScalarParameter_0}
    WidthTip {from=MaterialExpressionScalarParameter_0}
    WorldPosition {from=MaterialExpressionAdd_1}
    UVs for Projection {from=""}
    UVs for Thickness {from=""}
    UVs for texturing {from=""}
    Expand U or V UV Channel {from=""}
    DeriveNormalZ {from=""}
    AngleCorrectedNormals {from=""}
    AdditionalNormal {from=""}
    FlattenPixelNormal {from=""}
    UVs For NormalShading {from=""}
   ]
   properties {materialFunction=&SplineThicken}
  }
  MaterialExpressionMaterialFunctionCall_1 {
   class=MaterialExpressionMaterialFunctionCall
   inputs [
    Texture {from=MaterialExpressionTextureObjectParameter_0}
    AdditionalUVOffset {from=MaterialExpressionMultiply_3}
    UVControl_RGB {from=MaterialExpressionConvert_0}
    UVControl_A {from=MaterialExpressionReroute_2}
    "[Override]TimeInput" {from=""}
    "[Override]UVInput" {from=""}
   ]
   properties {materialFunction=&FXCommon_TextureSample}
  }
  MaterialExpressionMaterialFunctionCall_10 {
   class=MaterialExpressionMaterialFunctionCall
   inputs [
    Texture {from=MaterialExpressionTextureObjectParameter_4}
    AdditionalUVOffset {from=""}
    UVControl_RGB {from=MaterialExpressionVectorParameter_9}
    UVControl_A {from=MaterialExpressionVectorParameter_9}
    "[Override]TimeInput" {from=""}
    "[Override]UVInput" {from=""}
   ]
   properties {materialFunction=&FXCommon_TextureSample}
  }
  MaterialExpressionMaterialFunctionCall_11 {
   class=MaterialExpressionMaterialFunctionCall
   inputs [
    Texture {from=MaterialExpressionMaterialFunctionCall_4, output=Mask}
    RGB {from=MaterialExpressionVectorParameter_4, output=RGB}
    A {from=MaterialExpressionVectorParameter_4, output=RGB}
   ]
   properties {materialFunction=&MF_FX_Refine}
  }
  MaterialExpressionMaterialFunctionCall_12 {
   class=MaterialExpressionMaterialFunctionCall
   inputs [
    In {from=MaterialExpressionPower_2}
    Contrast {from=MaterialExpressionScalarParameter_2}
   ]
   properties {materialFunction=&CheapContrast}
  }
  MaterialExpressionMaterialFunctionCall_14 {
   class=MaterialExpressionMaterialFunctionCall
   inputs [
    Texture {from=MaterialExpressionTextureObjectParameter_3}
    AdditionalUVOffset {from=MaterialExpressionLinearInterpolate_3}
    UVControl_RGB {from=MaterialExpressionVectorParameter_7}
    UVControl_A {from=MaterialExpressionVectorParameter_7}
    "[Override]TimeInput" {from=""}
    "[Override]UVInput" {from=""}
   ]
   properties {materialFunction=&FXCommon_TextureSample}
  }
  MaterialExpressionMaterialFunctionCall_15 {
   class=MaterialExpressionMaterialFunctionCall
   inputs [
    In {from=MaterialExpressionPower_0}
    Contrast {from=MaterialExpressionScalarParameter_1}
   ]
   properties {materialFunction=&CheapContrast}
  }
  MaterialExpressionMaterialFunctionCall_16 {
   class=MaterialExpressionMaterialFunctionCall
   inputs [
    Texture {from=MaterialExpressionTextureObjectParameter_2}
    AdditionalUVOffset {from=""}
    UVControl_RGB {from=MaterialExpressionVectorParameter_6}
    UVControl_A {from=MaterialExpressionVectorParameter_6}
    "[Override]TimeInput" {from=""}
    "[Override]UVInput" {from=MaterialExpressionComponentMask_0}
   ]
   properties {materialFunction=&FXCommon_TextureSample}
  }
  MaterialExpressionMaterialFunctionCall_17 {
   class=MaterialExpressionMaterialFunctionCall
   inputs [
    Hue Shift Percentage {from=MaterialExpressionMultiply_8}
    Texture {from=MaterialExpressionParticleColor_0, output=RGB}
   ]
   properties {materialFunction=&HueShift}
  }
  MaterialExpressionMaterialFunctionCall_2 {
   class=MaterialExpressionMaterialFunctionCall
   inputs [Float2 {from=MaterialExpressionMaterialFunctionCall_5, output=NoiseValue}]
   properties {materialFunction=&BreakOutFloat2Components}
  }
  MaterialExpressionMaterialFunctionCall_3 {
   class=MaterialExpressionMaterialFunctionCall
   inputs [Float2 {from=MaterialExpressionTextureCoordinate_0}]
   properties {materialFunction=&BreakOutFloat2Components}
  }
  MaterialExpressionMaterialFunctionCall_4 {
   class=MaterialExpressionMaterialFunctionCall
   inputs [
    UorV {from=MaterialExpressionMaterialFunctionCall_3, output=R}
    EdgeControl {from=MaterialExpressionVectorParameter_1}
   ]
   properties {materialFunction=/Game/Effects/Niagara/MasterMaterials/Functions/Common/FXCommon_EdgeMask_Simple}
  }
  MaterialExpressionMaterialFunctionCall_5 {
   class=MaterialExpressionMaterialFunctionCall
   inputs [
    NoiseValue {from=MaterialExpressionMaterialFunctionCall_1, output=R}
    Control_RGB {from=MaterialExpressionVectorParameter_5}
    Control_A {from=MaterialExpressionVectorParameter_5}
   ]
   properties {materialFunction=/Game/Effects/Niagara/MasterMaterials/Functions/Common/FXCommon_NoiseControl}
  }
  MaterialExpressionMaterialFunctionCall_6 {
   class=MaterialExpressionMaterialFunctionCall
   inputs [Emissive {from=MaterialExpressionDivide_2}]
   properties {materialFunction=&MF_FX_EmissiveAdjustments}
  }
  MaterialExpressionMaterialFunctionCall_7 {
   class=MaterialExpressionMaterialFunctionCall
   inputs [
    NoiseValue {from=MaterialExpressionMaterialFunctionCall_8, output=R}
    Control_RGB {from=MaterialExpressionVectorParameter_2}
    Control_A {from=MaterialExpressionVectorParameter_2}
   ]
   properties {materialFunction=&FXCommon_NoiseControl}
  }
  MaterialExpressionMaterialFunctionCall_8 {
   class=MaterialExpressionMaterialFunctionCall
   inputs [
    Texture {from=MaterialExpressionTextureObjectParameter_1}
    AdditionalUVOffset {from=MaterialExpressionMultiply_17}
    UVControl_RGB {from=MaterialExpressionVectorParameter_3}
    UVControl_A {from=MaterialExpressionVectorParameter_3}
    "[Override]TimeInput" {from=""}
    "[Override]UVInput" {from=""}
   ]
   properties {materialFunction=&FXCommon_TextureSample}
  }
  MaterialExpressionMaterialFunctionCall_9 {
   class=MaterialExpressionMaterialFunctionCall
   inputs [Float2 {from=MaterialExpressionMaterialFunctionCall_7, output=NoiseValue}]
   properties {materialFunction=&BreakOutFloat2Components}
  }
  MaterialExpressionMultiply_0 {
   class=MaterialExpressionMultiply
   inputs [
    A {from=MaterialExpressionMultiply_2}
    B {from=MaterialExpressionStaticSwitchParameter_1}
   ]
  }
  MaterialExpressionMultiply_1 {
   class=MaterialExpressionMultiply
   inputs [
    A {from=MaterialExpressionMaterialFunctionCall_9, output=R}
    B {from=MaterialExpressionDynamicParameter_1, output=SoftWaveStr_NoiseB}
   ]
   properties {constB=100}
  }
  MaterialExpressionMultiply_10 {
   class=MaterialExpressionMultiply
   inputs [
    A {from=MaterialExpressionDynamicParameter_2, output=SoftWaveStr}
    B {from=MaterialExpressionReroute_10}
   ]
  }
  MaterialExpressionMultiply_11 {
   class=MaterialExpressionMultiply
   inputs [
    A {from=MaterialExpressionAdd_8}
    B {from=MaterialExpressionMaterialFunctionCall_17, output=Result}
   ]
   properties {constB=2.919}
  }
  MaterialExpressionMultiply_12 {
   class=MaterialExpressionMultiply
   inputs [
    A {from=MaterialExpressionPower_1}
    B {from=MaterialExpressionMultiply_14}
   ]
  }
  MaterialExpressionMultiply_13 {
   class=MaterialExpressionMultiply
   inputs [
    A {from=MaterialExpressionMultiply_9}
    B {from=MaterialExpressionDynamicParameter_5, output=WPOStr}
   ]
  }
  MaterialExpressionMultiply_14 {
   class=MaterialExpressionMultiply
   inputs [
    A {from=MaterialExpressionScalarParameter_6}
    B {from=MaterialExpressionDynamicParameter_7, output=SpikeB}
   ]
  }
  MaterialExpressionMultiply_15 {
   class=MaterialExpressionMultiply
   inputs [
    A {from=MaterialExpressionPower_3}
    B {from=MaterialExpressionMultiply_18}
   ]
  }
  MaterialExpressionMultiply_16 {
   class=MaterialExpressionMultiply
   inputs [
    A {from=MaterialExpressionMultiply_6}
    B {from=MaterialExpressionDynamicParameter_8, output=WPOStr_NoiseB}
   ]
  }
  MaterialExpressionMultiply_17 {
   class=MaterialExpressionMultiply
   inputs [
    A {from=MaterialExpressionDynamicParameter_4, output=Time_NoiseB}
    B {from=MaterialExpressionAppendVector_1}
   ]
  }
  MaterialExpressionMultiply_18 {
   class=MaterialExpressionMultiply
   inputs [
    A {from=MaterialExpressionScalarParameter_9}
    B {from=MaterialExpressionDynamicParameter_8, output=SpikeB_NoiseB}
   ]
  }
  MaterialExpressionMultiply_19 {
   class=MaterialExpressionMultiply
   inputs [
    A {from=""}
    B {from=MaterialExpressionParticleRandom_0}
   ]
   properties {constA=7, constB=15}
  }
  MaterialExpressionMultiply_2 {
   class=MaterialExpressionMultiply
   inputs [
    A {from=MaterialExpressionVertexNormalWS_0}
    B {from=MaterialExpressionDynamicParameter_0, output=MeshSize}
   ]
  }
  MaterialExpressionMultiply_20 {
   class=MaterialExpressionMultiply
   inputs [
    A {from=MaterialExpressionScalarParameter_13}
    B {from=""}
   ]
   properties {constB=0.001}
  }
  MaterialExpressionMultiply_24 {
   class=MaterialExpressionMultiply
   inputs [
    A {from=MaterialExpressionTextureCoordinate_4}
    B {from=MaterialExpressionAppendVector_2}
   ]
  }
  MaterialExpressionMultiply_27 {
   class=MaterialExpressionMultiply
   inputs [
    A {from=MaterialExpressionReroute_4}
    B {from=MaterialExpressionScalarParameter_17}
   ]
  }
  MaterialExpressionMultiply_3 {
   class=MaterialExpressionMultiply
   inputs [
    A {from=MaterialExpressionDynamicParameter_6, output=Time}
    B {from=MaterialExpressionAppendVector_0}
   ]
  }
  MaterialExpressionMultiply_4 {
   class=MaterialExpressionMultiply
   inputs [
    A {from=MaterialExpressionParticleColor_0, output=A}
    B {from=MaterialExpressionSaturate_0}
   ]
  }
  MaterialExpressionMultiply_5 {
   class=MaterialExpressionMultiply
   inputs [
    A {from=MaterialExpressionLinearInterpolate_0}
    B {from=MaterialExpressionDynamicParameter_2, output=SoftWaveStr}
   ]
   properties {constB=100}
  }
  MaterialExpressionMultiply_6 {
   class=MaterialExpressionMultiply
   inputs [
    A {from=MaterialExpressionMultiply_1}
    B {from=MaterialExpressionSaturate_2}
   ]
  }
  MaterialExpressionMultiply_7 {
   class=MaterialExpressionMultiply
   inputs [
    A {from=MaterialExpressionDynamicParameter_1, output=SoftWaveStr_NoiseB}
    B {from=MaterialExpressionAbs_2}
   ]
  }
  MaterialExpressionMultiply_8 {
   class=MaterialExpressionMultiply
   inputs [
    A {from=MaterialExpressionAdd_5}
    B {from=MaterialExpressionScalarParameter_4}
   ]
  }
  MaterialExpressionMultiply_9 {
   class=MaterialExpressionMultiply
   inputs [
    A {from=MaterialExpressionMultiply_5}
    B {from=MaterialExpressionSaturate_1}
   ]
  }
  MaterialExpressionNamedRerouteDeclaration_0 {
   class=MaterialExpressionNamedRerouteDeclaration
   inputs [None {from=MaterialExpressionConvert_1}]
   properties {name=ChromaticNoise, nodeColor=<rgba 1 0.1875 0 1>}
  }
  MaterialExpressionNamedRerouteUsage_0 {class=MaterialExpressionNamedRerouteUsage}
  MaterialExpressionPanner_0 {
   class=MaterialExpressionPanner
   inputs [
    Coordinate {from=MaterialExpressionMultiply_24}
    Time {from=""}
    Speed {from=MaterialExpressionAppendVector_3}
   ]
  }
  MaterialExpressionParticleColor_0 {class=MaterialExpressionParticleColor, inputs [], properties {}}
  MaterialExpressionParticleRandom_0 {class=MaterialExpressionParticleRandom, inputs [], properties {}}
  MaterialExpressionParticleRandom_1 {class=MaterialExpressionParticleRandom, inputs [], properties {}}
  MaterialExpressionPower_0 {
   class=MaterialExpressionPower
   inputs [
    Base {from=MaterialExpressionMaterialFunctionCall_14, output=R}
    Exp {from=MaterialExpressionScalarParameter_5}
   ]
  }
  MaterialExpressionPower_1 {
   class=MaterialExpressionPower
   inputs [
    Base {from=MaterialExpressionDivide_0}
    Exp {from=MaterialExpressionScalarParameter_7}
   ]
  }
  MaterialExpressionPower_2 {
   class=MaterialExpressionPower
   inputs [
    Base {from=MaterialExpressionMaterialFunctionCall_10, output=R}
    Exp {from=MaterialExpressionScalarParameter_8}
   ]
  }
  MaterialExpressionPower_3 {
   class=MaterialExpressionPower
   inputs [
    Base {from=MaterialExpressionDivide_3}
    Exp {from=MaterialExpressionScalarParameter_10}
   ]
  }
  MaterialExpressionReroute_0 {class=MaterialExpressionReroute, inputs [None {from=MaterialExpressionMultiply_0}]}
  MaterialExpressionReroute_1 {class=MaterialExpressionReroute, inputs [None {from=MaterialExpressionMultiply_16}]}
  MaterialExpressionReroute_10 {class=MaterialExpressionReroute, inputs [None {from=MaterialExpressionReroute_9}]}
  MaterialExpressionReroute_2 {class=MaterialExpressionReroute, inputs [None {from=MaterialExpressionVectorParameter_0, output=VPan}]}
  MaterialExpressionReroute_4 {class=MaterialExpressionReroute, inputs [None {from=MaterialExpressionStaticSwitchParameter_0}]}
  MaterialExpressionReroute_5 {class=MaterialExpressionReroute, inputs [None {from=MaterialExpressionReroute_7}]}
  MaterialExpressionReroute_6 {class=MaterialExpressionReroute, inputs [None {from=MaterialExpressionTextureObjectParameter_0}]}
  MaterialExpressionReroute_7 {class=MaterialExpressionReroute, inputs [None {from=MaterialExpressionPanner_0}]}
  MaterialExpressionReroute_8 {class=MaterialExpressionReroute, inputs [None {from=MaterialExpressionPanner_0}]}
  MaterialExpressionReroute_9 {class=MaterialExpressionReroute, inputs [None {from=MaterialExpressionVectorParameter_5, output=UStrength}]}
  MaterialExpressionSaturate_0 {class=MaterialExpressionSaturate, inputs [None {from=MaterialExpressionMaterialFunctionCall_11, output=Result}]}
  MaterialExpressionSaturate_1 {class=MaterialExpressionSaturate, inputs [None {from=MaterialExpressionAdd_4}]}
  MaterialExpressionSaturate_2 {class=MaterialExpressionSaturate, inputs [None {from=MaterialExpressionAdd_2}]}
  MaterialExpressionScalarParameter_0 {class=MaterialExpressionScalarParameter, properties {defaultValue=2, parameterName=SW}}
  MaterialExpressionScalarParameter_1 {class=MaterialExpressionScalarParameter, properties {defaultValue=0.5, parameterName=NoiseMask_Contrast}}
  MaterialExpressionScalarParameter_10 {class=MaterialExpressionScalarParameter, properties {defaultValue=1.7, parameterName=NoiseB_BrightP}}
  MaterialExpressionScalarParameter_11 {class=MaterialExpressionScalarParameter, properties {parameterName=awaw}}
  MaterialExpressionScalarParameter_12 {class=MaterialExpressionScalarParameter, properties {parameterName=FadeDistance}}
  MaterialExpressionScalarParameter_13 {class=MaterialExpressionScalarParameter, properties {defaultValue=8, parameterName=Chromatic_Offset}}
  MaterialExpressionScalarParameter_14 {class=MaterialExpressionScalarParameter, properties {parameterName=ParticleRandomLerp}}
  MaterialExpressionScalarParameter_15 {class=MaterialExpressionScalarParameter, properties {parameterName=Chromatic_Lerp}}
  MaterialExpressionScalarParameter_16 {class=MaterialExpressionScalarParameter, properties {parameterName=Chromatic_Lerp}}
  MaterialExpressionScalarParameter_17 {class=MaterialExpressionScalarParameter, properties {defaultValue=50, parameterName=Chromatic_Intensity}}
  MaterialExpressionScalarParameter_18 {class=MaterialExpressionScalarParameter, properties {parameterName=ParticleRandomLerp}}
  MaterialExpressionScalarParameter_2 {class=MaterialExpressionScalarParameter, properties {defaultValue=0.5, parameterName=NoiseB_Mask_Contrast}}
  MaterialExpressionScalarParameter_3 {class=MaterialExpressionScalarParameter, properties {defaultValue=-0.5, group=Hue, parameterName=Hue_GrayAdjust_Add}}
  MaterialExpressionScalarParameter_4 {class=MaterialExpressionScalarParameter, properties {defaultValue=0.5, group=Hue, parameterName=Hue_GrayAdjust_Mul}}
  MaterialExpressionScalarParameter_5 {class=MaterialExpressionScalarParameter, properties {defaultValue=0.7, parameterName=NoiseMask_PPower}}
  MaterialExpressionScalarParameter_6 {class=MaterialExpressionScalarParameter, properties {defaultValue=6, parameterName=NoiseBrightD}}
  MaterialExpressionScalarParameter_7 {class=MaterialExpressionScalarParameter, properties {defaultValue=1.7, parameterName=NoiseBrightP}}
  MaterialExpressionScalarParameter_8 {class=MaterialExpressionScalarParameter, properties {defaultValue=0.7, parameterName=NoiseB_Mask_PPower}}
  MaterialExpressionScalarParameter_9 {class=MaterialExpressionScalarParameter, properties {defaultValue=6, parameterName=NoiseB_BrightD}}
  MaterialExpressionStaticSwitchParameter_0 {
   class=MaterialExpressionStaticSwitchParameter
   inputs [
    True {from=MaterialExpressionAdd_3, output=""}
    False {from=MaterialExpressionMultiply_12, output=""}
   ]
   properties {parameterName=UseNoiseB}
  }
  MaterialExpressionStaticSwitchParameter_1 {
   class=MaterialExpressionStaticSwitchParameter
   inputs [
    True {from=MaterialExpressionAdd_6, output=""}
    False {from=MaterialExpressionMultiply_13, output=""}
   ]
   properties {parameterName=UseNoiseB}
  }
  MaterialExpressionStaticSwitchParameter_2 {
   class=MaterialExpressionStaticSwitchParameter
   inputs [
    True {from=MaterialExpressionDepthFade_0, output=""}
    False {from=MaterialExpressionMultiply_4, output=""}
   ]
   properties {parameterName=UseDepthFade}
  }
  MaterialExpressionSubtract_0 {
   class=MaterialExpressionSubtract
   inputs [
    A {from=MaterialExpressionReroute_8}
    B {from=MaterialExpressionMultiply_20}
   ]
  }
  MaterialExpressionSubtract_1 {
   class=MaterialExpressionSubtract
   inputs [
    A {from=MaterialExpressionVectorParameter_0, output=VScale}
    B {from=MaterialExpressionMultiply_19}
   ]
  }
  MaterialExpressionTextureCoordinate_0 {class=MaterialExpressionTextureCoordinate}
  MaterialExpressionTextureCoordinate_4 {class=MaterialExpressionTextureCoordinate}
  MaterialExpressionTextureObjectParameter_0 {
   class=MaterialExpressionTextureObjectParameter
   inputs [
    UVs {from="", output=""}
    Apply View MipBias {from="", output=""}
   ]
   properties {
    group=Noise
    parameterName=Noise_Texture
    samplerType=SAMPLERTYPE_LinearGrayscale
    sortPriority=0
    texture=&T_SimplePerlin_Greyscale
   }
  }
  MaterialExpressionTextureObjectParameter_1 {
   class=MaterialExpressionTextureObjectParameter
   inputs [
    UVs {from="", output=""}
    Apply View MipBias {from="", output=""}
   ]
   properties {
    group=NoiseB
    parameterName=NoiseB_Texture
    samplerType=SAMPLERTYPE_LinearGrayscale
    sortPriority=0
    texture=&T_SimplePerlin_Greyscale
   }
  }
  MaterialExpressionTextureObjectParameter_2 {
   class=MaterialExpressionTextureObjectParameter
   inputs [
    UVs {from="", output=""}
    Apply View MipBias {from="", output=""}
   ]
   properties {
    group=Hue
    parameterName=Hue_Texture
    samplerType=SAMPLERTYPE_LinearGrayscale
    sortPriority=0
    texture=&T_SimplePerlin_Greyscale
   }
  }
  MaterialExpressionTextureObjectParameter_3 {
   class=MaterialExpressionTextureObjectParameter
   inputs [
    UVs {from="", output=""}
    Apply View MipBias {from="", output=""}
   ]
   properties {
    group=NoiseMask
    parameterName=NoiseMask_Texture
    sortPriority=0
    texture=/Game/Effects/Fort_Effects/Emotes/UltraLight/Texture/T_SoftNoise_BC7
   }
  }
  MaterialExpressionTextureObjectParameter_4 {
   class=MaterialExpressionTextureObjectParameter
   inputs [
    UVs {from="", output=""}
    Apply View MipBias {from="", output=""}
   ]
   properties {
    group=NoiseBMask
    parameterName=NoiseB_Mask_Texture
    sortPriority=0
    texture=&T_SoftNoise_BC7
   }
  }
  MaterialExpressionTextureSample_0 {
   class=MaterialExpressionTextureSample
   inputs [
    UVs {from=""}
    Tex {from=""}
    Apply View MipBias {from=""}
   ]
  }
  MaterialExpressionTextureSample_1 {
   class=MaterialExpressionTextureSample
   inputs [
    UVs {from=MaterialExpressionAdd_10}
    Tex {from=MaterialExpressionReroute_6}
    Apply View MipBias {from=""}
   ]
  }
  MaterialExpressionTextureSample_2 {
   class=MaterialExpressionTextureSample
   inputs [
    UVs {from=MaterialExpressionSubtract_0}
    Tex {from=MaterialExpressionReroute_6}
    Apply View MipBias {from=""}
   ]
  }
  MaterialExpressionTextureSample_3 {
   class=MaterialExpressionTextureSample
   inputs [
    UVs {from=MaterialExpressionReroute_7}
    Tex {from=MaterialExpressionReroute_6}
    Apply View MipBias {from=""}
   ]
  }
  MaterialExpressionTransformPosition_0 {
   class=MaterialExpressionTransformPosition
   inputs [None {from=MaterialExpressionWorldPosition_1, output=XYZ}]
   properties {transformSourceType=TRANSFORMPOSSOURCE_World, transformType=TRANSFORMPOSSOURCE_Instance}
  }
  MaterialExpressionVectorParameter_0 {
   class=MaterialExpressionVectorParameter
   properties {group=Noise, parameterName=Noise_UVControl, channelNames {a=VPan, b=UPan, g=VScale, r=UScale}, defaultValue=<rgba 0.01 1 0 0>}
  }
  MaterialExpressionVectorParameter_1 {
   class=MaterialExpressionVectorParameter
   properties {group=E, parameterName=E_EdgeMask_Control, channelNames {a="", b=Power, g=Multiply, r=Offset}, defaultValue=<rgba 0 2 2 1>}
  }
  MaterialExpressionVectorParameter_10 {class=MaterialExpressionVectorParameter, properties {group=NoiseB, parameterName=NoiseB_UVPanSpeed, defaultValue=<rgba 60 -0.6 0 1>}}
  MaterialExpressionVectorParameter_11 {
   class=MaterialExpressionVectorParameter
   properties {group=Noise, parameterName=Noise_UVControl, channelNames {a=VPan, b=UPan, g=VScale, r=UScale}, defaultValue=<rgba 0.01 1 0 0>}
  }
  MaterialExpressionVectorParameter_2 {
   class=MaterialExpressionVectorParameter
   properties {
    group=NoiseB
    parameterName=NoiseB_NoiseControl
    sortPriority=60
    channelNames {a=VStrength, b=UStrength, g=VOffset, r=UOffset}
    defaultValue=<rgba -0.5 -0.5 20 0>
   }
  }
  MaterialExpressionVectorParameter_3 {
   class=MaterialExpressionVectorParameter
   properties {group=NoiseB, parameterName=NoiseB_UVControl, channelNames {a=VPan, b=UPan, g=VScale, r=UScale}, defaultValue=<rgba 0.01 1 0 0>}
  }
  MaterialExpressionVectorParameter_4 {class=MaterialExpressionVectorParameter, properties {parameterName=RF, defaultValue=<rgba 25 10 1 0.5>}}
  MaterialExpressionVectorParameter_5 {
   class=MaterialExpressionVectorParameter
   properties {
    group=Noise
    parameterName=Noise_NoiseControl
    sortPriority=60
    channelNames {a=VStrength, b=UStrength, g=VOffset, r=UOffset}
    defaultValue=<rgba -0.5 -0.5 20 0>
   }
  }
  MaterialExpressionVectorParameter_6 {
   class=MaterialExpressionVectorParameter
   properties {group=Hue, parameterName=Hue_UVControl, channelNames {a=VPan, b=UPan, g=VScale, r=UScale}, defaultValue=<rgba 0.0005 0.0005 0 300>}
  }
  MaterialExpressionVectorParameter_7 {
   class=MaterialExpressionVectorParameter
   properties {group=NoiseMask, parameterName=NoiseMask_UVControl, channelNames {a=VPan, b=UPan, g=VScale, r=UScale}, defaultValue=<rgba 0.01 1 30 0.1>}
  }
  MaterialExpressionVectorParameter_8 {class=MaterialExpressionVectorParameter, properties {group=Noise, parameterName=NoiseUVPanSpeed, defaultValue=<rgba 60 -0.6 0 1>}}
  MaterialExpressionVectorParameter_9 {
   class=MaterialExpressionVectorParameter
   properties {group=NoiseBMask, parameterName=NoiseB_Mask_UVControl, channelNames {a=VPan, b=UPan, g=VScale, r=UScale}, defaultValue=<rgba 0.01 1 30 0.1>}
  }
  MaterialExpressionVertexNormalWS_0 {class=MaterialExpressionVertexNormalWS, inputs [], properties {}}
  MaterialExpressionWorldPosition_0 {class=MaterialExpressionWorldPosition, inputs [], properties {worldPositionShaderOffset=WPT_ExcludeAllShaderOffsets}}
  MaterialExpressionWorldPosition_1 {class=MaterialExpressionWorldPosition, inputs [], properties {}}
 }
 properties {
  bUsedWithNiagaraMeshParticles=true
  blendMode=BLEND_Translucent
  pixelDepthOffsetMode=PDOM_Legacy
  shadingModel=MSM_Unlit
  twoSided=true
  usedShadingModels=Unlit
 }
}
