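# Parent material for the TechnoAwaken_Shake player body, adding animated, audio-reactive stripes between the shared character PreFX and PostFX functions.
# A mask map and Fresnel build rotated, panning stripe UVs that wobble with audio.
# Audio spectrum and amplitude, blended with a time-cycled hue SphereMask, scale stripe color into Base Color and Emissive within the Tech Art mask.
# Static switches add soft-stripe alternates and a Fresnel-sampled thin-film glow boosted by audio.
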Material /Game/Characters/Player/Male/Medium/Bodies/M_MED_TechnoAwaken_Shake/Materials/M_TechnoAwaken_Shake_Parent
graph {
 properties {
  bHasPixelAnimation=true
  bUseMaterialAttributes=true
  bUsedWithMeshDeformer=true
  bUsedWithMorphTargets=true
  bUsedWithNanite=true
  bUsedWithSkeletalMesh=true
  pixelDepthOffsetMode=PDOM_Legacy
  shadingModel=MSM_FromMaterialExpression
  usedShadingModels=Default Lit | Subsurface | Clear Coat | Subsurface Profile | Cloth
 }
 nodes [
  n1 = MaterialFunctionCall() {materialFunction=&MF_CharacterParent_2023_PreFX}
  n2 = GetMaterialAttributes(None=@n1.Result) {attributeGetTypes [#1, #2, #3]}
  n3 = ComponentMask(None=@n1.Tech Art Mask) {b=true}
  n4 = ScalarParameter() {defaultValue=0.3, group=TechArt, parameterName=Audio_Amplitude_Scale_Off}
  n5 = ScalarParameter() {defaultValue=16, group=TechArt, parameterName=Audio_Amplitude_Scale}
  n6 = TextureSampleParameter2D() {
   group=TechArt
   parameterName=CustomCoordsMap
   samplerSource=SSM_Wrap_WorldGroupSettings
   samplerType=SAMPLERTYPE_LinearColor
   texture=/Game/Characters/Player/Male/Medium/Bodies/M_MED_TechnoAwaken_Shake/Textures/T_TechnoAwaken_Shake_FX_Body_TechArtFX_Mask
  }
  n7 = StaticSwitchParameter(
   True=LinearInterpolate(A=@n5, B=ScalarParameter() {defaultValue=50, group=TechArt, parameterName=Audio_Amplitude_Scale_SoftStripes}, Alpha=@n6.B)
   False=@n5
  ) {desc="R = Magenta\r\nG = Purple\r\nB = light blue", group=TechArt, parameterName=EnableSoftStripes}
  n8 = MaterialFunctionCall() {materialFunction=/Game/Packages/Fortress_SharedMaterials/Base_Material_Functions/BaseCharacterShaderFunctions/Audio/MF_AudioReactivity}
  n9 = Multiply(A=@n8.Amplitude, B=NamedRerouteUsage())
  n10 = Multiply(A=@n8.Spectrum, B=NamedRerouteUsage())
  n11 = MaterialFunctionCall(Float3=ComponentMask(None=@n10) {b=true, g=true, r=true}) {materialFunction=&BreakOutFloat3Components}
  n12 = Fresnel(ExponentIn=ScalarParameter() {defaultValue=2, group=TechArt, parameterName=UV Fresnel Exponent}) {baseReflectFraction=0}
  n13 = ScalarParameter() {defaultValue=0.05, group=TechArt, parameterName=UV Fresnel bias}
  n14 = LinearInterpolate(
   A=@n6.R
   B=Saturate(None=Subtract(A=@n12, B=@n13))
   Alpha=ScalarParameter() {defaultValue=0.256, group=TechArt, parameterName=UV Fresnel Blend, sliderMax=1}
  ) {constAlpha=0.378}
  n15 = Rotator(
   Coordinate=AppendVector(A=LinearInterpolate(A=@n14, B=@n6.R, Alpha=@n6.R), B=@n6.G)
   Time=ScalarParameter() {defaultValue=-0.05, group=TechArt, parameterName=UV Rotation, sliderMax=1}
  ) {centerX=0, centerY=0, speed=6.283}
  n16 = AppendVector(A=ScalarParameter() {defaultValue=3, group=TechArt, parameterName=ScaleX}, B=Constant() {r=1})
  n17 = Multiply(
   A=Power(Base=NamedRerouteUsage(), Exp=ScalarParameter() {defaultValue=1.5, group=TechArt, parameterName=Audio_Amplitude_Wobble_Exp}) {constExponent=3}
   B=ScalarParameter() {defaultValue=-0.08, group=TechArt, parameterName=WobbleAmplitude}
  )
  n18 = Constant()
  n19 = ScalarParameter() {defaultValue=-0.08, group=TechArt, parameterName=PanX}
  n20 = ScalarParameter() {defaultValue=-0.08, group=TechArt, parameterName=PanX_SoftStripes}
  n21 = Panner(
   Coordinate=Add(A=Multiply(A=@n15, B=@n16) {constB=2}, B=AppendVector(A=@n17, B=@n18))
   Speed=AppendVector(A=LinearInterpolate(A=@n19, B=@n20, Alpha=@n6.B), B=ScalarParameter() {group=TechArt, parameterName=PanY})
  ) {speedX=0.05}
  n22 = TextureSampleParameter2D(UVs=@n21) {
   group=TechArt
   parameterName=Stripe Mask
   samplerType=SAMPLERTYPE_Masks
   texture=/Game/Characters/Player/Male/Medium/Bodies/M_MED_TechnoAwaken_Shake/Textures/T_TechnoAwaken_Shake_FX_M
  }
  n23 = TextureSampleParameter2D(UVs=@n21) {
   group=TechArt
   parameterName=Stripe Mask Alt
   samplerType=SAMPLERTYPE_Masks
   texture=/Game/Characters/Player/Male/Medium/Bodies/M_MED_TechnoAwaken_Shake/Textures/T_TechnoAwaken_Shake_FX_2_M
  }
  n24 = StaticSwitchParameter(True=LinearInterpolate(A=@n22.RGB, B=@n23.RGB, Alpha=@n6.B), False=@n22.RGB) {desc="R = Magenta\r\nG = Purple\r\nB = light blue", group=TechArt, parameterName=EnableSoftStripes}
  n25 = MaterialFunctionCall(f3=Multiply(
    A=MaterialFunctionCall(X=@n11.G, Y=@n11.G, Z=@n11.G) {materialFunction=&MakeFloat3}.Result
    B=@n24
   )) {materialFunction=&AddComponents}
  n26 = MaterialFunctionCall(
   Hue Shift Percentage=Multiply(A=Time(), B=ScalarParameter() {defaultValue=0.375, group=TechArt, parameterName=CycleSpeed})
   Texture=Constant3Vector() {constant=<rgba 1 0 0 0>}
  ) {materialFunction=&HueShift}
  n27 = ScalarParameter() {defaultValue=1, group=TechArt, parameterName=Cycle Range, sliderMax=2}
  n28 = LinearInterpolate(
   A=Power(Base=Multiply(A=@n9, B=@n25.f3), Exp=ScalarParameter() {defaultValue=8, group=TechArt, parameterName=Audio_Amplitude_Exp}) {constExponent=3}
   B=Multiply(
    A=MaterialFunctionCall(f3=@n24) {materialFunction=&AddComponents}.f3
    B=SphereMask(A=@n24, B=@n26.Result, Radius=@n27) {attenuationRadius=1, hardnessPercent=0}
   )
   Alpha=ScalarParameter() {defaultValue=0.0125, group=TechArt, parameterName=Reactive/Cyclical Blend, sliderMax=1}
  )
  n29 = TextureSampleParameter2D(UVs=@n21) {
   group=TechArt
   parameterName=Stripes
   texture=/Game/Characters/Player/Male/Medium/Bodies/M_MED_TechnoAwaken_Shake/Textures/T_TechnoAwaken_Shake_FX_A
  }
  n30 = TextureSampleParameter2D(UVs=@n21) {
   group=TechArt
   parameterName=Stripes Alt
   texture=/Game/Characters/Player/Male/Medium/Bodies/M_MED_TechnoAwaken_Shake/Textures/T_TechnoAwaken_Shake_FX_2_A
  }
  n31 = LinearInterpolate(A=@n29.RGB, B=@n30.RGB, Alpha=@n6.B)
  n32 = Multiply(
   A=@n3
   B=Multiply(
    A=LinearInterpolate(A=@n4, B=@n7, Alpha=@n28) {constA=0.1}
    B=StaticSwitchParameter(True=@n31, False=@n29.RGB) {desc="R = Magenta\r\nG = Purple\r\nB = light blue", group=TechArt, parameterName=EnableSoftStripes}
   )
  )
  n33 = BlendMaterialAttributes(
   A=@n2.MaterialAttributes
   B=SetMaterialAttributes!(MaterialAttributes=@n2.MaterialAttributes, Base Color=@n32, Emissive Color=Add(A=@n2.Emissive Color, B=@n32), Shading Model=ShadingModel()) {attributeSetTypes [#1, #2, #3]}
   Alpha=@n3
  )
  n34 = GetMaterialAttributes(None=@n33) {attributeGetTypes [#2]}
  n35 = Fresnel() {baseReflectFraction=0, exponent=2}
  n36 = TextureSample!(UVs=@n35, Tex, Level, Apply View MipBias) {
   constMipValue=7
   mipValueMode=TMVM_MipLevel
   samplerSource=SSM_Wrap_WorldGroupSettings
   texture=/Game/Characters/Player/Male/Medium/Bodies/M_MED_TechnoAwaken_Shake/Textures/T_TechnoAwaken_Shake_FX_ThinFilm
  }
  n37 = Multiply(A=@n35, B=Power(Base=@n36.RGB))
  n38 = LinearInterpolate(B=ScalarParameter() {defaultValue=10, group=TechArt, parameterName=ThinFilmAmplitudeBoost}, Alpha=NamedRerouteUsage()) {constA=1, constB=3}
  n39 = TextureSampleParameter2D!(UVs, Level, Apply View MipBias) {
   group=TechArt
   mipValueMode=TMVM_MipLevel
   parameterName=CustomCoordsMap
   samplerSource=SSM_Wrap_WorldGroupSettings
   samplerType=SAMPLERTYPE_LinearColor
   texture=&T_TechnoAwaken_Shake_FX_Body_TechArtFX_Mask
  }
  n40 = LinearInterpolate(A=@n34.Emissive Color, B=Add(A=@n34.Emissive Color, B=Multiply(A=@n37, B=@n38)) {constB=0}, Alpha=@n39.A)
  MaterialFunctionCall(
   Material=StaticSwitchParameter(
    True=SetMaterialAttributes!(MaterialAttributes=@n34.MaterialAttributes, Emissive Color=@n40) {attributeSetTypes [#2]}
    False=@n33
   ) {group=TechArt, parameterName=Enable CustomThinFilm}
   M=@n1.M
   Pre Skinned Local Position=@n1.M
   Distance Blend=@n1.M
   Fresnel=@n1.M
  ) {materialFunction=&MF_CharacterParent_2023_PostFX}
  NamedRerouteDeclaration(None=ScalarParameter() {defaultValue=1, parameterName=EnableAudioReactivity}) {name=ReactivityOn, nodeColor=<rgba 0.1719 0 1 1>}
  NamedRerouteDeclaration(None=MaterialFunctionCall(f2=ComponentMask(None=@n10) {g=true, r=true}) {materialFunction=&AddComponents}.f2) {name=Audio_Amplitude, nodeColor=<rgba 0.0547 0 1 1>}
 ]
}
