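# Corrective Control Rig for the TechnoAwaken Shake medium male player body, adjusting its knot and shoulder pad dynamic bones as the upper body moves.
# Parent constraints blend those bones across spine, clavicle and arm bones.
# Remapped bone rotation deltas and spherical pose readers then weight additional transform offsets, some smoothed with AlphaInterp.
# Gameplay state flags such as skydiving, emote, melee, pistol, crouch and DBNO gate or scale each correction.
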ControlRigBlueprint /Game/Characters/Player/Male/Medium/Bodies/M_MED_TechnoAwaken_Shake/Meshes/M_MED_TechnoAwaken_Shake_Corrective_CtrlRig
graphs [
 Multiply_11_2_ContainedGraph {nodes [
   n1 = RigVMFunctionEntryNode()
   RigVMFunctionReturnNode(Result=&RigVMFunction_MathDoubleMul(B=@n1.C, A=&RigVMFunction_MathDoubleMul(A=@n1.A, B=@n1.B).Result).Result)
  ]}
 Multiply_11_2_ContainedGraph {nodes [
   n1 = RigVMFunctionEntryNode()
   RigVMFunctionReturnNode(Result=&RigVMFunction_MathDoubleMul(B=@n1.C, A=&RigVMFunction_MathDoubleMul(A=@n1.A, B=@n1.B).Result).Result)
  ]}
 Multiply_12_1_ContainedGraph {nodes [
   n1 = RigVMFunctionEntryNode()
   RigVMFunctionReturnNode(Result=&RigVMFunction_MathDoubleMul(B=@n1.C, A=&RigVMFunction_MathDoubleMul(A=@n1.A, B=@n1.B).Result).Result)
  ]}
 Multiply_12_1_ContainedGraph {nodes [
   n1 = RigVMFunctionEntryNode()
   RigVMFunctionReturnNode(Result=&RigVMFunction_MathDoubleMul(B=@n1.C, A=&RigVMFunction_MathDoubleMul(A=@n1.A, B=@n1.B).Result).Result)
  ]}
 Multiply_13_3_2_1_ContainedGraph {nodes [
   n1 = RigVMFunctionEntryNode()
   RigVMFunctionReturnNode(Result=&RigVMFunction_MathDoubleMul(B=@n1.C, A=&RigVMFunction_MathDoubleMul(A=@n1.A, B=@n1.B).Result).Result)
  ]}
 Multiply_2_1_ContainedGraph {nodes [
   n1 = RigVMFunctionEntryNode()
   RigVMFunctionReturnNode(Result=&RigVMFunction_MathDoubleMul(B=@n1.C, A=&RigVMFunction_MathDoubleMul(A=@n1.A, B=@n1.B).Result).Result)
  ]}
 Multiply_2_1_ContainedGraph {nodes [
   n1 = RigVMFunctionEntryNode()
   RigVMFunctionReturnNode(Result=&RigVMFunction_MathDoubleMul(B=@n1.C, A=&RigVMFunction_MathDoubleMul(A=@n1.A, B=@n1.B).Result).Result)
  ]}
 RigVMFunctionLibrary {links [], nodes []}
 RigVMFunction_MathBoolAnd_2_1_ContainedGraph {nodes [
   n1 = RigVMFunctionEntryNode()
   RigVMFunctionReturnNode(Result=&RigVMFunction_MathBoolAnd(B=@n1.C, A=&RigVMFunction_MathBoolAnd(A=@n1.A, B=@n1.B).Result).Result)
  ]}
 RigVMFunction_MathBoolAnd_2_1_ContainedGraph {nodes [
   n1 = RigVMFunctionEntryNode()
   RigVMFunctionReturnNode(Result=&RigVMFunction_MathBoolAnd(B=@n1.C, A=&RigVMFunction_MathBoolAnd(A=@n1.A, B=@n1.B).Result).Result)
  ]}
 RigVMFunction_MathBoolAnd_4_ContainedGraph {nodes [
   n1 = RigVMFunctionEntryNode()
   RigVMFunction_MathBoolAnd = &RigVMFunction_MathBoolAnd(A=@n1.A, B=@n1.B)
   RigVMFunctionReturnNode(Result=&RigVMFunction_MathBoolAnd(B=@n1.C, A=@RigVMFunction_MathBoolAnd.Result).Result)
  ]}
 RigVMFunction_Sequence_1_1_ContainedGraph {nodes [
   n1 = RigVMFunctionEntryNode()
   n2 = &RigVMFunction_Sequence(ExecuteContext=@n1.ExecuteContext)
   n3 = &RigVMFunction_Sequence(ExecuteContext=@n2.B)
   n4 = &RigVMFunction_Sequence(ExecuteContext=@n3.B)
   RigVMFunctionReturnNode(A=@n2.A, B=@n3.A, C=@n4.A, D=@n4.B)
  ]}
 RigVMFunction_Sequence_1_1_ContainedGraph {nodes [
   n1 = RigVMFunctionEntryNode()
   n2 = &RigVMFunction_Sequence(ExecuteContext=@n1.ExecuteContext)
   n3 = &RigVMFunction_Sequence(ExecuteContext=@n2.B)
   n4 = &RigVMFunction_Sequence(ExecuteContext=@n3.B)
   RigVMFunctionReturnNode(A=@n2.A, B=@n3.A, C=@n4.A, D=@n4.B)
  ]}
 RigVMFunction_Sequence_1_2_ContainedGraph {nodes [
   n1 = RigVMFunctionEntryNode()
   n2 = &RigVMFunction_Sequence(ExecuteContext=@n1.ExecuteContext)
   n3 = &RigVMFunction_Sequence(ExecuteContext=@n2.B)
   n4 = &RigVMFunction_Sequence(ExecuteContext=@n3.B)
   RigVMFunctionReturnNode(A=@n2.A, B=@n3.A, C=@n4.A, D=@n4.B)
  ]}
 RigVMFunction_Sequence_1_2_ContainedGraph {nodes [
   n1 = RigVMFunctionEntryNode()
   n2 = &RigVMFunction_Sequence(ExecuteContext=@n1.ExecuteContext)
   n3 = &RigVMFunction_Sequence(ExecuteContext=@n2.B)
   n4 = &RigVMFunction_Sequence(ExecuteContext=@n3.B)
   RigVMFunctionReturnNode(A=@n2.A, B=@n3.A, C=@n4.A, D=@n4.B)
  ]}
 RigVMFunction_Sequence_4_1_ContainedGraph {nodes [
   n1 = RigVMFunctionEntryNode()
   n2 = &RigVMFunction_Sequence(ExecuteContext=@n1.ExecuteContext)
   n3 = &RigVMFunction_Sequence(ExecuteContext=@n2.B)
   n4 = &RigVMFunction_Sequence(ExecuteContext=@n3.B)
   n5 = &RigVMFunction_Sequence(ExecuteContext=@n4.B)
   RigVMFunctionReturnNode(A=@n2.A, B=@n3.A, C=@n4.A, D=@n5.A, E=@n5.B)
  ]}
 RigVMFunction_Sequence_5_1_ContainedGraph {nodes [
   n1 = RigVMFunctionEntryNode()
   n2 = &RigVMFunction_Sequence(ExecuteContext=@n1.ExecuteContext)
   n3 = &RigVMFunction_Sequence(ExecuteContext=@n2.B)
   RigVMFunctionReturnNode(A=@n2.A, B=@n3.A, C=@n3.B)
  ]}
 RigVMFunction_Sequence_7_1_ContainedGraph {nodes [
   n1 = RigVMFunctionEntryNode()
   n2 = &RigVMFunction_Sequence(ExecuteContext=@n1.ExecuteContext)
   n3 = &RigVMFunction_Sequence(ExecuteContext=@n2.B)
   n4 = &RigVMFunction_Sequence(ExecuteContext=@n3.B)
   RigVMFunctionReturnNode(A=@n2.A, B=@n3.A, C=@n4.A, D=@n4.B)
  ]}
 RigVMModel {nodes [
   RigUnit_BeginExecution = &RigUnit_BeginExecution()
   RigVMFunction_Sequence_7 = RigVMAggregateNode(ExecuteContext=@RigUnit_BeginExecution.ExecutePin)
   &RigUnit_ParentConstraint(
    AdvancedSettings=(InterpolationType=Average,RotationOrderForFilter=XZY)
    Child=(Type=Bone,Name="dyn_knot_l")
    Parents="((Item=(Type=Bone,Name=\"spine_03\"),Weight=0.0443),(Item=(Type=Bone,Name=\"spine_04\"),Weight=0.074),(Item=(Type=Bone,Name=\"spine_05\"),Weight=0.4344),(Item=(Type=Bone,Name=\"neck_02\"),Weight=0),(Item=(Type=Bone,Name=\"clavicle_l\"),Weight=0.3388),(Item=(Type=Bon ...(truncated)"
    Weight=1
    bMaintainOffset=true
    ExecutePin=@RigVMFunction_Sequence_7.A
   )
   > &RigUnit_ParentConstraint(
    AdvancedSettings=(InterpolationType=Average,RotationOrderForFilter=XZY)
    Child=(Type=Bone,Name="dyn_knot_r")
    Parents="((Item=(Type=Bone,Name=\"spine_03\"),Weight=0.0443),(Item=(Type=Bone,Name=\"spine_04\"),Weight=0.074),(Item=(Type=Bone,Name=\"spine_05\"),Weight=0.4344),(Item=(Type=Bone,Name=\"neck_02\"),Weight=0),(Item=(Type=Bone,Name=\"clavicle_r\"),Weight=0.3388),(Item=(Type=Bon ...(truncated)"
    Weight=1
    bMaintainOffset=true
   )
   > n1 = &RigUnit_SphericalPoseReader(
    ActiveRegionScaleFactors=(PositiveWidth=1,NegativeWidth=1,PositiveHeight=1,NegativeHeight=1)
    ActiveRegionSize=0.2
    Debug=(bDrawDebug=false,bDraw2D=False,bDrawLocalAxes=False,DebugScale=25,DebugSegments=20,DebugThickness=0.25)
    DriverAxis=(X=1,Y=0,Z=0)
    DriverItem=(Type=Bone,Name="upperarm_l")
    FalloffRegionScaleFactors=(PositiveWidth=1,NegativeWidth=1,PositiveHeight=1,NegativeHeight=1)
    FalloffSize=0.45
    FlipHeightScaling=False
    FlipWidthScaling=False
    OptionalParentItem=(Type=Bone,Name="spine_05")
    RotationOffset=(X=150,Y=-140,Z=0)
   )
   > Get Delta Ref_1_3_1 = RigVMFunctionReferenceNode(Item=(Type=Bone,Name="lowerarm_l"))
   Multiply_13_3_2_2_3_1 = RigVMAggregateNode(
    C=&RigVMFunction_MathDoubleMul(A=RigVMVariableNode() {variable=isPistol}.Value, B=RigVMVariableNode() {variable=isCrouching}.Value).Result
    B=&RigVMFunction_MathDoubleRemap(SourceMaximum=70, SourceMinimum=68, TargetMaximum=0, TargetMinimum=1, Value=@"Get Delta Ref_1_3_1".Result.Rotation.Yaw).Result
    A=@n1.OutputParam
   )
   n2 = &RigVMFunction_AlphaInterp(InterpSpeedDecreasing=10, InterpSpeedIncreasing=10, Scale=1, bClampResult=False, bInterpResult=False, Value=@Multiply_13_3_2_2_3_1.Result)
   > &RigUnit_ModifyTransforms(
    ItemToModify=((Item=(Type=Bone,Name="dyn_knot_l"),Transform=(Rotation=(X=0,Y=0,Z=0,W=1),Translation=(X=0,Y=1,Z=-1),Scale3D=(X=1,Y=1,Z=1))))
    WeightMinimum=0
    Weight=@n2.Result
   )
   &RigUnit_ParentConstraint(
    AdvancedSettings=(InterpolationType=Average,RotationOrderForFilter=XZY)
    Child=(Type=Bone,Name="dyn_pad_base_l")
    Filter=(TranslationFilter=(bX=True,bY=True,bZ=True),RotationFilter=(bX=true,bY=true,bZ=True),ScaleFilter=(bX=True,bY=True,bZ=True))
    Parents=((Item=(Type=Bone,Name="clavicle_l"),Weight=0.3),(Item=(Type=Bone,Name="dyn_pad_base_l"),Weight=0),(Item=(Type=Bone,Name="upperarm_l"),Weight=0.5),(Item=(Type=Bone,Name="upperarm_twist_01_l"),Weight=1))
    Weight=1
    ExecutePin=@RigVMFunction_Sequence_7.B
   )
   > &RigUnit_ParentConstraint(
    AdvancedSettings=(InterpolationType=Average,RotationOrderForFilter=XYZ)
    Child=(Type=Bone,Name="dyn_pad_base_r")
    Filter=(TranslationFilter=(bX=true,bY=true,bZ=true),RotationFilter=(bX=true,bY=true,bZ=True),ScaleFilter=(bX=True,bY=True,bZ=True))
    Parents=((Item=(Type=Bone,Name="clavicle_r"),Weight=0.3),(Item=(Type=Bone,Name="upperarm_r"),Weight=0.5),(Item=(Type=Bone,Name="upperarm_twist_01_r"),Weight=0.5))
    Weight=1
    bMaintainOffset=true
   )
   RigVMFunction_Sequence = &RigVMFunction_Sequence(ExecuteContext=@RigVMFunction_Sequence_7.C)
   DISPATCH_RigVMDispatch_CoreEquals = RigVMDispatchNode(B=1, A=RigVMVariableNode() {variable=isSkyDiving_NOT}.Value) {template="DISPATCH_RigVMDispatch_CoreEquals(in A,in B,out Result)"}
   RigVMFunction_MathBoolAnd = RigVMAggregateNode(
    A=@DISPATCH_RigVMDispatch_CoreEquals.Result
    B=RigVMDispatchNode(B=1, A=RigVMVariableNode() {variable=isCrouching_NOT}.Value) {template="DISPATCH_RigVMDispatch_CoreEquals(in A,in B,out Result)"}.Result
    C=RigVMDispatchNode(B=0, A=RigVMVariableNode() {variable=isDBNO}.Value) {template="DISPATCH_RigVMDispatch_CoreEquals(in A,in B,out Result)"}.Result
   )
   RigVMFunction_ControlFlowBranch = &RigVMFunction_ControlFlowBranch(Condition=@RigVMFunction_MathBoolAnd.Result, ExecuteContext=@RigVMFunction_Sequence.A)
   RigVMFunction_Sequence_1 = RigVMAggregateNode(ExecuteContext=@RigVMFunction_ControlFlowBranch.True)
   Get Delta Ref_5 = RigVMFunctionReferenceNode(Item=(Type=Bone,Name="upperarm_l"), ExecuteContext=@RigVMFunction_Sequence_1.A)
   n3 = &RigVMFunction_Sequence(ExecuteContext=@"Get Delta Ref_5".ExecuteContext)
   &RigUnit_ModifyTransforms(
    ItemToModify=((Item=(Type=Bone,Name="dyn_pad_l"),Transform=(Rotation=(X=0,Y=0.0872,Z=0,W=0.9962),Translation=(X=3,Y=0,Z=4),Scale3D=(X=1,Y=1.05,Z=1))))
    WeightMinimum=0
    Weight=&RigVMFunction_MathDoubleRemap(SourceMaximum=75, SourceMinimum=0, TargetMaximum=1, TargetMinimum=0, Value=@"Get Delta Ref_5".Result.Rotation.Pitch).Result
    ExecutePin=@n3.A
   )
   n4 = &RigVMFunction_MathDoubleRemap(SourceMaximum=0, SourceMinimum=-40, TargetMaximum=0, TargetMinimum=1, Value=@"Get Delta Ref_5".Result.Rotation.Pitch)
   Multiply = &RigVMFunction_MathDoubleMul(A=@n4.Result, B=RigVMVariableNode() {variable=isMelee_NOT}.Value)
   > &RigUnit_ModifyTransforms(
    ItemToModify=((Item=(Type=Bone,Name="dyn_pad_l"),Transform=(Rotation=(X=-0.0941,Y=-0.0789,Z=-0.0941,W=0.988),Translation=(X=2,Y=1,Z=0),Scale3D=(X=1,Y=1.05,Z=1))))
    WeightMinimum=0
    Weight=@Multiply.Result
   )
   > &RigUnit_ModifyTransforms(
    ItemToModify=((Item=(Type=Bone,Name="dyn_pad_l"),Transform=(Rotation=(X=0.122,Y=-0.0974,Z=-0.0747,W=0.9849),Translation=(X=2,Y=0,Z=1),Scale3D=(X=1,Y=1.05,Z=1))))
    WeightMinimum=0
    Weight=&RigVMFunction_MathDoubleMul(A=@n4.Result, B=RigVMVariableNode() {variable=isMelee}.Value).Result
   )
   &RigUnit_ModifyTransforms(
    ItemToModify=((Item=(Type=Bone,Name="dyn_pad_l"),Transform=(Rotation=(X=-0.5,Y=0,Z=0,W=0.866),Translation=(X=0,Y=0.5,Z=-1),Scale3D=(X=1,Y=1,Z=1))))
    WeightMinimum=0
    Weight=&RigVMFunction_MathDoubleRemap(SourceMaximum=-68, SourceMinimum=0, TargetMaximum=1, TargetMinimum=0, Value=@"Get Delta Ref_5".Result.Rotation.Yaw).Result
    ExecutePin=@n3.B
   )
   n5 = &RigVMFunction_MathDoubleRemap(SourceMaximum=110, SourceMinimum=90, TargetMaximum=1, TargetMinimum=0, Value=@"Get Delta Ref_5".Result.Rotation.Yaw)
   > &RigUnit_ModifyTransforms(
    ItemToModify=((Item=(Type=Bone,Name="dyn_pad_l"),Transform=(Rotation=(X=0.1209,Y=-0.1776,Z=-0.0199,W=0.9764),Translation=(X=-1.5,Y=1.5,Z=-1.5),Scale3D=(X=1,Y=1,Z=1))))
    WeightMinimum=0
    Weight=@n5.Result
   )
   > &RigUnit_ModifyTransforms(
    ItemToModify=((Item=(Type=Bone,Name="dyn_pad_l"),Transform=(Rotation=(X=0.0871,Y=-0.0435,Z=0.0038,W=0.9952),Translation=(X=-1,Y=0,Z=0),Scale3D=(X=1,Y=1,Z=1))))
    WeightMinimum=0
    Weight=&RigVMFunction_MathDoubleSub(
     B=@n5.Result
     A=&RigVMFunction_MathDoubleRemap(SourceMaximum=60, SourceMinimum=0, TargetMaximum=1, TargetMinimum=0, Value=@"Get Delta Ref_5".Result.Rotation.Yaw).Result
    ).Result
   )
   Get Delta Ref_5_2 = RigVMFunctionReferenceNode(Item=(Type=Bone,Name="upperarm_r"), ExecuteContext=@RigVMFunction_Sequence_1.B)
   n6 = &RigVMFunction_Sequence(ExecuteContext=@"Get Delta Ref_5_2".ExecuteContext)
   &RigUnit_ModifyTransforms(
    ItemToModify=((Item=(Type=Bone,Name="dyn_pad_r"),Transform=(Rotation=(X=0,Y=0.0872,Z=0,W=0.9962),Translation=(X=-3,Y=0,Z=-6),Scale3D=(X=1,Y=1.05,Z=1))))
    WeightMinimum=0
    Weight=&RigVMFunction_MathDoubleRemap(SourceMaximum=75, SourceMinimum=0, TargetMaximum=1, TargetMinimum=0, Value=@"Get Delta Ref_5_2".Result.Rotation.Pitch).Result
    ExecutePin=@n6.A
   )
   n7 = &RigVMFunction_MathDoubleRemap(SourceMaximum=0, SourceMinimum=-60, TargetMaximum=0, TargetMinimum=1, Value=@"Get Delta Ref_5_2".Result.Rotation.Pitch)
   > &RigUnit_ModifyTransforms(
    ItemToModify=((Item=(Type=Bone,Name="dyn_pad_r"),Transform=(Rotation=(X=0,Y=0,Z=0,W=1),Translation=(X=-1.5,Y=-0.5,Z=0),Scale3D=(X=1,Y=1.05,Z=1))))
    WeightMinimum=0
    Weight=&RigVMFunction_MathDoubleMul(B=@n7.Result, A=RigVMVariableNode() {variable=isMelee_NOT}.Value).Result
   )
   > &RigUnit_ModifyTransforms(
    ItemToModify=((Item=(Type=Bone,Name="dyn_pad_r"),Transform=(Rotation=(X=-0.173,Y=0.0151,Z=-0.0858,W=0.9811),Translation=(X=-1.5,Y=-1,Z=1),Scale3D=(X=1,Y=1.05,Z=1))))
    WeightMinimum=0
    Weight=&RigVMFunction_MathDoubleMul(A=@n7.Result, B=RigVMVariableNode() {variable=isMelee}.Value).Result
   )
   &RigUnit_ModifyTransforms(
    ItemToModify=((Item=(Type=Bone,Name="dyn_pad_r"),Transform=(Rotation=(X=-0.3827,Y=0,Z=0,W=0.9239),Translation=(X=0,Y=0.5,Z=1),Scale3D=(X=1,Y=1.05,Z=1))))
    WeightMinimum=0
    Weight=&RigVMFunction_MathDoubleRemap(SourceMaximum=-68, SourceMinimum=0, TargetMaximum=1, TargetMinimum=0, Value=@"Get Delta Ref_5_2".Result.Rotation.Yaw).Result
    ExecutePin=@n6.B
   )
   n8 = &RigVMFunction_MathDoubleRemap(SourceMaximum=110, SourceMinimum=80, TargetMaximum=1, TargetMinimum=0, Value=@"Get Delta Ref_5_2".Result.Rotation.Yaw)
   > &RigUnit_ModifyTransforms(
    ItemToModify=((Item=(Type=Bone,Name="dyn_pad_r"),Transform=(Rotation=(X=0.1209,Y=-0.1776,Z=-0.0199,W=0.9764),Translation=(X=1.5,Y=-1.5,Z=1.5),Scale3D=(X=1,Y=1,Z=1))))
    WeightMinimum=0
    Weight=@n8.Result
   )
   Subtract = &RigVMFunction_MathDoubleSub(
    A=&RigVMFunction_MathDoubleRemap(SourceMaximum=80, SourceMinimum=0, TargetMaximum=1, TargetMinimum=0, Value=@"Get Delta Ref_5_2".Result.Rotation.Yaw).Result
    B=@n8.Result
   )
   > &RigUnit_ModifyTransforms(
    ItemToModify=((Item=(Type=Bone,Name="dyn_pad_r"),Transform=(Rotation=(X=0.1285,Y=-0.1722,Z=0.0227,W=0.9764),Translation=(X=0,Y=0.5,Z=0),Scale3D=(X=1,Y=1.05,Z=1))))
    WeightMinimum=0
    Weight=@Subtract.Result
   )
   Get Delta Ref_5_1 = RigVMFunctionReferenceNode(Item=(Type=Bone,Name="upperarm_r"), ExecuteContext=@RigVMFunction_Sequence_1.C)
   n9 = &RigVMFunction_AlphaInterp(
    InterpSpeedDecreasing=50
    InterpSpeedIncreasing=50
    Scale=1
    bClampResult=False
    Value=&RigVMFunction_MathDoubleRemap(SourceMaximum=30, SourceMinimum=0, TargetMaximum=1, TargetMinimum=0, Value=@"Get Delta Ref_5_1".Result.Rotation.Roll).Result
   )
   n10 = RigVMVariableNode() {variable=isMelee_NOT}
   > &RigUnit_ModifyTransforms(
    ItemToModify=((Item=(Type=Bone,Name="dyn_pad_r"),Transform=(Rotation=(X=0.1305,Y=0,Z=0,W=0.9914),Translation=(X=0,Y=0,Z=0),Scale3D=(X=1,Y=1,Z=1))))
    WeightMinimum=0
    Weight=&RigVMFunction_MathDoubleMul(A=@n9.Result, B=@n10.Value).Result
   )
   n11 = &RigVMFunction_AlphaInterp(
    InterpSpeedDecreasing=50
    InterpSpeedIncreasing=50
    Scale=1
    bClampResult=False
    Value=&RigVMFunction_MathDoubleRemap(SourceMaximum=0, SourceMinimum=-30, TargetMaximum=0, TargetMinimum=1, Value=@"Get Delta Ref_5_1".Result.Rotation.Roll).Result
   )
   > &RigUnit_ModifyTransforms(
    ItemToModify=((Item=(Type=Bone,Name="dyn_pad_r"),Transform=(Rotation=(X=0,Y=0,Z=0,W=1),Translation=(X=0,Y=0,Z=0),Scale3D=(X=1,Y=1,Z=1))))
    WeightMinimum=0
    Weight=&RigVMFunction_MathDoubleMul(A=@n11.Result, B=@n10.Value).Result
   )
   n12 = &RigVMFunction_AlphaInterp(
    InterpSpeedDecreasing=50
    InterpSpeedIncreasing=50
    Scale=1
    bClampResult=False
    Value=&RigVMFunction_MathDoubleRemap(SourceMaximum=30, SourceMinimum=0, TargetMaximum=1, TargetMinimum=0, Value=@"Get Delta Ref_5_1".Result.Rotation.Roll).Result
   )
   > &RigUnit_ModifyTransforms(
    ItemToModify=((Item=(Type=Bone,Name="dyn_pad_r"),Transform=(Rotation=(X=0.1736,Y=0,Z=0,W=0.9848),Translation=(X=0,Y=0,Z=0),Scale3D=(X=1,Y=1,Z=1))))
    WeightMinimum=0
    Weight=&RigVMFunction_MathDoubleMul(A=@n12.Result, B=RigVMVariableNode() {variable=isMelee}.Value).Result
   )
   n13 = &RigVMFunction_AlphaInterp(
    InterpSpeedDecreasing=50
    InterpSpeedIncreasing=50
    Scale=1
    bClampResult=False
    Value=&RigVMFunction_MathDoubleRemap(SourceMaximum=0, SourceMinimum=-30, TargetMaximum=0, TargetMinimum=1, Value=@"Get Delta Ref_5_1".Result.Rotation.Roll).Result
   )
   > &RigUnit_ModifyTransforms(
    ItemToModify=((Item=(Type=Bone,Name="dyn_pad_r"),Transform=(Rotation=(X=-0.1305,Y=0,Z=0,W=0.9914),Translation=(X=0,Y=0,Z=0),Scale3D=(X=1,Y=1,Z=1))))
    WeightMinimum=0
    Weight=&RigVMFunction_MathDoubleMul(B=@n13.Result, A=RigVMVariableNode() {variable=isMelee}.Value).Result
   )
   Get Delta Ref_5_1_1 = RigVMFunctionReferenceNode(Item=(Type=Bone,Name="upperarm_l"), ExecuteContext=@RigVMFunction_Sequence_1.D)
   AlphaInterp = &RigVMFunction_AlphaInterp(
    InterpSpeedDecreasing=50
    InterpSpeedIncreasing=50
    Scale=1
    bClampResult=False
    Value=&RigVMFunction_MathDoubleRemap(SourceMaximum=30, SourceMinimum=0, TargetMaximum=1, TargetMinimum=0, Value=@"Get Delta Ref_5_1_1".Result.Rotation.Roll).Result
   )
   > &RigUnit_ModifyTransforms(
    ItemToModify=((Item=(Type=Bone,Name="dyn_pad_l"),Transform=(Rotation=(X=0.1736,Y=0,Z=0,W=0.9848),Translation=(X=0,Y=0,Z=0),Scale3D=(X=1,Y=1,Z=1))))
    WeightMinimum=0
    Weight=@AlphaInterp.Result
   )
   n14 = &RigVMFunction_AlphaInterp(
    InterpSpeedDecreasing=50
    InterpSpeedIncreasing=50
    Scale=1
    bClampResult=False
    Value=&RigVMFunction_MathDoubleRemap(SourceMaximum=0, SourceMinimum=-30, TargetMaximum=0, TargetMinimum=1, Value=@"Get Delta Ref_5_1_1".Result.Rotation.Roll).Result
   )
   > &RigUnit_ModifyTransforms(
    ItemToModify=((Item=(Type=Bone,Name="dyn_pad_l"),Transform=(Rotation=(X=0,Y=0,Z=0,W=1),Translation=(X=0,Y=0,Z=0),Scale3D=(X=1,Y=1.05,Z=1))))
    WeightMinimum=0
    Weight=@n14.Result
   )
   RigVMFunction_Sequence_1_1 = RigVMAggregateNode(ExecuteContext=@RigVMFunction_Sequence.B)
   n15 = &RigVMFunction_ControlFlowBranch(
    Condition=RigVMDispatchNode(B=1, A=RigVMVariableNode() {variable=isSkyDiving}.Value) {template="DISPATCH_RigVMDispatch_CoreEquals(in A,in B,out Result)"}.Result
    ExecuteContext=@RigVMFunction_Sequence_1_1.A
   )
   n16 = &RigUnit_SphericalPoseReader(
    ActiveRegionScaleFactors=(PositiveWidth=1,NegativeWidth=1,PositiveHeight=1,NegativeHeight=1)
    ActiveRegionSize=0.07
    Debug=(bDrawDebug=true,bDraw2D=False,bDrawLocalAxes=False,DebugScale=25,DebugSegments=20,DebugThickness=0.25)
    DriverAxis=(X=0,Y=0,Z=-1)
    DriverItem=(Type=Bone,Name="clavicle_r")
    FalloffRegionScaleFactors=(PositiveWidth=1,NegativeWidth=1,PositiveHeight=1,NegativeHeight=1)
    FalloffSize=0.5
    FlipHeightScaling=False
    FlipWidthScaling=False
    OptionalParentItem=(Type=Bone,Name="None")
    RotationOffset=(X=0,Y=-150,Z=0)
    ExecutePin=@n15.True
   )
   > n17 = &RigUnit_SphericalPoseReader(
    ActiveRegionScaleFactors=(PositiveWidth=1,NegativeWidth=1,PositiveHeight=1,NegativeHeight=1)
    ActiveRegionSize=0.1
    Debug=(bDrawDebug=false,bDraw2D=False,bDrawLocalAxes=False,DebugScale=25,DebugSegments=20,DebugThickness=0.25)
    DriverAxis=(X=-1,Y=0,Z=0)
    DriverItem=(Type=Bone,Name="upperarm_r")
    FalloffRegionScaleFactors=(PositiveWidth=1,NegativeWidth=1,PositiveHeight=1,NegativeHeight=1)
    FalloffSize=0.5
    FlipHeightScaling=False
    FlipWidthScaling=False
    OptionalParentItem=(Type=Bone,Name="None")
    RotationOffset=(X=0,Y=120,Z=5)
   )
   Multiply_11_2 = RigVMAggregateNode(A=@n16.OutputParam, B=@n17.OutputParam, C=RigVMVariableNode() {variable=isSkyDiving}.Value)
   > &RigUnit_ModifyTransforms(
    ItemToModify=((Item=(Type=Bone,Name="dyn_pad_r"),Transform=(Rotation=(X=-0.0872,Y=0,Z=0,W=0.9962),Translation=(X=-3,Y=-1,Z=-3),Scale3D=(X=1,Y=1,Z=1))))
    WeightMinimum=0
    Weight=@Multiply_11_2.Result
   )
   > n18 = &RigUnit_SphericalPoseReader(
    ActiveRegionScaleFactors=(PositiveWidth=1,NegativeWidth=1,PositiveHeight=1,NegativeHeight=1)
    ActiveRegionSize=0.07
    Debug=(bDrawDebug=false,bDraw2D=False,bDrawLocalAxes=False,DebugScale=25,DebugSegments=20,DebugThickness=0.25)
    DriverAxis=(X=0,Y=0,Z=1)
    DriverItem=(Type=Bone,Name="clavicle_l")
    FalloffRegionScaleFactors=(PositiveWidth=1,NegativeWidth=1,PositiveHeight=1,NegativeHeight=1)
    FalloffSize=0.5
    FlipHeightScaling=False
    FlipWidthScaling=False
    OptionalParentItem=(Type=Bone,Name="None")
    RotationOffset=(X=0,Y=30,Z=0)
   )
   > n19 = &RigUnit_SphericalPoseReader(
    ActiveRegionScaleFactors=(PositiveWidth=1,NegativeWidth=1,PositiveHeight=1,NegativeHeight=1)
    ActiveRegionSize=0.1
    Debug=(bDrawDebug=false,bDraw2D=False,bDrawLocalAxes=False,DebugScale=25,DebugSegments=20,DebugThickness=0.25)
    DriverAxis=(X=1,Y=0,Z=0)
    DriverItem=(Type=Bone,Name="upperarm_l")
    FalloffRegionScaleFactors=(PositiveWidth=1,NegativeWidth=1,PositiveHeight=1,NegativeHeight=1)
    FalloffSize=0.5
    FlipHeightScaling=False
    FlipWidthScaling=False
    OptionalParentItem=(Type=Bone,Name="None")
    RotationOffset=(X=0,Y=-45,Z=5)
   )
   Multiply_11 = RigVMAggregateNode(B=@n19.OutputParam, A=@n18.OutputParam, C=RigVMVariableNode() {variable=isSkyDiving}.Value)
   > &RigUnit_ModifyTransforms(
    ItemToModify=((Item=(Type=Bone,Name="dyn_pad_l"),Transform=(Rotation=(X=-0.0871,Y=0.0435,Z=0.0038,W=0.9952),Translation=(X=2,Y=0.5,Z=1.5),Scale3D=(X=1,Y=1,Z=1))))
    WeightMinimum=0
    Weight=@Multiply_11.Result
   )
   RigVMFunction_Sequence_4 = RigVMAggregateNode(ExecuteContext=@RigVMFunction_Sequence_1_1.B)
   n20 = &RigVMFunction_ControlFlowBranch(
    Condition=RigVMDispatchNode(B=1, A=RigVMVariableNode() {variable=isEmote}.Value) {template="DISPATCH_RigVMDispatch_CoreEquals(in A,in B,out Result)"}.Result
    ExecuteContext=@RigVMFunction_Sequence_4.A
   )
   n21 = &RigUnit_SphericalPoseReader(
    ActiveRegionScaleFactors=(PositiveWidth=1,NegativeWidth=1,PositiveHeight=1,NegativeHeight=1)
    ActiveRegionSize=0.07
    Debug=(bDrawDebug=false,bDraw2D=False,bDrawLocalAxes=False,DebugScale=25,DebugSegments=20,DebugThickness=0.25)
    DriverAxis=(X=0,Y=0,Z=1)
    DriverItem=(Type=Bone,Name="clavicle_l")
    FalloffRegionScaleFactors=(PositiveWidth=1,NegativeWidth=1,PositiveHeight=1,NegativeHeight=1)
    FalloffSize=0.3
    FlipHeightScaling=False
    FlipWidthScaling=False
    OptionalParentItem=(Type=Bone,Name="None")
    RotationOffset=(X=0,Y=30,Z=0)
    ExecutePin=@n20.True
   )
   > n22 = &RigUnit_SphericalPoseReader(
    ActiveRegionScaleFactors=(PositiveWidth=1.5,NegativeWidth=1,PositiveHeight=1,NegativeHeight=2)
    ActiveRegionSize=0.1
    Debug=(bDrawDebug=false,bDraw2D=False,bDrawLocalAxes=False,DebugScale=25,DebugSegments=20,DebugThickness=0.25)
    DriverAxis=(X=1,Y=0,Z=0)
    DriverItem=(Type=Bone,Name="upperarm_l")
    FalloffRegionScaleFactors=(PositiveWidth=1,NegativeWidth=1,PositiveHeight=1,NegativeHeight=2)
    FalloffSize=0.3
    FlipHeightScaling=False
    FlipWidthScaling=False
    OptionalParentItem=(Type=Bone,Name="None")
    RotationOffset=(X=0,Y=-30,Z=5)
   )
   Multiply_12 = RigVMAggregateNode(A=@n21.OutputParam, B=@n22.OutputParam, C=RigVMVariableNode() {variable=isEmote}.Value)
   > &RigUnit_ModifyTransforms(
    ItemToModify=((Item=(Type=Bone,Name="dyn_pad_l"),Transform=(Rotation=(X=0.0871,Y=0.0435,Z=-0.0038,W=0.9952),Translation=(X=1.5,Y=0.5,Z=1.3),Scale3D=(X=1,Y=1,Z=1))))
    WeightMinimum=0
    Weight=@Multiply_12.Result
   )
   > n23 = &RigUnit_SphericalPoseReader(
    ActiveRegionScaleFactors=(PositiveWidth=1,NegativeWidth=1,PositiveHeight=1,NegativeHeight=1)
    ActiveRegionSize=0.07
    Debug=(bDrawDebug=false,bDraw2D=False,bDrawLocalAxes=False,DebugScale=25,DebugSegments=20,DebugThickness=0.25)
    DriverAxis=(X=0,Y=0,Z=-1)
    DriverItem=(Type=Bone,Name="clavicle_r")
    FalloffRegionScaleFactors=(PositiveWidth=1,NegativeWidth=1,PositiveHeight=1,NegativeHeight=1)
    FalloffSize=0.3
    FlipHeightScaling=False
    FlipWidthScaling=False
    OptionalParentItem=(Type=Bone,Name="None")
    RotationOffset=(X=0,Y=-150,Z=0)
   )
   > n24 = &RigUnit_SphericalPoseReader(
    ActiveRegionScaleFactors=(PositiveWidth=1.5,NegativeWidth=1,PositiveHeight=1,NegativeHeight=2)
    ActiveRegionSize=0.1
    Debug=(bDrawDebug=false,bDraw2D=False,bDrawLocalAxes=False,DebugScale=25,DebugSegments=20,DebugThickness=0.25)
    DriverAxis=(X=-1,Y=0,Z=0)
    DriverItem=(Type=Bone,Name="upperarm_r")
    FalloffRegionScaleFactors=(PositiveWidth=1,NegativeWidth=1,PositiveHeight=1,NegativeHeight=2)
    FalloffSize=0.3
    FlipHeightScaling=False
    FlipWidthScaling=False
    OptionalParentItem=(Type=Bone,Name="None")
    RotationOffset=(X=0,Y=150,Z=5)
   )
   Multiply_12_1 = RigVMAggregateNode(A=@n23.OutputParam, B=@n24.OutputParam, C=RigVMVariableNode() {variable=isEmote}.Value)
   > &RigUnit_ModifyTransforms(
    ItemToModify=((Item=(Type=Bone,Name="dyn_pad_r"),Transform=(Rotation=(X=0,Y=0,Z=0,W=1),Translation=(X=-1,Y=-1,Z=0.5),Scale3D=(X=1,Y=1,Z=1))))
    WeightMinimum=0
    Weight=@Multiply_12_1.Result
   )
   n25 = &RigVMFunction_ControlFlowBranch(
    Condition=RigVMDispatchNode(B=1, A=RigVMVariableNode() {variable=isSkyDiving_NOT}.Value) {template="DISPATCH_RigVMDispatch_CoreEquals(in A,in B,out Result)"}.Result
    ExecuteContext=@RigVMFunction_Sequence_4.B
   )
   n26 = &RigUnit_SphericalPoseReader(
    ActiveRegionScaleFactors=(PositiveWidth=1,NegativeWidth=1,PositiveHeight=1,NegativeHeight=1)
    ActiveRegionSize=0.1
    Debug=(bDrawDebug=true,bDraw2D=False,bDrawLocalAxes=False,DebugScale=25,DebugSegments=20,DebugThickness=0.25)
    DriverAxis=(X=-1,Y=0,Z=0)
    DriverItem=(Type=Bone,Name="upperarm_r")
    FalloffRegionScaleFactors=(PositiveWidth=1,NegativeWidth=1,PositiveHeight=1,NegativeHeight=1)
    FalloffSize=0.2
    FlipHeightScaling=False
    FlipWidthScaling=False
    OptionalParentItem=(Type=Bone,Name="spine_05")
    RotationOffset=(X=-90,Y=0,Z=-60)
    ExecutePin=@n25.True
   )
   > &RigUnit_ModifyTransforms(
    ItemToModify=((Item=(Type=Bone,Name="dyn_pad_r"),Transform=(Rotation=(X=0.1304,Y=0.0432,Z=-0.0057,W=0.9905),Translation=(X=0,Y=0,Z=-1),Scale3D=(X=1,Y=1,Z=1))))
    WeightMinimum=0
    Weight=@n26.OutputParam
   )
   n27 = &RigVMFunction_ControlFlowBranch(
    Condition=RigVMDispatchNode(B=1, A=RigVMVariableNode() {variable=isRail}.Value) {template="DISPATCH_RigVMDispatch_CoreEquals(in A,in B,out Result)"}.Result
    ExecuteContext=@RigVMFunction_Sequence_4.C
   )
   n28 = &RigUnit_SphericalPoseReader(
    ActiveRegionScaleFactors=(PositiveWidth=1,NegativeWidth=1,PositiveHeight=1,NegativeHeight=1)
    ActiveRegionSize=0.2
    Debug=(bDrawDebug=true,bDraw2D=False,bDrawLocalAxes=False,DebugScale=25,DebugSegments=20,DebugThickness=0.25)
    DriverAxis=(X=-1,Y=0,Z=0)
    DriverItem=(Type=Bone,Name="upperarm_r")
    FalloffRegionScaleFactors=(PositiveWidth=1,NegativeWidth=1,PositiveHeight=1,NegativeHeight=1)
    FalloffSize=0.5
    FlipHeightScaling=False
    FlipWidthScaling=False
    OptionalParentItem=(Type=Bone,Name="spine_05")
    RotationOffset=(X=-130,Y=-25,Z=0)
    ExecutePin=@n27.True
   )
   > &RigUnit_ModifyTransforms(
    ItemToModify=((Item=(Type=Bone,Name="dyn_pad_r"),Transform=(Rotation=(X=-0.0871,Y=0.0435,Z=0.0038,W=0.9952),Translation=(X=0,Y=0,Z=-0.5),Scale3D=(X=1,Y=1.05,Z=1))))
    WeightMinimum=0
    Weight=@n28.OutputParam
   )
   > n29 = &RigUnit_SphericalPoseReader(
    ActiveRegionScaleFactors=(PositiveWidth=1,NegativeWidth=1,PositiveHeight=1,NegativeHeight=1)
    ActiveRegionSize=0.3
    Debug=(bDrawDebug=true,bDraw2D=False,bDrawLocalAxes=False,DebugScale=25,DebugSegments=20,DebugThickness=0.25)
    DriverAxis=(X=1,Y=0,Z=0)
    DriverItem=(Type=Bone,Name="upperarm_l")
    FalloffRegionScaleFactors=(PositiveWidth=1,NegativeWidth=1,PositiveHeight=1,NegativeHeight=1)
    FalloffSize=0.5
    FlipHeightScaling=False
    FlipWidthScaling=False
    OptionalParentItem=(Type=Bone,Name="spine_05")
    RotationOffset=(X=100,Y=0,Z=-45)
   )
   > &RigUnit_ModifyTransforms(
    ItemToModify=((Item=(Type=Bone,Name="dyn_pad_l"),Transform=(Rotation=(X=0,Y=0,Z=0,W=1),Translation=(X=-1,Y=0,Z=1),Scale3D=(X=1,Y=1.05,Z=1))))
    WeightMinimum=0
    Weight=@n29.OutputParam
   )
   n30 = &RigVMFunction_Sequence(ExecuteContext=@RigVMFunction_Sequence_4.D)
   RigVMFunction_MathBoolOr = &RigVMFunction_MathBoolOr(B=RigVMVariableNode() {variable=isFalling}.Value, A=RigVMVariableNode() {variable=isJumpting}.Value)
   n31 = &RigVMFunction_ControlFlowBranch(Condition=@RigVMFunction_MathBoolOr.Result, ExecuteContext=@n30.A)
   n32 = &RigUnit_SphericalPoseReader(
    ActiveRegionScaleFactors=(PositiveWidth=1,NegativeWidth=1,PositiveHeight=1,NegativeHeight=1)
    ActiveRegionSize=0.05
    Debug=(bDrawDebug=true,bDraw2D=False,bDrawLocalAxes=False,DebugScale=25,DebugSegments=20,DebugThickness=0.25)
    DriverAxis=(X=1,Y=0,Z=0)
    DriverItem=(Type=Bone,Name="upperarm_l")
    FalloffRegionScaleFactors=(PositiveWidth=1,NegativeWidth=1,PositiveHeight=1,NegativeHeight=1)
    FalloffSize=0.1
    FlipHeightScaling=False
    FlipWidthScaling=False
    OptionalParentItem=(Type=Bone,Name="spine_05")
    RotationOffset=(X=80,Y=0,Z=-45)
    ExecutePin=@n31.True
   )
   > &RigUnit_ModifyTransforms(
    ItemToModify=((Item=(Type=Bone,Name="dyn_pad_l"),Transform=(Rotation=(X=0.0472,Y=0.0396,Z=-0.0889,W=0.9941),Translation=(X=1,Y=0,Z=1),Scale3D=(X=1,Y=1.05,Z=1))))
    WeightMinimum=0
    Weight=@n32.OutputParam
   )
   n33 = &RigVMFunction_ControlFlowBranch(
    Condition=RigVMDispatchNode(B=1, A=RigVMVariableNode() {variable=isPistol}.Value) {template="DISPATCH_RigVMDispatch_CoreEquals(in A,in B,out Result)"}.Result
    ExecuteContext=@n30.B
   )
   Get Delta Ref_2 = RigVMFunctionReferenceNode(Item=(Type=Bone,Name="spine_05"), ExecuteContext=@n33.True)
   > &RigUnit_ModifyTransforms(
    ItemToModify=((Item=(Type=Bone,Name="dyn_pad_l"),Transform=(Rotation=(X=0,Y=0,Z=0,W=1),Translation=(X=0,Y=-1,Z=0),Scale3D=(X=1,Y=1,Z=1))))
    WeightMinimum=0
    Weight=&RigVMFunction_MathDoubleRemap(SourceMaximum=9, SourceMinimum=0, TargetMaximum=1, TargetMinimum=0, bClamp=False, Value=@"Get Delta Ref_2".Result.Rotation.Roll).Result
   )
   n34 = &RigVMFunction_ControlFlowBranch(
    Condition=RigVMDispatchNode(B=1, A=RigVMVariableNode() {variable=isDBNO}.Value) {template="DISPATCH_RigVMDispatch_CoreEquals(in A,in B,out Result)"}.Result
    ExecuteContext=@RigVMFunction_Sequence_4.E
   )
   n35 = &RigVMFunction_Sequence(ExecuteContext=@n34.True)
   Get Delta Ref_5_1_1_3 = RigVMFunctionReferenceNode(Item=(Type=Bone,Name="clavicle_l"), ExecuteContext=@n35.A)
   > &RigUnit_ModifyTransforms(
    ItemToModify=((Item=(Type=Bone,Name="dyn_pad_l"),Transform=(Rotation=(X=0.1736,Y=0,Z=0,W=0.9848),Translation=(X=0,Y=-0.5,Z=1),Scale3D=(X=1,Y=1,Z=1))))
    WeightMinimum=0
    Weight=&RigVMFunction_MathDoubleRemap(SourceMaximum=20, SourceMinimum=0, TargetMaximum=1, TargetMinimum=0, Value=@"Get Delta Ref_5_1_1_3".Result.Rotation.Yaw).Result
   )
   > &RigUnit_ModifyTransforms(
    ItemToModify=((Item=(Type=Bone,Name="dyn_pad_l"),Transform=(Rotation=(X=0.2912,Y=-0.1089,Z=-0.0567,W=0.9488),Translation=(X=1.5,Y=0,Z=0.5),Scale3D=(X=1,Y=1.05,Z=1))))
    WeightMinimum=0
    Weight=&RigVMFunction_MathDoubleRemap(SourceMaximum=-15, SourceMinimum=0, TargetMaximum=1, TargetMinimum=0, Value=@"Get Delta Ref_5_1_1_3".Result.Rotation.Yaw).Result
   )
   Get Delta Ref_5_1_1_3_1 = RigVMFunctionReferenceNode(Item=(Type=Bone,Name="clavicle_r"), ExecuteContext=@n35.B)
   > &RigUnit_ModifyTransforms(
    ItemToModify=((Item=(Type=Bone,Name="dyn_pad_r"),Transform=(Rotation=(X=0,Y=0.0872,Z=0,W=0.9962),Translation=(X=-1,Y=-0.7,Z=-3),Scale3D=(X=1,Y=1,Z=1))))
    WeightMinimum=0
    Weight=&RigVMFunction_MathDoubleRemap(SourceMaximum=10, SourceMinimum=0, TargetMaximum=1, TargetMinimum=0, Value=@"Get Delta Ref_5_1_1_3_1".Result.Rotation.Yaw).Result
   )
   > &RigUnit_ModifyTransforms(
    ItemToModify=((Item=(Type=Bone,Name="dyn_pad_r"),Transform=(Rotation=(X=0.2578,Y=-0.0842,Z=0.0226,W=0.9623),Translation=(X=-1,Y=0.5,Z=-1),Scale3D=(X=1,Y=1.05,Z=1))))
    WeightMinimum=0
    Weight=&RigVMFunction_MathDoubleRemap(SourceMaximum=-20, SourceMinimum=0, TargetMaximum=1, TargetMinimum=0, Value=@"Get Delta Ref_5_1_1_3_1".Result.Rotation.Yaw).Result
   )
   > n36 = &RigUnit_SphericalPoseReader(
    ActiveRegionScaleFactors=(PositiveWidth=1,NegativeWidth=1,PositiveHeight=1,NegativeHeight=1)
    ActiveRegionSize=0.1
    Debug=(bDrawDebug=True,bDraw2D=False,bDrawLocalAxes=False,DebugScale=25,DebugSegments=20,DebugThickness=0.25)
    DriverAxis=(X=-1,Y=0,Z=0)
    DriverItem=(Type=Bone,Name="upperarm_r")
    FalloffRegionScaleFactors=(PositiveWidth=1,NegativeWidth=1,PositiveHeight=1,NegativeHeight=1)
    FalloffSize=0.2
    FlipHeightScaling=False
    FlipWidthScaling=False
    OptionalParentItem=(Type=Bone,Name="None")
    RotationOffset=(X=-90,Y=0,Z=0)
   )
   > &RigUnit_ModifyTransforms(
    ItemToModify=((Item=(Type=Bone,Name="dyn_pad_r"),Transform=(Rotation=(X=0.0872,Y=0,Z=0,W=0.9962),Translation=(X=0,Y=0,Z=0),Scale3D=(X=1,Y=1.05,Z=1))))
    WeightMinimum=0
    Weight=@n36.OutputParam
   )
   > n37 = &RigUnit_SphericalPoseReader(
    ActiveRegionScaleFactors=(PositiveWidth=1,NegativeWidth=1,PositiveHeight=1,NegativeHeight=1)
    ActiveRegionSize=0.1
    Debug=(bDrawDebug=True,bDraw2D=False,bDrawLocalAxes=False,DebugScale=25,DebugSegments=20,DebugThickness=0.25)
    DriverAxis=(X=1,Y=0,Z=0)
    DriverItem=(Type=Bone,Name="upperarm_l")
    FalloffRegionScaleFactors=(PositiveWidth=1,NegativeWidth=1,PositiveHeight=1,NegativeHeight=1)
    FalloffSize=0.2
    FlipHeightScaling=False
    FlipWidthScaling=False
    OptionalParentItem=(Type=Bone,Name="None")
    RotationOffset=(X=80,Y=0,Z=-10)
   )
   > &RigUnit_ModifyTransforms(
    ItemToModify=((Item=(Type=Bone,Name="dyn_pad_l"),Transform=(Rotation=(X=0.1305,Y=0,Z=0,W=0.9914),Translation=(X=0,Y=0,Z=0),Scale3D=(X=1,Y=1.05,Z=1))))
    WeightMinimum=0
    Weight=@n37.OutputParam
   )
   RigVMFunction_MathBoolAnd_2 = RigVMAggregateNode(
    B=RigVMDispatchNode(B=1, A=RigVMVariableNode() {variable=isSkyDiving_NOT}.Value) {template="DISPATCH_RigVMDispatch_CoreEquals(in A,in B,out Result)"}.Result
    A=RigVMDispatchNode(B=1, A=RigVMVariableNode() {variable=isMelee}.Value) {template="DISPATCH_RigVMDispatch_CoreEquals(in A,in B,out Result)"}.Result
    C=RigVMDispatchNode(B=1, A=RigVMVariableNode() {variable=isCrouching_NOT}.Value) {template="DISPATCH_RigVMDispatch_CoreEquals(in A,in B,out Result)"}.Result
   )
   n38 = &RigVMFunction_ControlFlowBranch(Condition=@RigVMFunction_MathBoolAnd_2.Result, ExecuteContext=@RigVMFunction_Sequence_1_1.C)
   n39 = &RigUnit_SphericalPoseReader(
    ActiveRegionScaleFactors=(PositiveWidth=1,NegativeWidth=1,PositiveHeight=1,NegativeHeight=1)
    ActiveRegionSize=0.1
    Debug=(bDrawDebug=true,bDraw2D=False,bDrawLocalAxes=False,DebugScale=25,DebugSegments=20,DebugThickness=0.25)
    DriverAxis=(X=1,Y=0,Z=0)
    DriverItem=(Type=Bone,Name="upperarm_l")
    FalloffRegionScaleFactors=(PositiveWidth=1,NegativeWidth=1,PositiveHeight=1,NegativeHeight=1)
    FalloffSize=0.2
    FlipHeightScaling=False
    FlipWidthScaling=False
    OptionalParentItem=(Type=Bone,Name="spine_05")
    RotationOffset=(X=90,Y=0,Z=-60)
    ExecutePin=@n38.True
   )
   n40 = &RigVMFunction_AlphaInterp(InterpSpeedDecreasing=25, InterpSpeedIncreasing=25, Scale=1, bClampResult=False, Value=@n39.OutputParam)
   > &RigUnit_ModifyTransforms(
    ItemToModify=((Item=(Type=Bone,Name="dyn_pad_l"),Transform=(Rotation=(X=-0.1736,Y=0,Z=0,W=0.9848),Translation=(X=0,Y=0,Z=0),Scale3D=(X=1,Y=1,Z=1))))
    WeightMinimum=0
    Weight=@n40.Result
   )
   > n41 = &RigUnit_SphericalPoseReader(
    ActiveRegionScaleFactors=(PositiveWidth=1,NegativeWidth=1,PositiveHeight=1,NegativeHeight=1)
    ActiveRegionSize=0.1
    Debug=(bDrawDebug=true,bDraw2D=False,bDrawLocalAxes=False,DebugScale=25,DebugSegments=20,DebugThickness=0.25)
    DriverAxis=(X=1,Y=0,Z=0)
    DriverItem=(Type=Bone,Name="upperarm_l")
    FalloffRegionScaleFactors=(PositiveWidth=1,NegativeWidth=1,PositiveHeight=1,NegativeHeight=1)
    FalloffSize=0.2
    FlipHeightScaling=False
    FlipWidthScaling=False
    OptionalParentItem=(Type=Bone,Name="spine_01")
    RotationOffset=(X=130,Y=0,Z=-10)
   )
   n42 = &RigVMFunction_AlphaInterp(InterpSpeedDecreasing=25, InterpSpeedIncreasing=25, Scale=1, bClampResult=False, Value=@n41.OutputParam)
   > &RigUnit_ModifyTransforms(
    ItemToModify=((Item=(Type=Bone,Name="dyn_pad_l"),Transform=(Rotation=(X=-0.0872,Y=0,Z=0,W=0.9962),Translation=(X=0,Y=-1.5,Z=0),Scale3D=(X=1,Y=1.05,Z=1))))
    WeightMinimum=0
    Weight=@n42.Result
   )
   > Get Delta Ref = RigVMFunctionReferenceNode(Item=(Type=Bone,Name="upperarm_l"))
   > Get Delta Ref_3 = RigVMFunctionReferenceNode(Item=(Type=Bone,Name="clavicle_l"))
   Multiply_2 = RigVMAggregateNode(
    B=&RigVMFunction_MathDoubleRemap(SourceMaximum=22, SourceMinimum=0, TargetMaximum=1, TargetMinimum=0, Value=@"Get Delta Ref_3".Result.Rotation.Pitch).Result
    C=&RigVMFunction_MathDoubleRemap(SourceMaximum=-14, SourceMinimum=0, TargetMaximum=1, TargetMinimum=0, Value=@"Get Delta Ref_3".Result.Rotation.Yaw).Result
    A=&RigVMFunction_MathDoubleRemap(SourceMaximum=-33, SourceMinimum=-35, TargetMaximum=1, TargetMinimum=0, Value=@"Get Delta Ref_3".Result.Rotation.Roll).Result
   )
   Remap = &RigVMFunction_MathDoubleRemap(SourceMaximum=-94, SourceMinimum=-33, TargetMaximum=1, TargetMinimum=0, Value=@"Get Delta Ref".Result.Rotation.Roll)
   Multiply_2_1 = RigVMAggregateNode(
    A=@Remap.Result
    B=&RigVMFunction_MathDoubleRemap(SourceMaximum=8, SourceMinimum=-5, TargetMaximum=1, TargetMinimum=0, Value=@"Get Delta Ref".Result.Rotation.Pitch).Result
    C=&RigVMFunction_MathDoubleRemap(SourceMaximum=26, SourceMinimum=-6, TargetMaximum=1, TargetMinimum=0, Value=@"Get Delta Ref".Result.Rotation.Yaw).Result
   )
   n43 = &RigVMFunction_AlphaInterp(
    InterpSpeedDecreasing=10
    InterpSpeedIncreasing=25
    Scale=1
    bClampResult=False
    Value=&RigVMFunction_MathDoubleMul(A=@Multiply_2.Result, B=@Multiply_2_1.Result).Result
   )
   > &RigUnit_ModifyTransforms(
    ItemToModify=((Item=(Type=Bone,Name="dyn_pad_l"),Transform=(Rotation=(X=-0.0864,Y=-0.13,Z=-0.0114,W=0.9877),Translation=(X=0,Y=0,Z=0),Scale3D=(X=1,Y=1.05,Z=1))))
    WeightMinimum=0
    Weight=@n43.Result
   )
   > n44 = &RigUnit_SphericalPoseReader(
    ActiveRegionScaleFactors=(PositiveWidth=1,NegativeWidth=1,PositiveHeight=1,NegativeHeight=1)
    ActiveRegionSize=0.09
    Debug=(bDrawDebug=true,bDraw2D=False,bDrawLocalAxes=False,DebugScale=25,DebugSegments=20,DebugThickness=0.25)
    DriverAxis=(X=1,Y=0,Z=0)
    DriverItem=(Type=Bone,Name="upperarm_l")
    FalloffRegionScaleFactors=(PositiveWidth=1,NegativeWidth=1,PositiveHeight=1,NegativeHeight=1)
    FalloffSize=0.3
    FlipHeightScaling=False
    FlipWidthScaling=False
    OptionalParentItem=(Type=Bone,Name="spine_01")
    RotationOffset=(X=70,Y=0,Z=55)
   )
   > &RigUnit_ModifyTransforms(
    ItemToModify=((Item=(Type=Bone,Name="dyn_pad_l"),Transform=(Rotation=(X=0.2996,Y=-0.0262,Z=-0.0831,W=0.9501),Translation=(X=0,Y=0,Z=0.5),Scale3D=(X=1,Y=1.05,Z=1))))
    WeightMinimum=0
    Weight=@n44.OutputParam
   )
   n45 = &RigVMFunction_ControlFlowBranch(
    Condition=RigVMDispatchNode(B=1, A=RigVMVariableNode() {variable=isFrontend}.Value) {template="DISPATCH_RigVMDispatch_CoreEquals(in A,in B,out Result)"}.Result
    ExecuteContext=@RigVMFunction_Sequence_1_1.D
   )
   RigVMFunction_Sequence_5 = RigVMAggregateNode(ExecuteContext=@n45.True)
   Get Delta Ref_5_1_2 = RigVMFunctionReferenceNode(Item=(Type=Bone,Name="upperarm_r"), ExecuteContext=@RigVMFunction_Sequence_5.A)
   n46 = &RigVMFunction_AlphaInterp(
    InterpSpeedDecreasing=50
    InterpSpeedIncreasing=50
    Scale=1
    bClampResult=False
    Value=&RigVMFunction_MathDoubleRemap(SourceMaximum=30, SourceMinimum=0, TargetMaximum=1, TargetMinimum=0, Value=@"Get Delta Ref_5_1_2".Result.Rotation.Roll).Result
   )
   > &RigUnit_ModifyTransforms(
    ItemToModify=((Item=(Type=Bone,Name="dyn_pad_r"),Transform=(Rotation=(X=0.0435,Y=0.0871,Z=-0.0038,W=0.9952),Translation=(X=0.5,Y=0,Z=0),Scale3D=(X=1,Y=1.05,Z=1))))
    WeightMinimum=0
    Weight=@n46.Result
   )
   n47 = &RigVMFunction_AlphaInterp(
    InterpSpeedDecreasing=50
    InterpSpeedIncreasing=50
    Scale=1
    bClampResult=False
    Value=&RigVMFunction_MathDoubleRemap(SourceMaximum=0, SourceMinimum=-30, TargetMaximum=0, TargetMinimum=1, Value=@"Get Delta Ref_5_1_2".Result.Rotation.Roll).Result
   )
   > &RigUnit_ModifyTransforms(
    ItemToModify=((Item=(Type=Bone,Name="dyn_pad_r"),Transform=(Rotation=(X=0,Y=0,Z=0,W=1),Translation=(X=0,Y=0,Z=0),Scale3D=(X=1,Y=1,Z=1))))
    WeightMinimum=0
    Weight=@n47.Result
   )
   Get Delta Ref_5_1_1_1 = RigVMFunctionReferenceNode(Item=(Type=Bone,Name="upperarm_l"), ExecuteContext=@RigVMFunction_Sequence_5.B)
   n48 = &RigVMFunction_AlphaInterp(
    InterpSpeedDecreasing=50
    InterpSpeedIncreasing=50
    Scale=1
    bClampResult=False
    Value=&RigVMFunction_MathDoubleRemap(SourceMaximum=30, SourceMinimum=0, TargetMaximum=1, TargetMinimum=0, Value=@"Get Delta Ref_5_1_1_1".Result.Rotation.Roll).Result
   )
   > &RigUnit_ModifyTransforms(
    ItemToModify=((Item=(Type=Bone,Name="dyn_pad_l"),Transform=(Rotation=(X=0.0868,Y=-0.0076,Z=-0.0868,W=0.9924),Translation=(X=0,Y=0,Z=1),Scale3D=(X=1,Y=1.05,Z=1))))
    WeightMinimum=0
    Weight=@n48.Result
   )
   n49 = &RigVMFunction_AlphaInterp(
    InterpSpeedDecreasing=50
    InterpSpeedIncreasing=50
    Scale=1
    bClampResult=False
    Value=&RigVMFunction_MathDoubleRemap(SourceMaximum=0, SourceMinimum=-30, TargetMaximum=0, TargetMinimum=1, Value=@"Get Delta Ref_5_1_1_1".Result.Rotation.Roll).Result
   )
   > &RigUnit_ModifyTransforms(
    ItemToModify=((Item=(Type=Bone,Name="dyn_pad_l"),Transform=(Rotation=(X=0,Y=0,Z=0,W=1),Translation=(X=0,Y=0,Z=0),Scale3D=(X=1,Y=1,Z=1))))
    WeightMinimum=0
    Weight=@n49.Result
   )
   n50 = &RigVMFunction_ControlFlowBranch(
    Condition=RigVMDispatchNode(B=0, A=RigVMVariableNode() {variable=isEmote}.Value) {template="DISPATCH_RigVMDispatch_CoreEquals(in A,in B,out Result)"}.Result
    ExecuteContext=@RigVMFunction_Sequence_5.C
   )
   n51 = &RigUnit_SphericalPoseReader(
    ActiveRegionScaleFactors=(PositiveWidth=1,NegativeWidth=1,PositiveHeight=1,NegativeHeight=1)
    ActiveRegionSize=0.1
    Debug=(bDrawDebug=True,bDraw2D=False,bDrawLocalAxes=False,DebugScale=25,DebugSegments=20,DebugThickness=0.25)
    DriverAxis=(X=-1,Y=0,Z=0)
    DriverItem=(Type=Bone,Name="upperarm_r")
    FalloffRegionScaleFactors=(PositiveWidth=1,NegativeWidth=1,PositiveHeight=1,NegativeHeight=1)
    FalloffSize=0.2
    FlipHeightScaling=False
    FlipWidthScaling=False
    OptionalParentItem=(Type=Bone,Name="None")
    RotationOffset=(X=0,Y=60,Z=0)
    ExecutePin=@n50.Completed
   )
   > &RigUnit_ModifyTransforms(
    ItemToModify=((Item=(Type=Bone,Name="dyn_pad_r"),Transform=(Rotation=(X=-0.0872,Y=0,Z=0,W=0.9962),Translation=(X=-1,Y=0,Z=0.5),Scale3D=(X=1,Y=1,Z=1))))
    WeightMinimum=0
    Weight=@n51.OutputParam
   )
   > n52 = &RigUnit_SphericalPoseReader(
    ActiveRegionScaleFactors=(PositiveWidth=1,NegativeWidth=1,PositiveHeight=1,NegativeHeight=1)
    ActiveRegionSize=0.1
    Debug=(bDrawDebug=True,bDraw2D=False,bDrawLocalAxes=False,DebugScale=25,DebugSegments=20,DebugThickness=0.25)
    DriverAxis=(X=1,Y=0,Z=0)
    DriverItem=(Type=Bone,Name="upperarm_l")
    FalloffRegionScaleFactors=(PositiveWidth=1,NegativeWidth=1,PositiveHeight=1,NegativeHeight=1)
    FalloffSize=0.2
    FlipHeightScaling=False
    FlipWidthScaling=False
    OptionalParentItem=(Type=Bone,Name="None")
    RotationOffset=(X=0,Y=-90,Z=0)
   )
   > n53 = &RigUnit_SphericalPoseReader(
    ActiveRegionScaleFactors=(PositiveWidth=1,NegativeWidth=1,PositiveHeight=1,NegativeHeight=1)
    ActiveRegionSize=0.1
    Debug=(bDrawDebug=True,bDraw2D=False,bDrawLocalAxes=False,DebugScale=25,DebugSegments=20,DebugThickness=0.25)
    DriverAxis=(X=0,Y=1,Z=0)
    DriverItem=(Type=Bone,Name="upperarm_l")
    FalloffRegionScaleFactors=(PositiveWidth=1,NegativeWidth=1,PositiveHeight=1,NegativeHeight=1)
    FalloffSize=0.2
    FlipHeightScaling=False
    FlipWidthScaling=False
    OptionalParentItem=(Type=Bone,Name="None")
    RotationOffset=(X=100,Y=0,Z=0)
   )
   > &RigUnit_ModifyTransforms(
    ItemToModify=((Item=(Type=Bone,Name="dyn_pad_l"),Transform=(Rotation=(X=0,Y=0,Z=0,W=1),Translation=(X=0,Y=0.7,Z=0),Scale3D=(X=1,Y=1.05,Z=1))))
    WeightMinimum=0
    Weight=&RigVMFunction_MathDoubleMul(B=@n52.OutputParam, A=@n53.OutputParam).Result
   )
   n54 = &RigVMFunction_Sequence(ExecuteContext=@RigVMFunction_Sequence_7.D)
   n55 = RigVMVariableNode() {variable=isCrouching}
   n56 = RigVMVariableNode() {variable=isSkyDiving_NOT}
   n57 = &RigVMFunction_ControlFlowBranch(
    Condition=&RigVMFunction_MathBoolAnd(
     B=RigVMDispatchNode(B=1, A=@n55.Value) {template="DISPATCH_RigVMDispatch_CoreEquals(in A,in B,out Result)"}.Result
     A=RigVMDispatchNode(B=1, A=@n56.Value) {template="DISPATCH_RigVMDispatch_CoreEquals(in A,in B,out Result)"}.Result
    ).Result
    ExecuteContext=@n54.B
   )
   RigVMFunction_Sequence_1_3 = RigVMAggregateNode(ExecuteContext=@n57.True)
   Get Delta Ref_5_4 = RigVMFunctionReferenceNode(Item=(Type=Bone,Name="upperarm_l"), ExecuteContext=@RigVMFunction_Sequence_1_3.A)
   n58 = &RigVMFunction_Sequence(ExecuteContext=@"Get Delta Ref_5_4".ExecuteContext)
   &RigUnit_ModifyTransforms(
    ItemToModify=((Item=(Type=Bone,Name="dyn_pad_l"),Transform=(Rotation=(X=0,Y=0.0872,Z=0,W=0.9962),Translation=(X=3,Y=0,Z=4),Scale3D=(X=1,Y=1,Z=1))))
    WeightMinimum=0
    Weight=&RigVMFunction_MathDoubleRemap(SourceMaximum=75, SourceMinimum=0, TargetMaximum=1, TargetMinimum=0, Value=@"Get Delta Ref_5_4".Result.Rotation.Pitch).Result
    ExecutePin=@n58.A
   )
   n59 = &RigVMFunction_MathDoubleRemap(SourceMaximum=0, SourceMinimum=-65, TargetMaximum=0, TargetMinimum=1, Value=@"Get Delta Ref_5_4".Result.Rotation.Pitch)
   > &RigUnit_ModifyTransforms(
    ItemToModify=((Item=(Type=Bone,Name="dyn_pad_l"),Transform=(Rotation=(X=-0.0579,Y=-0.1691,Z=-0.0933,W=0.9795),Translation=(X=0.5,Y=0,Z=-1),Scale3D=(X=1,Y=1,Z=1))))
    WeightMinimum=0
    Weight=&RigVMFunction_MathDoubleMul(A=@n59.Result, B=RigVMVariableNode() {variable=isMelee_NOT}.Value).Result
   )
   > &RigUnit_ModifyTransforms(
    ItemToModify=((Item=(Type=Bone,Name="dyn_pad_l"),Transform=(Rotation=(X=0.122,Y=-0.0974,Z=-0.0747,W=0.9849),Translation=(X=2,Y=1,Z=1),Scale3D=(X=1,Y=1.05,Z=1))))
    WeightMinimum=0
    Weight=&RigVMFunction_MathDoubleMul(A=@n59.Result, B=RigVMVariableNode() {variable=isMelee}.Value).Result
   )
   &RigUnit_ModifyTransforms(
    ItemToModify=((Item=(Type=Bone,Name="dyn_pad_l"),Transform=(Rotation=(X=-0.5,Y=0,Z=0,W=0.866),Translation=(X=0,Y=0.5,Z=-1),Scale3D=(X=1,Y=1,Z=1))))
    WeightMinimum=0
    Weight=&RigVMFunction_MathDoubleRemap(SourceMaximum=-68, SourceMinimum=0, TargetMaximum=1, TargetMinimum=0, Value=@"Get Delta Ref_5_4".Result.Rotation.Yaw).Result
    ExecutePin=@n58.B
   )
   n60 = &RigVMFunction_MathDoubleRemap(SourceMaximum=110, SourceMinimum=90, TargetMaximum=1, TargetMinimum=0, Value=@"Get Delta Ref_5_4".Result.Rotation.Yaw)
   > &RigUnit_ModifyTransforms(
    ItemToModify=((Item=(Type=Bone,Name="dyn_pad_l"),Transform=(Rotation=(X=0.2588,Y=0.0045,Z=0.0169,W=0.9658),Translation=(X=0,Y=1.5,Z=0),Scale3D=(X=1,Y=1,Z=1))))
    WeightMinimum=0
    Weight=@n60.Result
   )
   > &RigUnit_ModifyTransforms(
    ItemToModify=((Item=(Type=Bone,Name="dyn_pad_l"),Transform=(Rotation=(X=0.1735,Y=-0.043,Z=0.0076,W=0.9839),Translation=(X=0,Y=0,Z=1),Scale3D=(X=1,Y=1.05,Z=1))))
    WeightMinimum=0
    Weight=&RigVMFunction_MathDoubleSub(
     B=@n60.Result
     A=&RigVMFunction_MathDoubleRemap(SourceMaximum=60, SourceMinimum=0, TargetMaximum=1, TargetMinimum=0, Value=@"Get Delta Ref_5_4".Result.Rotation.Yaw).Result
    ).Result
   )
   Get Delta Ref_5_2_2 = RigVMFunctionReferenceNode(Item=(Type=Bone,Name="upperarm_r"), ExecuteContext=@RigVMFunction_Sequence_1_3.B)
   n61 = &RigVMFunction_Sequence(ExecuteContext=@"Get Delta Ref_5_2_2".ExecuteContext)
   &RigUnit_ModifyTransforms(
    ItemToModify=((Item=(Type=Bone,Name="dyn_pad_r"),Transform=(Rotation=(X=0,Y=0.0872,Z=0,W=0.9962),Translation=(X=-3,Y=0,Z=-6),Scale3D=(X=1,Y=1.05,Z=1))))
    WeightMinimum=0
    Weight=&RigVMFunction_MathDoubleRemap(SourceMaximum=75, SourceMinimum=0, TargetMaximum=1, TargetMinimum=0, Value=@"Get Delta Ref_5_2_2".Result.Rotation.Pitch).Result
    ExecutePin=@n61.A
   )
   n62 = &RigVMFunction_MathDoubleRemap(SourceMaximum=0, SourceMinimum=-60, TargetMaximum=0, TargetMinimum=1, Value=@"Get Delta Ref_5_2_2".Result.Rotation.Pitch)
   > &RigUnit_ModifyTransforms(
    ItemToModify=((Item=(Type=Bone,Name="dyn_pad_r"),Transform=(Rotation=(X=0,Y=0,Z=-0.0872,W=0.9962),Translation=(X=-1.5,Y=0,Z=0),Scale3D=(X=1,Y=1.05,Z=1))))
    WeightMinimum=0
    Weight=&RigVMFunction_MathDoubleMul(B=@n62.Result, A=RigVMVariableNode() {variable=isMelee_NOT}.Value).Result
   )
   > &RigUnit_ModifyTransforms(
    ItemToModify=((Item=(Type=Bone,Name="dyn_pad_r"),Transform=(Rotation=(X=-0.173,Y=0.0151,Z=-0.0858,W=0.9811),Translation=(X=-1.5,Y=-1,Z=1),Scale3D=(X=1,Y=1.05,Z=1))))
    WeightMinimum=0
    Weight=&RigVMFunction_MathDoubleMul(A=@n62.Result, B=RigVMVariableNode() {variable=isMelee}.Value).Result
   )
   &RigUnit_ModifyTransforms(
    ItemToModify=((Item=(Type=Bone,Name="dyn_pad_r"),Transform=(Rotation=(X=-0.3827,Y=0,Z=0,W=0.9239),Translation=(X=0,Y=0.5,Z=1),Scale3D=(X=1,Y=1.05,Z=1))))
    WeightMinimum=0
    Weight=&RigVMFunction_MathDoubleRemap(SourceMaximum=-68, SourceMinimum=0, TargetMaximum=1, TargetMinimum=0, Value=@"Get Delta Ref_5_2_2".Result.Rotation.Yaw).Result
    ExecutePin=@n61.B
   )
   n63 = &RigVMFunction_MathDoubleRemap(SourceMaximum=131, SourceMinimum=80, TargetMaximum=1, TargetMinimum=0, Value=@"Get Delta Ref_5_2_2".Result.Rotation.Yaw)
   > &RigUnit_ModifyTransforms(
    ItemToModify=((Item=(Type=Bone,Name="dyn_pad_r"),Transform=(Rotation=(X=0.0858,Y=-0.173,Z=0.0151,W=0.9811),Translation=(X=1.5,Y=-2,Z=1),Scale3D=(X=1,Y=1,Z=1))))
    WeightMinimum=0
    Weight=@n63.Result
   )
   > &RigUnit_ModifyTransforms(
    ItemToModify=((Item=(Type=Bone,Name="dyn_pad_r"),Transform=(Rotation=(X=0.171,Y=-0.171,Z=0.0302,W=0.9698),Translation=(X=0,Y=0,Z=0),Scale3D=(X=1,Y=1.05,Z=1))))
    WeightMinimum=0
    Weight=&RigVMFunction_MathDoubleSub(
     B=@n63.Result
     A=&RigVMFunction_MathDoubleRemap(SourceMaximum=80, SourceMinimum=0, TargetMaximum=1, TargetMinimum=0, Value=@"Get Delta Ref_5_2_2".Result.Rotation.Yaw).Result
    ).Result
   )
   Get Delta Ref_5_1_3 = RigVMFunctionReferenceNode(Item=(Type=Bone,Name="upperarm_r"), ExecuteContext=@RigVMFunction_Sequence_1_3.C)
   n64 = &RigVMFunction_AlphaInterp(
    InterpSpeedDecreasing=50
    InterpSpeedIncreasing=50
    Scale=1
    bClampResult=False
    Value=&RigVMFunction_MathDoubleRemap(SourceMaximum=30, SourceMinimum=0, TargetMaximum=1, TargetMinimum=0, Value=@"Get Delta Ref_5_1_3".Result.Rotation.Roll).Result
   )
   n65 = RigVMVariableNode() {variable=isMelee_NOT}
   > &RigUnit_ModifyTransforms(
    ItemToModify=((Item=(Type=Bone,Name="dyn_pad_r"),Transform=(Rotation=(X=0.1305,Y=0,Z=0,W=0.9914),Translation=(X=0,Y=0,Z=0),Scale3D=(X=1,Y=1,Z=1))))
    WeightMinimum=0
    Weight=&RigVMFunction_MathDoubleMul(A=@n64.Result, B=@n65.Value).Result
   )
   n66 = &RigVMFunction_AlphaInterp(
    InterpSpeedDecreasing=50
    InterpSpeedIncreasing=50
    Scale=1
    bClampResult=False
    Value=&RigVMFunction_MathDoubleRemap(SourceMaximum=0, SourceMinimum=-30, TargetMaximum=0, TargetMinimum=1, Value=@"Get Delta Ref_5_1_3".Result.Rotation.Roll).Result
   )
   > &RigUnit_ModifyTransforms(
    ItemToModify=((Item=(Type=Bone,Name="dyn_pad_r"),Transform=(Rotation=(X=0.0872,Y=0,Z=0,W=0.9962),Translation=(X=0,Y=0,Z=0),Scale3D=(X=1,Y=1,Z=1))))
    WeightMinimum=0
    Weight=&RigVMFunction_MathDoubleMul(A=@n66.Result, B=@n65.Value).Result
   )
   n67 = &RigVMFunction_AlphaInterp(
    InterpSpeedDecreasing=50
    InterpSpeedIncreasing=50
    Scale=1
    bClampResult=False
    Value=&RigVMFunction_MathDoubleRemap(SourceMaximum=30, SourceMinimum=0, TargetMaximum=1, TargetMinimum=0, Value=@"Get Delta Ref_5_1_3".Result.Rotation.Roll).Result
   )
   > &RigUnit_ModifyTransforms(
    ItemToModify=((Item=(Type=Bone,Name="dyn_pad_r"),Transform=(Rotation=(X=0.1736,Y=0,Z=0,W=0.9848),Translation=(X=0,Y=0,Z=0),Scale3D=(X=1,Y=1,Z=1))))
    WeightMinimum=0
    Weight=&RigVMFunction_MathDoubleMul(A=@n67.Result, B=RigVMVariableNode() {variable=isMelee}.Value).Result
   )
   n68 = &RigVMFunction_AlphaInterp(
    InterpSpeedDecreasing=50
    InterpSpeedIncreasing=50
    Scale=1
    bClampResult=False
    Value=&RigVMFunction_MathDoubleRemap(SourceMaximum=0, SourceMinimum=-30, TargetMaximum=0, TargetMinimum=1, Value=@"Get Delta Ref_5_1_3".Result.Rotation.Roll).Result
   )
   > &RigUnit_ModifyTransforms(
    ItemToModify=((Item=(Type=Bone,Name="dyn_pad_r"),Transform=(Rotation=(X=-0.1305,Y=0,Z=0,W=0.9914),Translation=(X=0,Y=0,Z=0),Scale3D=(X=1,Y=1,Z=1))))
    WeightMinimum=0
    Weight=&RigVMFunction_MathDoubleMul(B=@n68.Result, A=RigVMVariableNode() {variable=isMelee}.Value).Result
   )
   Get Delta Ref_5_1_1_2 = RigVMFunctionReferenceNode(Item=(Type=Bone,Name="upperarm_l"), ExecuteContext=@RigVMFunction_Sequence_1_3.D)
   n69 = &RigVMFunction_AlphaInterp(
    InterpSpeedDecreasing=50
    InterpSpeedIncreasing=50
    Scale=1
    bClampResult=False
    Value=&RigVMFunction_MathDoubleRemap(SourceMaximum=30, SourceMinimum=0, TargetMaximum=1, TargetMinimum=0, Value=@"Get Delta Ref_5_1_1_2".Result.Rotation.Roll).Result
   )
   > &RigUnit_ModifyTransforms(
    ItemToModify=((Item=(Type=Bone,Name="dyn_pad_l"),Transform=(Rotation=(X=0.1736,Y=0,Z=0,W=0.9848),Translation=(X=0,Y=0,Z=0),Scale3D=(X=1,Y=1,Z=1))))
    WeightMinimum=0
    Weight=@n69.Result
   )
   n70 = &RigVMFunction_AlphaInterp(
    InterpSpeedDecreasing=50
    InterpSpeedIncreasing=50
    Scale=1
    bClampResult=False
    Value=&RigVMFunction_MathDoubleRemap(SourceMaximum=0, SourceMinimum=-30, TargetMaximum=0, TargetMinimum=1, Value=@"Get Delta Ref_5_1_1_2".Result.Rotation.Roll).Result
   )
   > &RigUnit_ModifyTransforms(
    ItemToModify=((Item=(Type=Bone,Name="dyn_pad_l"),Transform=(Rotation=(X=-0.0872,Y=0,Z=0,W=0.9962),Translation=(X=0,Y=0,Z=0),Scale3D=(X=1,Y=1,Z=1))))
    WeightMinimum=0
    Weight=@n70.Result
   )
   RigVMFunction_Sequence_1_1_1 = RigVMAggregateNode(ExecuteContext=@n54.A)
   RigVMFunction_MathBoolAnd_2_1 = RigVMAggregateNode(
    C=RigVMDispatchNode(B=1, A=RigVMVariableNode() {variable=isCrouching}.Value) {template="DISPATCH_RigVMDispatch_CoreEquals(in A,in B,out Result)"}.Result
    A=RigVMDispatchNode(B=1, A=RigVMVariableNode() {variable=isMelee}.Value) {template="DISPATCH_RigVMDispatch_CoreEquals(in A,in B,out Result)"}.Result
    B=RigVMDispatchNode(B=1, A=RigVMVariableNode() {variable=isSkyDiving_NOT}.Value) {template="DISPATCH_RigVMDispatch_CoreEquals(in A,in B,out Result)"}.Result
   )
   n71 = &RigVMFunction_ControlFlowBranch(Condition=@RigVMFunction_MathBoolAnd_2_1.Result, ExecuteContext=@RigVMFunction_Sequence_1_1_1.A)
   n72 = &RigUnit_SphericalPoseReader(
    ActiveRegionScaleFactors=(PositiveWidth=1,NegativeWidth=1,PositiveHeight=1,NegativeHeight=1)
    ActiveRegionSize=0.1
    Debug=(bDrawDebug=true,bDraw2D=False,bDrawLocalAxes=False,DebugScale=25,DebugSegments=20,DebugThickness=0.25)
    DriverAxis=(X=1,Y=0,Z=0)
    DriverItem=(Type=Bone,Name="upperarm_l")
    FalloffRegionScaleFactors=(PositiveWidth=1,NegativeWidth=1,PositiveHeight=1,NegativeHeight=1)
    FalloffSize=0.2
    FlipHeightScaling=False
    FlipWidthScaling=False
    OptionalParentItem=(Type=Bone,Name="spine_05")
    RotationOffset=(X=90,Y=0,Z=-60)
    ExecutePin=@n71.True
   )
   > &RigUnit_ModifyTransforms(
    ItemToModify=((Item=(Type=Bone,Name="dyn_pad_l"),Transform=(Rotation=(X=0,Y=0,Z=0,W=1),Translation=(X=0,Y=0,Z=1),Scale3D=(X=1,Y=1,Z=1))))
    WeightMinimum=0
    Weight=@n72.OutputParam
   )
   > n73 = &RigUnit_SphericalPoseReader(
    ActiveRegionScaleFactors=(PositiveWidth=1,NegativeWidth=1,PositiveHeight=1,NegativeHeight=1)
    ActiveRegionSize=0.1
    Debug=(bDrawDebug=true,bDraw2D=False,bDrawLocalAxes=False,DebugScale=25,DebugSegments=20,DebugThickness=0.25)
    DriverAxis=(X=1,Y=0,Z=0)
    DriverItem=(Type=Bone,Name="upperarm_l")
    FalloffRegionScaleFactors=(PositiveWidth=1,NegativeWidth=1,PositiveHeight=1,NegativeHeight=1)
    FalloffSize=0.2
    FlipHeightScaling=False
    FlipWidthScaling=False
    OptionalParentItem=(Type=Bone,Name="spine_01")
    RotationOffset=(X=130,Y=0,Z=-10)
   )
   > &RigUnit_ModifyTransforms(
    ItemToModify=((Item=(Type=Bone,Name="dyn_pad_l"),Transform=(Rotation=(X=-0.0872,Y=0,Z=0,W=0.9962),Translation=(X=0,Y=-1.5,Z=0),Scale3D=(X=1,Y=1.05,Z=1))))
    WeightMinimum=0
    Weight=@n73.OutputParam
   )
   n74 = &RigVMFunction_ControlFlowBranch(
    Condition=RigVMDispatchNode(B=1, A=RigVMVariableNode() {variable=isPistol}.Value) {template="DISPATCH_RigVMDispatch_CoreEquals(in A,in B,out Result)"}.Result
    ExecuteContext=@RigVMFunction_Sequence_1_1_1.B
   )
   Get Delta Ref_5_1_3_1 = RigVMFunctionReferenceNode(Item=(Type=Bone,Name="clavicle_r"), ExecuteContext=@n74.True)
   > Get Delta Ref_5_2_2_1 = RigVMFunctionReferenceNode(Item=(Type=Bone,Name="lowerarm_r"))
   > &RigUnit_ModifyTransforms(
    ItemToModify=((Item=(Type=Bone,Name="dyn_pad_r"),Transform=(Rotation=(X=0.1337,Y=0.0317,Z=-0.092,W=0.9862),Translation=(X=0,Y=0.5,Z=-1),Scale3D=(X=1,Y=1,Z=1))))
    WeightMinimum=0
    Weight=&RigVMFunction_MathDoubleMul(
     A=&RigVMFunction_MathDoubleRemap(SourceMaximum=4, SourceMinimum=19, TargetMaximum=1, TargetMinimum=0, Value=@"Get Delta Ref_5_1_3_1".Result.Rotation.Pitch).Result
     B=&RigVMFunction_MathDoubleRemap(SourceMaximum=30, SourceMinimum=90, TargetMaximum=1, TargetMinimum=0, Value=@"Get Delta Ref_5_2_2_1".Result.Rotation.Yaw).Result
    ).Result
   )
   &RigUnit_ModifyTransforms(
    ExecutePin=()
    ItemToModify=((Item=(Type=Bone,Name="dyn_pad_r"),Transform=(Rotation=(X=0.0435,Y=0.0038,Z=0.0871,W=0.9952),Translation=(X=-1.5,Y=0,Z=0),Scale3D=(X=1,Y=1.05,Z=1))))
    Weight=1
    WeightMinimum=0
   )
   RigVMVariableNode() {variable=isCrouching_NOT}
  ]}
]
hierarchy [
 root {loc=<0 0 0>, rot=<0 0 0 1>}
 pelvis {parent=root, loc=<0 -0.7071 93.1191>, rot=<0.0127 0.707 -0.0127 -0.707>}
 spine_01 {parent=pelvis, loc=<6.7494 -0.6388 0>, rot=<0 0 0.034 -0.9994>}
 spine_02 {parent=spine_01, loc=<7.2518 -0.035 0>, rot=<0 0 0.0238 -0.9997>}
 spine_03 {parent=spine_02, loc=<6.0104 0 0>, rot=<0 0 0.0149 -0.9999>}
 spine_04 {parent=spine_03, loc=<8.1024 0 0>, rot=<0 0 -0.0132 -0.9999>}
 spine_05 {parent=spine_04, loc=<10.0146 0 0>, rot=<0 0 -0.0015 -1>}
 attach_cape {parent=spine_05, loc=<9.017 -11.714 0>, rot=<0.0616 -0.7044 -0.0616 -0.7044>}
 neck_01 {parent=spine_05, loc=<11.3138 0.1529 0>, rot=<0 0 -0.2003 -0.9797>}
 neck_02 {parent=neck_01, loc=<4.5627 0 0>, rot=<0 0 0.0504 -0.9987>}
 head {parent=neck_02, loc=<5.3914 0 0>, rot=<0 0 0.115 -0.9934>}
 attach_backpack {parent=spine_05, loc=<-0.123 -13.473 0>, rot=<-0.0093 -0.707 0.0093 -0.707>}
 clavicle_l {parent=spine_05, loc=<4.3575 -0.1466 -2.3223>, rot=<0.0465 -0.7289 -0.009 -0.6829>}
 upperarm_l {parent=clavicle_l, loc=<15.5111 0 0>, rot=<0.0226 -0.3234 -0.0088 -0.9459>}
 lowerarm_l {parent=upperarm_l, loc=<26.6514 0 0>, rot=<0.0391 -0.0099 -0.2453 -0.9686>}
 hand_l {parent=lowerarm_l, loc=<22.3491 0 0>, rot=<-0.7081 -0.0507 0.0975 -0.6975>}
 thumb_01_l {parent=hand_l, loc=<1.3118 -0.3191 -2.459>, rot=<0.6546 -0.4042 0.0664 -0.6354>}
 thumb_02_l {parent=thumb_01_l, loc=<5.9336 -0.0448 0.0401>, rot=<-0.0037 0.0213 0.298 -0.9543>}
 thumb_03_l {parent=thumb_02_l, loc=<2.2223 0.3682 -0.0907>, rot=<0.0118 -0.013 0.0066 -0.9998>}
 index_metacarpal_l {parent=hand_l, loc=<1.6535 0.0183 -2.96>, rot=<0.0158 -0.0124 -0.0173 -0.9996>}
 index_01_l {parent=index_metacarpal_l, loc=<8.3008 0.0357 -0.0524>, rot=<-0.0044 0.0485 0.0924 -0.9945>}
 index_02_l {parent=index_01_l, loc=<3.3587 -0.3004 0.0452>, rot=<-0.002 0.0027 0.3225 -0.9466>}
 index_03_l {parent=index_02_l, loc=<2.5578 -0.123 -0.0252>, rot=<-0.002 0.0067 0.2558 -0.9667>}
 middle_metacarpal_l {parent=hand_l, loc=<2.1856 -0.1725 -0.5984>, rot=<0.006 0.0168 -0.0329 -0.9993>}
 middle_01_l {parent=middle_metacarpal_l, loc=<7.3275 -0.2447 0.0129>, rot=<-0.0208 0.0042 0.0858 -0.9961>}
 middle_02_l {parent=middle_01_l, loc=<4.0123 0.1401 0.1175>, rot=<-0.0218 0.0258 0.3512 -0.9357>}
 middle_03_l {parent=middle_02_l, loc=<2.6077 0.0534 0.1373>, rot=<0.0082 0.0188 0.1708 -0.9851>}
 ring_metacarpal_l {parent=hand_l, loc=<2.2074 -0.3625 1.3638>, rot=<-0.0284 0.0722 -0.0458 -0.9959>}
 ring_01_l {parent=ring_metacarpal_l, loc=<7.0295 -0.6197 -0.1446>, rot=<-0.0489 -0.0252 0.1892 -0.9804>}
 ring_02_l {parent=ring_01_l, loc=<3.6174 0.4647 -0.1027>, rot=<0 0 0.2806 -0.9598>}
 ring_03_l {parent=ring_02_l, loc=<2.4152 0.114 0.0199>, rot=<0 0 0.1189 -0.9929>}
 pinky_metacarpal_l {parent=hand_l, loc=<1.7616 -0.9753 3.3814>, rot=<-0.0397 0.0814 -0.069 -0.9935>}
 pinky_01_l {parent=pinky_metacarpal_l, loc=<7.1709 -0.4831 -0.1094>, rot=<-0.0654 -0.0035 0.2242 -0.9723>}
 pinky_02_l {parent=pinky_01_l, loc=<2.8274 0.0893 -0.1043>, rot=<0.0037 0.0011 0.2854 -0.9584>}
 pinky_03_l {parent=pinky_02_l, loc=<2.2046 0.0307 -0.0078>, rot=<-0.0002 0.0098 0.1445 -0.9895>}
 deform_wrist_dn_base_l {parent=hand_l, loc=<0 -2.2342 0>, rot=<0 0 0 -1>}
 deform_wrist_dn_l {parent=deform_wrist_dn_base_l, loc=<0 0 0>, rot=<0 0 0 1>}
 deform_wrist_up_base_l {parent=hand_l, loc=<0 2.2682 0>, rot=<0 0 0 -1>}
 deform_wrist_up_l {parent=deform_wrist_up_base_l, loc=<0 0 0>, rot=<0 0 0 1>}
 lowerarm_twist_02_l {parent=lowerarm_l, loc=<9.6199 0 0>, rot=<0 0 0 -1>}
 dyn_elbow_pad_l {parent=lowerarm_twist_02_l, loc=<-8.2926 1.1915 5.4011>, rot=<0 0 0 -1>}
 lowerarm_twist_01_l {parent=lowerarm_l, loc=<17.8549 0 0>, rot=<0 0 0 -1>}
 deform_elbow_in_base_l {parent=lowerarm_l, loc=<0.2112 5.259 0.8274>, rot=<0 0 0 -1>}
 deform_elbow_in_l {parent=deform_elbow_in_base_l, loc=<0 0 0>, rot=<0 0 0 1>}
 deform_elbow_out_base_l {parent=lowerarm_l, loc=<0 -4 0>, rot=<0 0 0 -1>}
 deform_elbow_out_l {parent=deform_elbow_out_base_l, loc=<0 0 0>, rot=<0 0 0 1>}
 upperarm_twist_01_l {parent=upperarm_l, loc=<0.0041 0 0>, rot=<0 0 0 1>}
 deform_upperarm_bk_base_l {parent=upperarm_twist_01_l, loc=<1.0813 -7.2193 -1.9453>, rot=<0 0 0 1>}
 deform_upperarm_bk_l {parent=deform_upperarm_bk_base_l, loc=<0 0 0>, rot=<0 0 0 1>}
 deform_upperarm_fr_base_l {parent=upperarm_twist_01_l, loc=<-0.3691 7.0506 -2.765>, rot=<0 0 0 1>}
 deform_upperarm_up_base_l {parent=upperarm_twist_01_l, loc=<-3.2216 -1.1391 3.3311>, rot=<0 0 0 1>}
 deform_upperarm_up_l {parent=deform_upperarm_up_base_l, loc=<0 0 0>, rot=<0 0 0 1>}
 dyn_pad_base_l {parent=upperarm_twist_01_l, loc=<-0.0041 0 0>, rot=<0 0 0 1>}
 dyn_pad_l {parent=dyn_pad_base_l, loc=<0 0 0>, rot=<0 0 0 1>}
 dyn_pad_bass_01_l {parent=dyn_pad_l, loc=<0.4746 3.3426 6.3526>, rot=<-0.0126 0.5914 -0.3371 -0.7324>}
 dyn_pad_bass_02_l {parent=dyn_pad_l, loc=<0.9292 -5.0319 6.0154>, rot=<-0.0447 0.5987 0.3584 -0.7149>}
 upperarm_twist_02_l {parent=upperarm_l, loc=<13.7941 0 0>, rot=<0 0 0 1>}
 dyn_knot_l {parent=clavicle_l, loc=<8.258 6.4981 2.6157>, rot=<-0.1517 0.2971 -0.4944 -0.8027>}
 clavicle_r {parent=spine_05, loc=<4.3575 -0.1466 2.3223>, rot=<0.7289 0.0465 -0.6829 0.009>}
 upperarm_r {parent=clavicle_r, loc=<-15.5111 0 0>, rot=<0.0226 -0.3234 -0.0088 -0.9459>}
 lowerarm_r {parent=upperarm_r, loc=<-26.6515 0 0>, rot=<0.0391 -0.0099 -0.2453 -0.9686>}
 hand_r {parent=lowerarm_r, loc=<-22.3492 0 0>, rot=<-0.7081 -0.0507 0.0975 -0.6975>}
 thumb_01_r {parent=hand_r, loc=<-1.3118 0.3191 2.459>, rot=<0.6546 -0.4042 0.0664 -0.6354>}
 thumb_02_r {parent=thumb_01_r, loc=<-5.9336 0.0448 -0.0401>, rot=<-0.0037 0.0213 0.298 -0.9543>}
 thumb_03_r {parent=thumb_02_r, loc=<-2.2223 -0.3682 0.0907>, rot=<0.0118 -0.013 0.0066 -0.9998>}
 index_metacarpal_r {parent=hand_r, loc=<-1.6535 -0.0183 2.96>, rot=<0.0158 -0.0124 -0.0173 -0.9996>}
 index_01_r {parent=index_metacarpal_r, loc=<-8.3008 -0.0357 0.0524>, rot=<-0.0044 0.0485 0.0924 -0.9945>}
 index_02_r {parent=index_01_r, loc=<-3.3587 0.3004 -0.0452>, rot=<-0.002 0.0027 0.3225 -0.9466>}
 index_03_r {parent=index_02_r, loc=<-2.5578 0.123 0.0252>, rot=<-0.002 0.0067 0.2558 -0.9667>}
 middle_metacarpal_r {parent=hand_r, loc=<-2.1856 0.1725 0.5984>, rot=<0.006 0.0168 -0.0329 -0.9993>}
 middle_01_r {parent=middle_metacarpal_r, loc=<-7.3275 0.2447 -0.0129>, rot=<-0.0208 0.0042 0.0858 -0.9961>}
 middle_02_r {parent=middle_01_r, loc=<-4.0123 -0.1401 -0.1175>, rot=<-0.0218 0.0258 0.3512 -0.9357>}
 middle_03_r {parent=middle_02_r, loc=<-2.6077 -0.0534 -0.1373>, rot=<0.0082 0.0188 0.1708 -0.9851>}
 ring_metacarpal_r {parent=hand_r, loc=<-2.2074 0.3625 -1.3638>, rot=<-0.0284 0.0722 -0.0458 -0.9959>}
 ring_01_r {parent=ring_metacarpal_r, loc=<-7.0295 0.6197 0.1446>, rot=<-0.0489 -0.0252 0.1892 -0.9804>}
 ring_02_r {parent=ring_01_r, loc=<-3.6174 -0.4647 0.1027>, rot=<0 0 0.2806 -0.9598>}
 ring_03_r {parent=ring_02_r, loc=<-2.4152 -0.114 -0.0199>, rot=<0 0 0.1189 -0.9929>}
 pinky_metacarpal_r {parent=hand_r, loc=<-1.7616 0.9753 -3.3814>, rot=<-0.0397 0.0814 -0.069 -0.9935>}
 pinky_01_r {parent=pinky_metacarpal_r, loc=<-7.1709 0.4831 0.1094>, rot=<-0.0654 -0.0035 0.2242 -0.9723>}
 pinky_02_r {parent=pinky_01_r, loc=<-2.8274 -0.0893 0.1043>, rot=<0.0037 0.0011 0.2854 -0.9584>}
 pinky_03_r {parent=pinky_02_r, loc=<-2.2046 -0.0307 0.0078>, rot=<-0.0002 0.0098 0.1445 -0.9895>}
 deform_wrist_dn_base_r {parent=hand_r, loc=<-0.0003 2.2345 -0.0001>, rot=<0 0 0 -1>}
 deform_wrist_dn_r {parent=deform_wrist_dn_base_r, loc=<0 0 0>, rot=<0 0 0 1>}
 deform_wrist_up_base_r {parent=hand_r, loc=<0.0004 -2.2684 0.0001>, rot=<0 0 0 -1>}
 deform_wrist_up_r {parent=deform_wrist_up_base_r, loc=<0 0 0>, rot=<0 0 0 1>}
 lowerarm_twist_02_r {parent=lowerarm_r, loc=<-9.6199 0 0>, rot=<0 0 0 -1>}
 dyn_elbow_pad_r {parent=lowerarm_twist_02_r, loc=<7.8534 -1.1257 -4.23>, rot=<-0.007 -0.0129 -0.0019 -0.9999>}
 lowerarm_twist_01_r {parent=lowerarm_r, loc=<-17.8549 0 0>, rot=<0 0 0 -1>}
 deform_elbow_in_base_r {parent=lowerarm_r, loc=<-0.211 -5.2591 -0.8275>, rot=<0 0 0 1>}
 deform_elbow_in_r {parent=deform_elbow_in_base_r, loc=<-0.0001 0 0.0002>, rot=<0 0 0 1>}
 deform_elbow_out_base_r {parent=lowerarm_r, loc=<0 4 -0.0001>, rot=<0 0 0 1>}
 deform_elbow_out_r {parent=deform_elbow_out_base_r, loc=<0 0 0>, rot=<0 0 0 1>}
 upperarm_twist_01_r {parent=upperarm_r, loc=<-0.0041 0 0>, rot=<0 0 0 1>}
 deform_upperarm_bk_base_r {parent=upperarm_twist_01_r, loc=<-1.0814 7.2192 1.9455>, rot=<0 0 0 1>}
 deform_upperarm_bk_r {parent=deform_upperarm_bk_base_r, loc=<0 0 0>, rot=<0 0 0 1>}
 deform_upperarm_fr_base_r {parent=upperarm_twist_01_r, loc=<0.3689 -7.0506 2.7652>, rot=<0 0 0 1>}
 deform_upperarm_up_base_r {parent=upperarm_twist_01_r, loc=<3.2214 1.1391 -3.3308>, rot=<0 0 0 1>}
 deform_upperarm_up_r {parent=deform_upperarm_up_base_r, loc=<0 0 0>, rot=<0 0 0 -1>}
 upperarm_twist_02_r {parent=upperarm_r, loc=<-13.7941 0 0>, rot=<0 0 0 1>}
 dyn_knot_r {parent=clavicle_r, loc=<-8.258 -6.4981 -2.6162>, rot=<-0.1517 0.2971 -0.4944 -0.8027>}
 dyn_pad_base_r {parent=clavicle_r, loc=<-15.5112 0 -0.0001>, rot=<0.0225 -0.3375 -0.0092 -0.941>}
 dyn_pad_r {parent=dyn_pad_base_r, loc=<0 0 0>, rot=<0 0 0 1>}
 dyn_pad_bass_01_r {parent=dyn_pad_r, loc=<-0.2909 -3.3424 -6.3634>, rot=<0.02 0.5886 -0.3615 -0.7228>}
 dyn_pad_bass_02_r {parent=dyn_pad_r, loc=<-0.7428 5.0328 -6.0406>, rot=<-0.0991 0.6039 0.3673 -0.7004>}
 thigh_l {parent=pelvis, loc=<-2.5693 -0.0109 -7.7501>, rot=<0.0012 0.0572 0.0155 -0.9982>}
 calf_l {parent=thigh_l, loc=<-39.0484 0 0>, rot=<0.0006 -0.0161 -0.0391 -0.9991>}
 foot_l {parent=calf_l, loc=<-42.561 0.1077 0.4292>, rot=<0.0418 -0.0455 0.0343 -0.9975>}
 ball_l {parent=foot_l, loc=<-6.1599 13.8264 0.6806>, rot=<0.0001 -0.0001 -0.7063 -0.7079>}
 deform_knee_in_base_l {parent=calf_l, loc=<1.298 -5.2957 -0.0496>, rot=<0 0 0 -1>}
 deform_knee_in_l {parent=deform_knee_in_base_l, loc=<0 0 0>, rot=<0 0 0 1>}
 deform_knee_out_base_l {parent=calf_l, loc=<0 4 0>, rot=<0 0 0 -1>}
 deform_knee_out_l {parent=deform_knee_out_base_l, loc=<0 0 0>, rot=<0 0 0 1>}
 calf_twist_01_l {parent=calf_l, loc=<-35.813 -0.0967 0.0397>, rot=<0.0237 0.0095 -0.0035 -0.9997>}
 dyn_calf_twist_left_l {parent=calf_l, loc=<-38.2997 1.2177 -6.4639>, rot=<0.0041 -0.9968 0.0644 0.0466>}
 dyn_calf_twist_back_l {parent=calf_l, loc=<-39.6085 -7.9636 0.8174>, rot=<0.0314 -0.7374 0.6745 0.017>}
 dyn_calf_twist_front_l {parent=calf_l, loc=<-35.3573 9.8662 0.3218>, rot=<-0.0543 -0.732 -0.6785 -0.0282>}
 dyn_calf_twist_right_l {parent=calf_l, loc=<-37.4031 1.2808 9.002>, rot=<0.0155 -0.018 0.9981 0.0566>}
 dyn_calf_twist_upper_l {parent=calf_l, loc=<-18.8064 -0.0967 0.0397>, rot=<0.0237 0.0095 -0.0035 -0.9997>}
 thigh_twist_01_l {parent=thigh_l, loc=<-21.6303 0 0>, rot=<0.0013 0 0 -1>}
 deform_groin_base_l {parent=thigh_l, loc=<-2.1908 8.1142 -1.7442>, rot=<0 0 0 -1>}
 deform_groin_l {parent=deform_groin_base_l, loc=<0 0 0>, rot=<0 0 0 1>}
 thigh_r {parent=pelvis, loc=<-2.5691 -0.0109 7.7501>, rot=<0.0572 -0.0012 0.9982 0.0155>}
 calf_r {parent=thigh_r, loc=<39.0487 0 0>, rot=<0.0006 -0.0161 -0.0391 -0.9991>}
 foot_r {parent=calf_r, loc=<42.561 -0.1077 -0.4292>, rot=<0.0418 -0.0455 0.0343 -0.9975>}
 ball_r {parent=foot_r, loc=<6.1599 -13.8264 -0.6806>, rot=<0.0001 -0.0001 -0.7063 -0.7079>}
 calf_twist_01_r {parent=calf_r, loc=<35.813 0.0967 -0.0397>, rot=<0.0237 0.0095 -0.0035 -0.9997>}
 dyn_calf_twist_back_r {parent=calf_r, loc=<39.8508 8.2614 -0.3381>, rot=<0.0314 -0.7374 0.6745 0.017>}
 dyn_calf_twist_right_r {parent=calf_r, loc=<37.5774 -1.0136 -8.5945>, rot=<0.0155 -0.018 0.9981 0.0566>}
 dyn_calf_twist_left_r {parent=calf_r, loc=<38.2696 -1.3634 8.1853>, rot=<0.0041 -0.9968 0.0644 0.0466>}
 dyn_calf_twist_front_r {parent=calf_r, loc=<35.8611 -9.3526 -0.3939>, rot=<-0.0543 -0.732 -0.6785 -0.0282>}
 deform_knee_in_base_r {parent=calf_r, loc=<-1.2981 5.2958 0.0496>, rot=<0 0 0 1>}
 deform_knee_in_r {parent=deform_knee_in_base_r, loc=<0 0 0>, rot=<0 0 0 1>}
 deform_knee_out_base_r {parent=calf_r, loc=<-0.0002 -4 0>, rot=<0 0 0 -1>}
 deform_knee_out_r {parent=deform_knee_out_base_r, loc=<0 0 0>, rot=<0 0 0 1>}
 dyn_calf_twist_upper_r {parent=calf_r, loc=<18.8064 0.0967 -0.0397>, rot=<0.0237 0.0095 -0.0035 -0.9997>}
 thigh_twist_01_r {parent=thigh_r, loc=<21.6303 0 0>, rot=<0.0013 0 0 -1>}
 deform_groin_base_r {parent=thigh_r, loc=<2.1908 -8.1142 1.7442>, rot=<0 0 0 -1>}
 deform_groin_r {parent=deform_groin_base_r, loc=<0 0 0>, rot=<0 0 0 1>}
 deform_glute_base_l {parent=pelvis, loc=<-8.8064 -8.8222 -8.4287>, rot=<0.0275 0.0501 -0.4472 -0.8926>}
 deform_glute_l {parent=deform_glute_base_l, loc=<0 0 0>, rot=<0 0 0 1>}
 deform_glute_base_r {parent=pelvis, loc=<-8.8061 -8.8222 8.4287>, rot=<0.0501 -0.0275 0.8926 -0.4472>}
 deform_glute_r {parent=deform_glute_base_r, loc=<0 0 0>, rot=<0 0 0 1>}
]
variables [
 isAWS {type=double, default=0}
 isAWS_NOT {type=double, default=0}
 isCrouching {type=double, default=0}
 isCrouching_NOT {type=double, default=0}
 isDBNO {type=double, default=0}
 isDBNO_NOT {type=double, default=0}
 isEmote {type=double, default=0}
 isEmote_NOT {type=double, default=0}
 isFalling {type=bool, default=False}
 isFrontend {type=double, default=0}
 isFrontend_NOT {type=double, default=0}
 isJumpting {type=bool, default=False}
 isMelee {type=double, default=0}
 isMelee_NOT {type=double, default=0}
 isMoving {type=double, default=0}
 isMoving_NOT {type=double, default=0}
 isParachuteOpened {type=double, default=0}
 isParachuteOpened_NOT {type=double, default=0}
 isPistol {type=double, default=0}
 isPistol_NOT {type=double, default=0}
 isRail {type=double, default=0}
 isRail_NOT {type=double, default=0}
 isSkyDiving {type=double, default=0}
 isSkyDiving_NOT {type=double, default=0}
]
