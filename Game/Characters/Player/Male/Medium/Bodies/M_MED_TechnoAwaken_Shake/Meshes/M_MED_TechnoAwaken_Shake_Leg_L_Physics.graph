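# Physics asset for the left leg mesh of the TechnoAwaken Shake medium male player body.
# Capsules cover the calf and foot, while five dyn_calf_twist bones carry small capsules or, for the upper one, a sphere.
# Each links to the calf: back, front, left and right keep only swing2 unlocked, while upper locks all angular axes.
# Twist shape names carry right-side suffixes, suggesting mirroring from the right leg.
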PhysicsAsset /Game/Characters/Player/Male/Medium/Bodies/M_MED_TechnoAwaken_Shake/Meshes/M_MED_TechnoAwaken_Shake_Leg_L_Physics
bodies {
 calf_l [{shapeType=Capsule, length=41.5486, radius=8.785, center=<-24.001 -1.2457 0.3984>, rotation=<rot -0.0071 89.3909 -90.1098>}]
 dyn_calf_twist_back_l [{
   shapeType=Capsule
   length=4.63
   radius=2.63
   shapeName=dyn_calf_twist_back_r_capsule
   center=<3.3496 -0.0001 4.5264>
   rotation=<rot -87.932 179.8064 -179.8062>
  }]
 dyn_calf_twist_front_l [{
   shapeType=Capsule
   length=1.75
   radius=2.75
   shapeName=dyn_calf_twist_front_r_capsule
   center=<1.8201 0 0.4511>
   rotation=<rot 84.2253 -0.0001 -0.0002>
  }]
 dyn_calf_twist_left_l [{
   shapeType=Capsule
   length=6.255
   radius=2.63
   shapeName=dyn_calf_twist_right_r_capsule
   center=<1.5318 0.0003 4.8631>
   rotation=<rot -84.906 0.0116 0.0001>
  }]
 dyn_calf_twist_right_l [{
   shapeType=Capsule
   length=5.5
   radius=2.38
   shapeName=dyn_calf_twist_left_r_capsule
   center=<2.0928 0 2.4535>
   rotation=<rot 87.6446 -0.0008 -0.0004>
  }]
 dyn_calf_twist_upper_l [{shapeType=Sphere, radius=3.5793, shapeName=dyn_calf_twist_upper_r_sphere}]
 foot_l [{shapeType=Capsule, length=17.4494, radius=7.9828, center=<-3.7028 6.5228 0.0583>, rotation=<rot 0 0.8667 -89.4543>}]
}
constraints [
 {bone1Name=foot_l, bone2Name=calf_l, swing1LimitDegrees=45, swing2LimitDegrees=45, twistLimitDegrees=45}
 {
  bone1Name=dyn_calf_twist_back_l
  bone2Name=calf_l
  swing1LimitDegrees=45
  swing1Motion=Locked
  swing2LimitDegrees=10
  twistLimitDegrees=10
  twistMotion=Locked
 }
 {
  bone1Name=dyn_calf_twist_front_l
  bone2Name=calf_l
  swing1LimitDegrees=45
  swing1Motion=Locked
  swing2LimitDegrees=10
  twistLimitDegrees=10
  twistMotion=Locked
 }
 {
  bone1Name=dyn_calf_twist_left_l
  bone2Name=calf_l
  swing1LimitDegrees=45
  swing1Motion=Locked
  swing2LimitDegrees=10
  twistLimitDegrees=10
  twistMotion=Locked
 }
 {
  bone1Name=dyn_calf_twist_right_l
  bone2Name=calf_l
  swing1LimitDegrees=45
  swing1Motion=Locked
  swing2LimitDegrees=5
  twistLimitDegrees=10
  twistMotion=Locked
 }
 {
  bone1Name=dyn_calf_twist_upper_l
  bone2Name=calf_l
  swing1LimitDegrees=45
  swing1Motion=Locked
  swing2LimitDegrees=45
  swing2Motion=Locked
  twistLimitDegrees=20
  twistMotion=Locked
 }
]
