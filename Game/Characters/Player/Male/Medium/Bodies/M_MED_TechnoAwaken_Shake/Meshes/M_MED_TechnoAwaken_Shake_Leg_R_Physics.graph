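# Physics asset for the right leg mesh of the medium male TechnoAwaken Shake player body, covering the calf, foot and five dynamic calf twist bones.
# Capsules approximate the calf, foot and back, front, left and right twist bones, while a sphere covers the upper twist bone.
# All bones constrain directly to calf_r: the upper twist bone is rotationally locked, and the other four lock swing1 and twist, leaving only narrow swing2 movement.
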PhysicsAsset /Game/Characters/Player/Male/Medium/Bodies/M_MED_TechnoAwaken_Shake/Meshes/M_MED_TechnoAwaken_Shake_Leg_R_Physics
bodies {
 calf_r [{shapeType=Capsule, length=40.436, radius=9.4353, center=<23.2543 1.0567 -0.168>, rotation=<rot 0.0211 88.5495 -89.6133>}]
 dyn_calf_twist_back_r [{
   shapeType=Capsule
   length=4.63
   radius=2.755
   shapeName=dyn_calf_twist_back_r_capsule
   center=<-2.6748 0 -4.569>
   rotation=<rot -86.8225 179.9979 -179.9969>
  }]
 dyn_calf_twist_front_r [{
   shapeType=Capsule
   length=2.75
   radius=2.75
   shapeName=dyn_calf_twist_front_r_capsule
   center=<-2.1656 0 -2.0711>
   rotation=<rot -84.5924 179.9999 -179.9999>
  }]
 dyn_calf_twist_left_r [{
   shapeType=Capsule
   length=5.5
   radius=2.38
   shapeName=dyn_calf_twist_left_r_capsule
   center=<-2.2488 0 -3.8868>
   rotation=<rot 84.6392 -179.2117 -179.2155>
  }]
 dyn_calf_twist_right_r [{
   shapeType=Capsule
   length=6.255
   radius=2.63
   shapeName=dyn_calf_twist_right_r_capsule
   center=<-3.1544 0 -3.4729>
   rotation=<rot -86.7528 179.9992 -179.9996>
  }]
 dyn_calf_twist_upper_r [{shapeType=Sphere, radius=3.5793, shapeName=dyn_calf_twist_upper_r_sphere, center=<0.167 0.1883 -0.12>}]
 foot_r [{shapeType=Capsule, length=17.4494, radius=7.9828, center=<2.8738 -6.5742 -0.1097>, rotation=<rot 0 0.8667 -89.4543>}]
}
constraints [
 {bone1Name=foot_r, bone2Name=calf_r, swing1LimitDegrees=45, swing2LimitDegrees=45, twistLimitDegrees=45}
 {
  bone1Name=dyn_calf_twist_upper_r
  bone2Name=calf_r
  swing1LimitDegrees=45
  swing1Motion=Locked
  swing2LimitDegrees=10
  swing2Motion=Locked
  twistLimitDegrees=20
  twistMotion=Locked
 }
 {
  bone1Name=dyn_calf_twist_back_r
  bone2Name=calf_r
  swing1LimitDegrees=45
  swing1Motion=Locked
  swing2LimitDegrees=10
  twistLimitDegrees=10
  twistMotion=Locked
 }
 {
  bone1Name=dyn_calf_twist_front_r
  bone2Name=calf_r
  swing1LimitDegrees=45
  swing1Motion=Locked
  swing2LimitDegrees=10
  twistLimitDegrees=10
  twistMotion=Locked
 }
 {
  bone1Name=dyn_calf_twist_left_r
  bone2Name=calf_r
  swing1LimitDegrees=45
  swing1Motion=Locked
  swing2LimitDegrees=10
  twistLimitDegrees=10
  twistMotion=Locked
 }
 {
  bone1Name=dyn_calf_twist_right_r
  bone2Name=calf_r
  swing1LimitDegrees=45
  swing1Motion=Locked
  swing2LimitDegrees=5
  twistLimitDegrees=10
  twistMotion=Locked
 }
]
