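# Animation Blueprint for the TechnoAwaken_Shake medium male player body, layering procedural leg motion and pose correctives over the parent mesh's pose.
# The copied pose feeds a Control Rig, right and left calf-based RigidBody simulations, then a corrective Control Rig driven by character-state and weapon flags converted to floats.
# Each update also feeds foot and calf data into material WPO cloth simulation, enabling wind and turbulence during skydiving or parachuting.
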AnimBlueprint /Game/Characters/Player/Male/Medium/Bodies/M_MED_TechnoAwaken_Shake/Meshes/M_MED_TechnoAwaken_Shake_AnimBP
target_skeleton=&/M_MED_TechnoAwaken_Shake_Skeleton
graphs [
 AnimGraph {nodes [
   n1 = Control Rig(Source=Copy Pose From Mesh() {bUseAttachedParent=true}) {
    alphaBoolBlend {blendOption=Linear}
    controlRigAssetReference {blueprintRigClass=&/M_MED_TechnoAwaken_Shake_CtrlRig.M_MED_TechnoAwaken_Shake_CtrlRig_C}
   }
   n2 = RigidBody(ComponentPose=Local To Component(LocalPose=@n1)) {
    bOverrideWorldGravity=true
    overridePhysicsAsset=&/M_MED_TechnoAwaken_Shake_Leg_R_Physics
    simulationSpace=BaseBoneSpace
    alphaBoolBlend {blendOption=Linear}
    alphaScaleBiasClamp {bMapRange=true, outRange {max=0, min=1}}
    baseBoneRef {boneName=calf_r}
    componentAppliedLinearAccClamp=<8000 8000 8000>
    componentLinearAccScale=<0.6 0.6 0.6>
    componentLinearVelScale=<0.1 0.1 0.1>
    overrideWorldGravity {z=-980}
    simSpaceSettings {maxAngularAcceleration=500, maxAngularVelocity=200, maxLinearAcceleration=400, maxLinearVelocity=200, velocityScaleZ=0, worldAlpha=0.85}
   }
   n3 = RigidBody(ComponentPose=@n2) {
    bOverrideWorldGravity=true
    overridePhysicsAsset=&/M_MED_TechnoAwaken_Shake_Leg_L_Physics
    simulationSpace=BaseBoneSpace
    alphaBoolBlend {blendOption=Linear}
    alphaScaleBiasClamp {bMapRange=true, outRange {max=0, min=1}}
    baseBoneRef {boneName=calf_l}
    componentAppliedLinearAccClamp=<8000 8000 8000>
    componentLinearAccScale=<0.6 0.6 0.6>
    componentLinearVelScale=<0.1 0.1 0.1>
    overrideWorldGravity {z=-980}
    simSpaceSettings {maxAngularAcceleration=500, maxAngularVelocity=200, maxLinearAcceleration=400, maxLinearVelocity=200, velocityScaleZ=0, worldAlpha=0.85}
   }
   n4 = Control Rig!(
    Source=Component To Local(ComponentPose=@n3)
    Alpha
    isSkyDiving:double=Get isSkyDiving() -> (isSkyDiving:double)
    isEmote:double=Get isEmote() -> (isEmote:double)
    isParachuteOpened:double=Get isParachuteOpened() -> (isParachuteOpened:double)
    isDBNO:double=Get isDBNO() -> (isDBNO:double)
    isPistol:double=Get isPistolEquiped() -> (isPistolEquiped:double)
    isAWS:double=Get isTargeting() -> (isTargeting:double)
    isSkyDiving_NOT:double=Get isSkyDiving_not() -> (isSkyDiving_not:double)
    isEmote_NOT:double=Get isEmote_not() -> (isEmote_not:double)
    isParachuteOpened_NOT:double=Get isParachuteOpened_NOT() -> (isParachuteOpened_NOT:double)
    isDBNO_NOT:double=Get isDBNO_not() -> (isDBNO_not:double)
    isPistol_NOT:double=Get isPistolEquiped_not() -> (isPistolEquiped_not:double)
    isAWS_NOT:double=Get isTargeting_not() -> (isTargeting_not:double)
    isCrouching:double=Get isCrouching() -> (isCrouching:double)
    isCrouching_NOT:double=Get isCrouching_not() -> (isCrouching_not:double)
    isMoving:double=Get isMoving() -> (isMoving:double)
    isMoving_NOT:double=Get isMoving_NOT() -> (isMoving_NOT:double)
    isRail:double=Get isRail() -> (isRail:double)
    isRail_NOT:double=Get isRail_NOT() -> (isRail_NOT:double)
    isMelee:double=Get isMeleeHave() -> (isMeleeHave:double)
    isMelee_NOT:double=Get isMelee_NOT() -> (isMelee_NOT:double)
    isFrontend:double=Get isFrontend() -> (isFrontend:double)
    isFrontend_NOT:double=Get isFrontend_NOT() -> (isFrontend_NOT:double)
    isJumpting:bool=Get bIsJumping()
    isFalling:bool=Get bIsFalling()
   ) -> (Pose) {
    alphaBoolBlend {blendOption=Linear}
    controlRigAssetReference {blueprintRigClass=&/M_MED_TechnoAwaken_Shake_Corrective_CtrlRig.M_MED_TechnoAwaken_Shake_Corrective_CtrlRig_C}
   }
   Output Pose(Result=@n4) {layerGroup=None}
  ]}
 EventGraph {nodes [
   n1 = Event BlueprintUpdateAnimation()
   > n2 = Sequence!(execute) -> (then_0, then_1, then_2:Exec, then_3:Exec)
   > Set Delta Time X(Delta Time X=@n1.DeltaTimeX)
   Is Valid(exec=@n2.then_2, InputObject=Get Player Pawn() -> (Player Pawn:Fort Player Pawn Athena Object Reference))
   > Set WPO Globals Function(
    Material=Get Material Instance Dynamic() -> (Material Instance Dynamic:Material Instance Dynamic Object Reference)
    Overlay Material=Get Overlay Material Instance Dynamic() -> (Overlay Material Instance Dynamic:Material Instance Dynamic Object Reference)
    Skeletal Mesh=Get Skeletal Mesh() -> (Skeletal Mesh:Skeletal Mesh Component Object Reference)
    "Local Noise Transform  Location"="Get Local Noise Transform  Location"() -> ("Local Noise Transform  Location":Vector)
    Local Noise Offset=Get Local Noise Offset()
   )
   > Set WPO Regional Function!(
    execute
    Material:Material Instance Dynamic Object Reference=Get Material Instance Dynamic()
    Overlay Material:Material Instance Dynamic Object Reference=Get Overlay Material Instance Dynamic()
    Skeletal Mesh:Skeletal Mesh Component Object Reference=Get Skeletal Mesh()
    Velocity Transform:Name=foot_r
    Velocity Interp Speed:double=Get Velocity Interp Speed()
    Collision Transform:Name=calf_r
    Collision Transform Axis:EAxis Enum
    Invert Collision Transform Axis:bool=false
    Delta Time X:double=@n1.DeltaTimeX
    Previous Velocity Transform Location:"Vector (by ref)"=Get R1 Velocity Transform Location()
    Previous Velocity Transform Velocity:"Vector (by ref)"=Get R1 Velocity Transform Velocity()
    Region:int=1
   ) -> (then)
   > Set WPO Regional Function(
    Material=Get Material Instance Dynamic()
    Overlay Material=Get Overlay Material Instance Dynamic()
    Skeletal Mesh=Get Skeletal Mesh()
    Velocity Transform=foot_l
    Velocity Interp Speed=Get Velocity Interp Speed()
    Collision Transform=calf_l
    Invert Collision Transform Axis=true
    Delta Time X=@n1.DeltaTimeX
    Previous Velocity Transform Location=Get R2 Velocity Transform Location()
    Previous Velocity Transform Velocity=Get R2 Velocity Transform Velocity()
    Region=2
   )
   > SetScalarParameterValue(
    self=[Get Material Instance Dynamic(), Get Overlay Material Instance Dynamic()]
    ParameterName=ClothSim_WindStrength
    Value=Get WPO Wind() -> (WPO Wind:double)
   )
   > SetScalarParameterValue(
    self=[Get Overlay Material Instance Dynamic(), Get Material Instance Dynamic()]
    ParameterName=ClothSim_TurbulenceNoiseTileDivisor
    Value=Get Turbulence Noise() -> (Turbulence Noise:float)
   )
   n3 = OR Boolean!(A=Get bIsSkydiveFloating(), B=Get bIsSkydiveDiveMode(), C:bool=Get bIsParachuteOpened()) -> (ReturnValue)
   Set isWPOEnabled(execute=@n2.then_3, isWPOEnabled:bool=@n3) -> (Output_Get:bool)
   > n4 = Branch(Condition=@n3)
   > Set Velocity Interp Speed(Velocity Interp Speed:double=40) -> (Output_Get:double)
   > Set WPO Wind(WPO Wind:double=1) -> (Output_Get:double)
   > Set Turbulence Noise(Turbulence Noise:float=80) -> (Output_Get:float)
   Set Velocity Interp Speed(execute=@n4.else, Velocity Interp Speed=0)
   > Set WPO Wind(WPO Wind=0)
   > Set Turbulence Noise(Turbulence Noise=0)
   n5 = Sequence!(execute=@n2.then_1) -> (then_0, then_1, then_2, then_3, then_4:Exec, then_5:Exec, then_6:Exec, then_7:Exec, then_8:Exec, then_9:Exec, then_10:Exec, then_11:Exec)
   n6 = Get bIsCrouching()
   Set isCrouching(
    execute=@n5.then_1
    isCrouching:double=FInterpTo(Current=Get isCrouching(), Target="To Float (Boolean)"(InBool=@n6), DeltaTime=Get Delta Time X(), InterpSpeed=0)
   ) -> (Output_Get:double)
   n7 = "To Float (Boolean)"(InBool=NOT Boolean(A=Get bIsCrouching()))
   > Set isCrouching_not(isCrouching_not:double=@n7) -> (Output_Get:double)
   n8 = Get isMelee() -> (isMelee:bool)
   Set isMeleeHave(
    execute=@n5.then_10
    isMeleeHave:double=FInterpTo(Current=Get isMeleeHave(), Target="To Float (Boolean)"(InBool=@n8), DeltaTime=Get Delta Time X(), InterpSpeed=0)
   ) -> (Output_Get:double)
   n9 = NOT Boolean(A=Get isMelee())
   > Set isMelee_NOT(isMelee_NOT:double=FInterpTo(Current=Get isMelee_NOT(), Target="To Float (Boolean)"(InBool=@n9), DeltaTime=Get Delta Time X(), InterpSpeed=0)) -> (Output_Get:double)
   n10 = Get bIsInFrontEnd()
   Set isFrontend(
    execute=@n5.then_11
    isFrontend:double=FInterpTo(Current=Get isFrontend(), Target="To Float (Boolean)"(InBool=@n10), DeltaTime=Get Delta Time X(), InterpSpeed=0)
   ) -> (Output_Get:double)
   n11 = NOT Boolean(A=Get bIsInFrontEnd())
   > Set isFrontend_NOT(isFrontend_NOT:double=FInterpTo(Current=Get isFrontend_NOT(), Target="To Float (Boolean)"(InBool=@n11), DeltaTime=Get Delta Time X(), InterpSpeed=0)) -> (Output_Get:double)
   n12 = Get bIsDBNO()
   Set isDBNO(
    execute=@n5.then_2
    isDBNO:double=FInterpTo(Current=Get isDBNO(), Target="To Float (Boolean)"(InBool=@n12), DeltaTime=Get Delta Time X(), InterpSpeed=0)
   ) -> (Output_Get:double)
   n13 = NOT Boolean(A=Get bIsDBNO())
   > Set isDBNO_not(isDBNO_not:double=FInterpTo(Current=Get isDBNO_not(), Target="To Float (Boolean)"(InBool=@n13), DeltaTime=Get Delta Time X(), InterpSpeed=0)) -> (Output_Get:double)
   n14 = Get bIsPlayingEmote()
   Set isEmote(
    execute=@n5.then_3
    isEmote:double=FInterpTo(Current=Get isEmote(), Target="To Float (Boolean)"(InBool=@n14), DeltaTime=Get Delta Time X(), InterpSpeed=0)
   ) -> (Output_Get:double)
   n15 = NOT Boolean(A=Get bIsPlayingEmote())
   > Set isEmote_not(isEmote_not:double=FInterpTo(Current=Get isEmote_not(), Target="To Float (Boolean)"(InBool=@n15), DeltaTime=Get Delta Time X(), InterpSpeed=0)) -> (Output_Get:double)
   n16 = Get isPistol() -> (isPistol:bool)
   Set isPistolEquiped(
    execute=@n5.then_4
    isPistolEquiped:double=FInterpTo(Current=Get isPistolEquiped(), Target="To Float (Boolean)"(InBool=@n16), DeltaTime=Get Delta Time X(), InterpSpeed=0)
   ) -> (Output_Get:double)
   n17 = NOT Boolean(A=Get isPistol())
   > Set isPistolEquiped_not(isPistolEquiped_not:double=FInterpTo(Current=Get isPistolEquiped_not(), Target="To Float (Boolean)"(InBool=@n17), DeltaTime=Get Delta Time X(), InterpSpeed=0)) -> (Output_Get:double)
   n18 = Get bIsSkyDiving()
   Set isSkyDiving(
    execute=@n5.then_5
    isSkyDiving:double=FInterpTo(Current=Get isSkyDiving(), Target="To Float (Boolean)"(InBool=@n18), DeltaTime=Get Delta Time X(), InterpSpeed=0)
   ) -> (Output_Get:double)
   n19 = NOT Boolean(A=Get bIsSkyDiving())
   > Set isSkyDiving_not(isSkyDiving_not:double=FInterpTo(Current=Get isSkyDiving_not(), Target="To Float (Boolean)"(InBool=@n19), DeltaTime=Get Delta Time X(), InterpSpeed=0)) -> (Output_Get:double)
   n20 = Get bIsTargeting()
   Set isTargeting(
    execute=@n5.then_6
    isTargeting:double=FInterpTo(Current=Get isTargeting(), Target="To Float (Boolean)"(InBool=@n20), DeltaTime=Get Delta Time X(), InterpSpeed=0)
   ) -> (Output_Get:double)
   n21 = NOT Boolean(A=Get bIsTargeting())
   > Set isTargeting_not(isTargeting_not:double=FInterpTo(Current=Get isTargeting_not(), Target="To Float (Boolean)"(InBool=@n21), DeltaTime=Get Delta Time X(), InterpSpeed=0)) -> (Output_Get:double)
   n22 = Get bIsParachuteOpened()
   Set isParachuteOpened(
    execute=@n5.then_7
    isParachuteOpened:double=FInterpTo(Current=Get isParachuteOpened(), Target="To Float (Boolean)"(InBool=@n22), DeltaTime=Get Delta Time X(), InterpSpeed=0)
   ) -> (Output_Get:double)
   n23 = NOT Boolean(A=Get bIsParachuteOpened())
   > Set isParachuteOpened_NOT(isParachuteOpened_NOT:double=FInterpTo(Current=Get isParachuteOpened_NOT(), Target="To Float (Boolean)"(InBool=@n23), DeltaTime=Get Delta Time X(), InterpSpeed=0)) -> (Output_Get:double)
   n24 = Get bIsMoving2D()
   Set isMoving(
    execute=@n5.then_8
    isMoving:double=FInterpTo(Current=Get isMoving(), Target="To Float (Boolean)"(InBool=@n24), DeltaTime=Get Delta Time X(), InterpSpeed=0)
   ) -> (Output_Get:double)
   n25 = NOT Boolean(A=Get bIsMoving2D())
   > Set isMoving_NOT(isMoving_NOT:double=FInterpTo(Current=Get isMoving_NOT(), Target="To Float (Boolean)"(InBool=@n25), DeltaTime=Get Delta Time X(), InterpSpeed=0)) -> (Output_Get:double)
   n26 = Get bIsOnGrindRail()
   Set isRail(
    execute=@n5.then_9
    isRail:double=FInterpTo(Current=Get isRail(), Target="To Float (Boolean)"(InBool=@n26), DeltaTime=Get Delta Time X(), InterpSpeed=0)
   ) -> (Output_Get:double)
   n27 = NOT Boolean(A=Get bIsOnGrindRail())
   > Set isRail_NOT(isRail_NOT:double=FInterpTo(Current=Get isRail_NOT(), Target="To Float (Boolean)"(InBool=@n27), DeltaTime=Get Delta Time X(), InterpSpeed=0)) -> (Output_Get:double)
   n28 = Cast To FortPlayerAnimInstance(execute=@n5.then_0, Object=GetAnimInstance(self=GetOwnerSkeletalMeshComponent()))
   > n29 = Switch on EFortWeaponCoreAnimation!(execute, Selection:EFortWeaponCoreAnimation Enum=Get WeaponCoreAnim(self=@n28)) -> (
    Melee:Exec
    Pistol:Exec
    Shotgun:Exec
    PaperBlueprint:Exec
    Rifle:Exec
    MeleeOneHand:Exec
    MachinePistol:Exec
    RocketLauncher:Exec
    GrenadeLauncher:Exec
    GoingCommando:Exec
    AssaultRifle:Exec
    TacticalShotgun:Exec
    SniperRifle:Exec
    TrapPlacement:Exec
    ShoulderLauncher:Exec
    AbilityDecoTool:Exec
    Crossbow:Exec
    C4:Exec
    RemoteControl:Exec
    DualWield:Exec
    AR_BullPup:Exec
    AR_ForwardGrip:Exec
    MedPackPaddles:Exec
    SMG_P90:Exec
    AR_DrumGun:Exec
    Consumable_Small:Exec
    Consumable_Large:Exec
    Balloon:Exec
    MountedTurret:Exec
    CreativeTool:Exec
    ExplosiveBow:Exec
    AshtonIndigo:Exec
    AshtonChicago:Exec
    MeleeDualWield:Exec
    Unarmed:Exec
   )
   Set isPistol(
    execute=[
     @n29.AssaultRifle
     @n29.GoingCommando
     @n29.GrenadeLauncher
     @n29.MachinePistol
     @n29.MeleeDualWield
     @n29.MeleeOneHand
     @n29.PaperBlueprint
     @n29.Rifle
     @n29.RocketLauncher
     @n29.Shotgun
     @n29.SniperRifle
     @n29.TacticalShotgun
     @n29.Unarmed
    ]
    isPistol=false
   )
   > Set isMelee(isMelee:bool=false) -> (Output_Get:bool)
   Set isMelee(execute=@n29.Melee, isMelee=true)
   > Set isPistol(isPistol=false)
   Set isPistol(execute=@n29.Pistol, isPistol=true)
   > Set isMelee(isMelee=false)
   Event OnCharacterCustomizationLoaded()
   > n30 = Set Skeletal Mesh(Skeletal Mesh:Skeletal Mesh Component Object Reference=GetPartSkeletalMeshComponent()) -> (Output_Get:Skeletal Mesh Component Object Reference)
   > n31 = Get Dynamic Material Instance(SkeletalMeshComponent=@n30)
   > Set Material Instance Dynamic(Material Instance Dynamic:Material Instance Dynamic Object Reference=@n31) -> (Output_Get:Material Instance Dynamic Object Reference)
   > n32 = GetOrCreateDynamicOverlayMaterial(MeshComponent=@n30, MaterialName=MI_TechnoAwaken_Shake_Body)
   > Set Overlay Material Instance Dynamic(Overlay Material Instance Dynamic:Material Instance Dynamic Object Reference=@n32) -> (Output_Get:Material Instance Dynamic Object Reference)
   Event BlueprintInitializeAnimation()
   > Set Skeletal Mesh(Skeletal Mesh=GetOwnerSkeletalMeshComponent())
   n33 = TryGetPawnOwner()
   > Is Valid(InputObject=@n33)
   > n34 = Cast To FortPlayerPawnAthena(Object=@n33)
   > Set Player Pawn(Player Pawn:Fort Player Pawn Athena Object Reference=@n34) -> (Output_Get:Fort Player Pawn Athena Object Reference)
   Reroute Node()
   > n35 = Set isWPOEnabled(isWPOEnabled=OR Boolean(A=Get bIsSkydiveFloating(), B=Get bIsSkydiveDiveMode(), C=Get bIsParachuteOpened()))
   > n36 = Branch(Condition=@n35)
   > n37 = Set WPO Wind(WPO Wind=1)
   SetScalarParameterValue(
    execute=[@n37.then, @n38.then]
    self=[Get Overlay Material Instance Dynamic(), Get Material Instance Dynamic()]
    ParameterName=ClothSim_WindStrength
    Value=Get WPO Wind()
   )
   n38 = Set WPO Wind(execute=@n36.else, WPO Wind=0)
   FInterpTo(Current=Get isCrouching_not(), Target=@n7, DeltaTime=Get Delta Time X(), InterpSpeed=0)
  ]}
]
variables [
 Delta Time X {type=number}
 Gravity Override {type=Vector, default=<0 0 -980>}
 Gravity Override Params {type=GravityOverrideParamsStruct}
 isCrouching {type=number}
 isCrouching_not {type=number}
 isDBNO {type=number}
 isDBNO_not {type=number}
 isEmote {type=number}
 isEmote_not {type=number}
 isFrontend {type=number}
 isFrontend_NOT {type=number}
 isMelee {type=boolean}
 isMelee_NOT {type=number}
 isMeleeHave {type=number}
 isMoving {type=number}
 isMoving_NOT {type=number}
 isParachuteOpened {type=number}
 isParachuteOpened_NOT {type=number}
 isPistol {type=boolean, default=true}
 isPistolEquiped {type=number}
 isPistolEquiped_not {type=number}
 isRail {type=number}
 isRail_NOT {type=number}
 isRifle {type=boolean}
 isSkyDiving {type=number}
 isSkyDiving_not {type=number}
 isTargeting {type=number}
 isTargeting_not {type=number}
 isWPOEnabled {type=boolean}
 L Previous Velocity Transform Location {type=Vector}
 L Previous Velocity Transform Velocity {type=Vector}
 Local Noise Offset {type=Vector}
 "Local Noise Transform  Location" {type=Vector}
 Material Instance Dynamic {type=MaterialInstanceDynamic}
 Overlay Material Instance Dynamic {type=MaterialInstanceDynamic}
 Player Pawn {type=FortPlayerPawnAthena}
 R1 Velocity Transform Location {type=Vector}
 R1 Velocity Transform Velocity {type=Vector}
 R2 Velocity Transform Location {type=Vector}
 R2 Velocity Transform Velocity {type=Vector}
 Skeletal Mesh {type=SkeletalMeshComponent}
 Temp {type=boolean}
 Turbulence Noise {type=number}
 Velocity Interp Speed {type=number}
 WPO Wind {type=number}
]
